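# Control Rig for the M_MED_MoonSatellite player body that procedurally corrects its dyn_ outfit bones as the body moves.
# It converts thigh, calf and upper-arm angles, foot-to-pelvis distances and spherical pose readers into alpha variables.
# These drive weighted parent constraints and additive transform offsets on front, side, back and hip chains, shoulder pads and pant legs.
# AlphaInterp smooths an isAccelerating offset and a forward-thigh correction.
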ControlRigBlueprint /BRCosmetics/Characters/Player/Male/Medium/Bodies/M_MED_MoonSatellite/Meshes/M_MED_MoonSatellite_CtrlRig
graphs [
 RigVMFunctionLibrary {links [], nodes []}
 RigVMModel {nodes [
   RigUnit_BeginExecution = &RigUnit_BeginExecution()
   n1 = &RigVMFunction_MathQuaternionToEuler(
    RotationOrder=ZXY
    Value=&RigUnit_GetRelativeTransformForItem(Child=(Type=Bone,Name="thigh_r"), Parent=(Type=Bone,Name="pelvis"), bChildInitial=False, bParentInitial=False).RelativeTransform.Rotation
   )
   RigVMVariableNode(
    Value=&RigVMFunction_MathDoubleRemap(SourceMaximum=60, SourceMinimum=40, TargetMaximum=1, TargetMinimum=0, Value=@n1.Result.Y).Result
    ExecuteContext=@RigUnit_BeginExecution.ExecutePin
   ) {variable=R leg wayOut}
   > RigVMVariableNode(Value=&RigVMFunction_MathDoubleRemap(SourceMaximum=-7, SourceMinimum=10, TargetMaximum=1, TargetMinimum=0, Value=@n1.Result.Y).Result) {variable=R thighInward alpha}
   Absolute = &RigVMFunction_MathDoubleAbs(Value=@n1.Result.Z)
   Subtract = &RigVMFunction_MathDoubleSub(A=180, B=@Absolute.Result)
   > RigVMVariableNode(Value=&RigVMFunction_MathDoubleRemap(SourceMaximum=40, SourceMinimum=0, TargetMaximum=1, TargetMinimum=0, Value=@Subtract.Result).Result) {variable=R thighForward alpha}
   n2 = &RigVMFunction_MathQuaternionToEuler(
    RotationOrder=ZXY
    Value=&RigUnit_GetRelativeTransformForItem(Child=(Type=Bone,Name="thigh_l"), Parent=(Type=Bone,Name="pelvis"), bChildInitial=False, bParentInitial=False).RelativeTransform.Rotation
   )
   > RigVMVariableNode(Value=&RigVMFunction_MathDoubleRemap(SourceMaximum=-60, SourceMinimum=-40, TargetMaximum=1, TargetMinimum=0, Value=@n2.Result.Y).Result) {variable="L  leg wayOut"}
   > RigVMVariableNode(Value=&RigVMFunction_MathDoubleRemap(SourceMaximum=-7, SourceMinimum=10, TargetMaximum=0, TargetMinimum=1, Value=@n2.Result.Y).Result) {variable=L thighInward alpha}
   > RigVMVariableNode(Value=&RigVMFunction_MathDoubleRemap(SourceMaximum=-40, SourceMinimum=0, TargetMaximum=1, TargetMinimum=0, Value=@n2.Result.Z).Result) {variable=L thighForward alpha}
   n3 = &RigUnit_GetRelativeTransformForItem(Child=(Type=Bone,Name="calf_l"), Parent=(Type=Bone,Name="thigh_l"), bChildInitial=False, bParentInitial=False)
   > RigVMVariableNode(Value=&RigVMFunction_MathDoubleRemap(
     SourceMaximum=110
     SourceMinimum=70
     TargetMaximum=1
     TargetMinimum=0
     Value=&RigVMFunction_MathQuaternionToEuler(RotationOrder=ZXY, Value=@n3.RelativeTransform.Rotation).Result.Z
    ).Result) {variable=L calfBent alpha}
   n4 = &RigVMFunction_MathVectorDistance(
    B=&RigUnit_GetTransform(Item=(Type=Bone,Name="foot_l"), bInitial=False).Transform.Translation
    A=&RigUnit_GetTransform(Item=(Type=Bone,Name="pelvis"), bInitial=False).Transform.Translation
   )
   > RigVMVariableNode(Value=&RigVMFunction_MathDoubleRemap(SourceMaximum=35, SourceMinimum=90, TargetMaximum=1, TargetMinimum=0, Value=@n4.Result).Result) {variable=L footPelvis delta}
   > RigVMVariableNode(Value=&RigVMFunction_MathDoubleRemap(SourceMaximum=25, SourceMinimum=35, TargetMaximum=1, TargetMinimum=0, Value=@n4.Result).Result) {variable=L footPelvis delta evenCloser}
   n5 = &RigVMFunction_MathVectorDistance(
    B=&RigUnit_GetTransform(Item=(Type=Bone,Name="foot_r"), bInitial=False).Transform.Translation
    A=&RigUnit_GetTransform(Item=(Type=Bone,Name="pelvis"), bInitial=False).Transform.Translation
   )
   Remap = &RigVMFunction_MathDoubleRemap(SourceMaximum=35, SourceMinimum=90, TargetMaximum=1, TargetMinimum=0, Value=@n5.Result)
   > RigVMVariableNode(Value=@Remap.Result) {variable=R footPelvis delta}
   > RigVMVariableNode(Value=&RigVMFunction_MathDoubleRemap(SourceMaximum=25, SourceMinimum=35, TargetMaximum=1, TargetMinimum=0, Value=@n5.Result).Result) {variable=R footPelvis delta evenCloser}
   RigVMFunction_MathQuaternionToEuler = &RigVMFunction_MathQuaternionToEuler(
    RotationOrder=ZYX
    Value=&RigUnit_GetRelativeTransformForItem(Child=(Type=Bone,Name="thigh_l"), Parent=(Type=Bone,Name="pelvis"), bChildInitial=False, bParentInitial=False).RelativeTransform.Rotation
   )
   > RigVMVariableNode(Value=&RigVMFunction_MathDoubleMul(B=-1, A=@RigVMFunction_MathQuaternionToEuler.Result.Z).Result) {variable=L_thigh_Z_value}
   n6 = &RigVMFunction_MathQuaternionToEuler(
    RotationOrder=ZYX
    Value=&RigUnit_GetRelativeTransformForItem(Child=(Type=Bone,Name="thigh_r"), Parent=(Type=Bone,Name="pelvis"), bChildInitial=False, bParentInitial=False).RelativeTransform.Rotation
   )
   > RigVMVariableNode(Value=&RigVMFunction_MathDoubleSub(A=180, B=&RigVMFunction_MathDoubleAbs(Value=@n6.Result.Z).Result).Result) {variable=R_thigh_Z_value}
   > n7 = &RigUnit_SphericalPoseReader(
    ActiveRegionScaleFactors=(PositiveWidth=2,NegativeWidth=2,PositiveHeight=2,NegativeHeight=2)
    ActiveRegionSize=0.05
    Debug=(bDrawDebug=false,DebugScale=25,DebugSegments=20,DebugThickness=0.25)
    DriverAxis=(X=-1,Y=0,Z=0)
    DriverItem=(Type=Bone,Name="thigh_l")
    FalloffRegionScaleFactors=(PositiveWidth=1,NegativeWidth=1,PositiveHeight=1,NegativeHeight=1)
    FalloffSize=0.25
    FlipHeightScaling=false
    OptionalParentItem=(Type=Bone)
    RotationOffset=(X=-90,Y=0,Z=-50)
   )
   > RigVMVariableNode(Value=@n7.OutputParam) {variable=L legBack alpha}
   > n8 = &RigUnit_SphericalPoseReader(
    ActiveRegionScaleFactors=(PositiveWidth=2,NegativeWidth=2,PositiveHeight=2,NegativeHeight=2)
    ActiveRegionSize=0.05
    Debug=(bDrawDebug=false,DebugScale=25,DebugSegments=20,DebugThickness=0.25)
    DriverAxis=(X=1,Y=0,Z=0)
    DriverItem=(Type=Bone,Name="thigh_r")
    FalloffRegionScaleFactors=(PositiveWidth=1,NegativeWidth=1,PositiveHeight=1,NegativeHeight=1)
    FalloffSize=0.25
    FlipHeightScaling=false
    OptionalParentItem=(Type=Bone)
    RotationOffset=(X=90,Y=0,Z=-50)
   )
   > RigVMVariableNode(Value=@n8.OutputParam) {variable=R legBack alpha}
   n9 = &RigUnit_GetRelativeTransformForItem(Child=(Type=Bone,Name="upperarm_l"), Parent=(Type=Bone,Name="clavicle_l"), bChildInitial=False, bParentInitial=False)
   > RigVMVariableNode(Value=&RigVMFunction_MathDoubleRemap(
     SourceMaximum=60
     SourceMinimum=40
     TargetMaximum=1
     TargetMinimum=0
     Value=&RigVMFunction_MathQuaternionToEuler(RotationOrder=ZXY, Value=@n9.RelativeTransform.Rotation).Result.X
    ).Result) {variable=L upperarm twist alpha}
   n10 = &RigUnit_GetRelativeTransformForItem(Child=(Type=Bone,Name="upperarm_r"), Parent=(Type=Bone,Name="clavicle_r"), bChildInitial=False, bParentInitial=False)
   > RigVMVariableNode(Value=&RigVMFunction_MathDoubleRemap(
     SourceMaximum=60
     SourceMinimum=40
     TargetMaximum=1
     TargetMinimum=0
     Value=&RigVMFunction_MathQuaternionToEuler(RotationOrder=ZXY, Value=@n10.RelativeTransform.Rotation).Result.X
    ).Result) {variable=R upperarm twist alpha}
   > /Script/ControlRig.RigUnit_PositionConstraintLocalSpaceOffset(
    Child=(Type=Bone,Name="dyn_L_side_chainMain_1")
    Filter=(bX=True,bY=True,bZ=True)
    Parents=((Item=(Type=Bone,Name="thigh_l"),Weight=0.4),(Item=(Type=Bone,Name="pelvis"),Weight=0.6))
    Weight=0.7
    bMaintainOffset=True
   ) {is=struct RigVMUnitNode}
   > &RigUnit_ParentConstraint(
    AdvancedSettings=(InterpolationType=Average,RotationOrderForFilter=XZY)
    Child=(Type=Bone,Name="dyn_L_side_chainMain_1")
    Filter=(TranslationFilter=(bX=false,bY=false,bZ=false),RotationFilter=(bX=True,bY=True,bZ=True),ScaleFilter=(bX=True,bY=True,bZ=True))
    Parents=((Item=(Type=Bone,Name="thigh_l"),Weight=0.4),(Item=(Type=Bone,Name="pelvis"),Weight=0.6))
    Weight=1
   )
   > &RigUnit_ParentConstraint(
    AdvancedSettings=(InterpolationType=Average,RotationOrderForFilter=XZY)
    Child=(Type=Bone,Name="dyn_L_side_chainMain_2")
    Filter=(TranslationFilter=(bX=false,bY=false,bZ=false),RotationFilter=(bX=true,bY=True,bZ=True),ScaleFilter=(bX=True,bY=True,bZ=True))
    Parents=((Item=(Type=Bone,Name="thigh_l"),Weight=0.9),(Item=(Type=Bone,Name="calf_l"),Weight=0.1))
    Weight=0.6
   )
   > &RigUnit_ModifyTransforms(
    ItemToModify=((Item=(Type=Bone,Name="dyn_L_side_chainMain_1"),Transform=(Rotation=(X=-0.1151,Y=-0.365,Z=-0.2778,W=0.8811),Translation=(X=0,Y=0,Z=0),Scale3D=(X=1,Y=1,Z=1))))
    WeightMinimum=0
    Weight=RigVMVariableNode() {variable=L footPelvis delta}.Value
   )
   > &RigUnit_ModifyTransforms(
    ItemToModify=((Item=(Type=Bone,Name="dyn_L_side_chainMain_1"),Transform=(Rotation=(X=-0.0522,Y=-0.2961,Z=-0.1656,W=0.9392),Translation=(X=-2,Y=4,Z=5),Scale3D=(X=1,Y=1,Z=1))))
    WeightMinimum=0
    Weight=RigVMVariableNode() {variable=L footPelvis delta evenCloser}.Value
   )
   > &RigUnit_ModifyTransforms(
    ItemToModify=((Item=(Type=Bone,Name="dyn_L_side_chainMain_1"),Transform=(Rotation=(X=-0.118,Y=-0.2199,Z=0.014,W=0.9683),Translation=(X=0,Y=0,Z=0),Scale3D=(X=1,Y=1,Z=1))))
    WeightMinimum=0
    Weight=&RigVMFunction_MathDoubleMul(B=RigVMVariableNode() {variable=R thighInward alpha}.Value, A=RigVMVariableNode() {variable=L thighInward alpha}.Value).Result
   )
   > &RigUnit_ParentConstraint(
    AdvancedSettings=(InterpolationType=Average,RotationOrderForFilter=XZY)
    Child=(Type=Bone,Name="dyn_L_front_chainMain_B_1")
    Filter=(TranslationFilter=(bX=true,bY=true,bZ=true),RotationFilter=(bX=True,bY=True,bZ=True),ScaleFilter=(bX=True,bY=True,bZ=True))
    Parents=((Item=(Type=Bone,Name="thigh_l"),Weight=1))
    Weight=1
   )
   > &RigUnit_ParentConstraint(
    AdvancedSettings=(InterpolationType=Average,RotationOrderForFilter=XZY)
    Child=(Type=Bone,Name="dyn_L_frontChain_B_1")
    Filter=(TranslationFilter=(bX=true,bY=true,bZ=true),RotationFilter=(bX=True,bY=True,bZ=True),ScaleFilter=(bX=True,bY=True,bZ=True))
    Parents=((Item=(Type=Bone,Name="dyn_L_frontChain_A_1"),Weight=1))
    Weight=1
   )
   n11 = &RigVMFunction_MathDoubleAdd(B=10, A=RigVMVariableNode() {variable=R_thigh_Z_value}.Value)
   n12 = RigVMVariableNode() {variable=L_thigh_Z_value}
   n13 = &RigVMFunction_AlphaInterp(
    InterpSpeedDecreasing=15
    InterpSpeedIncreasing=15
    Scale=1
    bClampResult=true
    Value=RigVMDispatchNode(
     True=0
     Condition=/Script/RigVM.RigVMFunction_MathDoubleGreater(B=@n11.Result, A=@n12.Value) {is=struct RigVMUnitNode, template="Greater::Execute(in A,in B,out Result)"}.Result
     False=RigVMVariableNode() {variable=R thighForward alpha}.Value
    ) {template="DISPATCH_RigVMDispatch_If(in Condition,in True,in False,out Result)"}.Result
   )
   > &RigUnit_ModifyTransforms(
    ItemToModify="((Item=(Type=Bone,Name=\"dyn_L_front_chainMain_A_1\"),Transform=(Rotation=(X=0.3827,Y=0,Z=0,W=0.9239),Translation=(X=-3,Y=-2,Z=2),Scale3D=(X=1,Y=1,Z=1))),(Item=(Type=Bone,Name=\"dyn_L_front_chainMain_B_1\"),Transform=(Rotation=(X=0,Y=-0.0872,Z=0,W=0.9962),Tran ...(truncated)"
    WeightMinimum=0
    Weight=@n13.Result
   )
   n14 = RigVMVariableNode() {variable=L calfBent alpha}
   > &RigUnit_ModifyTransforms(
    ItemToModify="((Item=(Type=Bone,Name=\"dyn_L_frontChain_A_1\"),Transform=(Rotation=(X=0,Y=0.0872,Z=0,W=0.9962),Translation=(X=0,Y=0,Z=0),Scale3D=(X=1,Y=1,Z=1))),(Item=(Type=Bone,Name=\"dyn_L_frontChain_A_2\"),Transform=(Rotation=(X=0,Y=0.1736,Z=0,W=0.9848),Translation=(X=0, ...(truncated)"
    WeightMinimum=0
    Weight=@n14.Value
   )
   > &RigUnit_ModifyTransforms(
    ItemToModify="((Item=(Type=Bone,Name=\"dyn_L_frontChain_B_1\"),Transform=(Rotation=(X=0.0151,Y=0.173,Z=-0.0858,W=0.9811),Translation=(X=0,Y=0,Z=0),Scale3D=(X=1,Y=1,Z=1))),(Item=(Type=Bone,Name=\"dyn_L_frontChain_B_2\"),Transform=(Rotation=(X=0.0151,Y=0.173,Z=-0.0858,W=0.981 ...(truncated)"
    WeightMinimum=0
    Weight=@n14.Value
   )
   > &RigUnit_ParentConstraint(
    AdvancedSettings=(InterpolationType=Average,RotationOrderForFilter=XZY)
    Child=(Type=Bone,Name="dyn_R_side_chainMain_1")
    Filter=(TranslationFilter=(bX=true,bY=true,bZ=true),RotationFilter=(bX=true,bY=True,bZ=True),ScaleFilter=(bX=True,bY=True,bZ=True))
    Parents=((Item=(Type=Bone,Name="thigh_r"),Weight=0.4),(Item=(Type=Bone,Name="pelvis"),Weight=0.6))
    Weight=1
   )
   > &RigUnit_ParentConstraint(
    AdvancedSettings=(InterpolationType=Average,RotationOrderForFilter=XZY)
    Child=(Type=Bone,Name="dyn_R_side_chainMain_2")
    Filter=(TranslationFilter=(bX=false,bY=false,bZ=true),RotationFilter=(bX=true,bY=True,bZ=True),ScaleFilter=(bX=True,bY=True,bZ=True))
    Parents=((Item=(Type=Bone,Name="thigh_r"),Weight=0.9),(Item=(Type=Bone,Name="calf_r"),Weight=0.1))
    Weight=0.6
   )
   > &RigUnit_ModifyTransforms(
    ItemToModify=((Item=(Type=Bone,Name="dyn_R_side_chainMain_1"),Transform=(Rotation=(X=-0.1151,Y=-0.365,Z=-0.2778,W=0.8811),Translation=(X=0,Y=0,Z=0),Scale3D=(X=1,Y=1,Z=1))))
    WeightMinimum=0
    Weight=RigVMVariableNode() {variable=R footPelvis delta}.Value
   )
   > &RigUnit_ModifyTransforms(
    ItemToModify="((Item=(Type=Bone,Name=\"dyn_R_side_chainMain_1\"),Transform=(Rotation=(X=-0.0522,Y=-0.2961,Z=-0.1656,W=0.9392),Translation=(X=2,Y=-4,Z=-5),Scale3D=(X=1,Y=1,Z=1))),(Item=(Type=Bone,Name=\"dyn_R_sideChain_B_2\"),Transform=(Rotation=(X=0,Y=0.342,Z=0,W=0.9397),Tr ...(truncated)"
    WeightMinimum=0
    Weight=RigVMVariableNode() {variable=R footPelvis delta evenCloser}.Value
   )
   > &RigUnit_ModifyTransforms(
    ItemToModify=((Item=(Type=Bone,Name="dyn_R_side_chainMain_1"),Transform=(Rotation=(X=0.0151,Y=-0.173,Z=0.0858,W=0.9811),Translation=(X=0,Y=0,Z=0),Scale3D=(X=1,Y=1,Z=1))))
    WeightMinimum=0
    Weight=&RigVMFunction_MathDoubleMul(B=RigVMVariableNode() {variable=R thighInward alpha}.Value, A=RigVMVariableNode() {variable=L thighInward alpha}.Value).Result
   )
   n15 = RigVMVariableNode() {variable=R leg wayOut}
   n16 = RigVMVariableNode() {variable="L  leg wayOut"}
   n17 = &RigVMFunction_MathDoubleMul(
    B=0.7
    A=&RigVMFunction_MathDoubleSub(A=1, B=&RigVMFunction_MathDoubleMul(B=@n15.Value, A=@n16.Value).Result).Result
   )
   > &RigUnit_ParentConstraint(
    AdvancedSettings=(InterpolationType=Average,RotationOrderForFilter=XZY)
    Child=(Type=Bone,Name="dyn_back_chainMain_1")
    Filter=(TranslationFilter=(bX=true,bY=true,bZ=true),RotationFilter=(bX=True,bY=True,bZ=True),ScaleFilter=(bX=True,bY=True,bZ=True))
    Parents=((Item=(Type=Bone,Name="dyn_L_side_chainMain_1"),Weight=0.5),(Item=(Type=Bone,Name="dyn_R_side_chainMain_1"),Weight=0.5))
    Weight=@n17.Result
   )
   > &RigUnit_ParentConstraint(
    AdvancedSettings=(InterpolationType=Average,RotationOrderForFilter=XZY)
    Child=(Type=Bone,Name="dyn_back_chain_1")
    Filter=(TranslationFilter=(bX=true,bY=true,bZ=true),RotationFilter=(bX=True,bY=True,bZ=True),ScaleFilter=(bX=True,bY=True,bZ=True))
    Parents=((Item=(Type=Bone,Name="dyn_L_sideChain_B_1"),Weight=0.5),(Item=(Type=Bone,Name="dyn_R_sideChain_B_1"),Weight=0.5))
    Weight=@n17.Result
   )
   n18 = &RigVMFunction_MathDoubleMul(B=RigVMVariableNode() {variable=L thighForward alpha}.Value, A=RigVMVariableNode() {variable=R footPelvis delta}.Value)
   n19 = &RigVMFunction_MathDoubleClamp(Value=&RigVMFunction_MathDoubleSub(A=&RigVMFunction_MathDoubleMul(B=2, A=@n18.Result).Result, B=RigVMVariableNode() {variable=R thighForward alpha}.Value).Result)
   Multiply = &RigVMFunction_MathDoubleMul(B=RigVMVariableNode() {variable=R thighForward alpha}.Value, A=RigVMVariableNode() {variable=L footPelvis delta}.Value)
   n20 = &RigVMFunction_MathDoubleClamp(Value=&RigVMFunction_MathDoubleSub(A=&RigVMFunction_MathDoubleMul(B=2, A=@Multiply.Result).Result, B=RigVMVariableNode() {variable=L thighForward alpha}.Value).Result)
   Add = &RigVMFunction_MathDoubleAdd(B=@n19.Result, A=@n20.Result)
   Clamp = &RigVMFunction_MathDoubleClamp(Value=&RigVMFunction_MathDoubleSub(A=1, B=@Add.Result).Result)
   n21 = RigVMVariableNode() {variable=L footPelvis delta evenCloser}
   n22 = RigVMVariableNode() {variable=R footPelvis delta evenCloser}
   > &RigUnit_ParentConstraint(
    AdvancedSettings=(InterpolationType=Average,RotationOrderForFilter=XZY)
    Child=(Type=Bone,Name="dyn_back_chainMain_2")
    Filter=(TranslationFilter=(bX=false,bY=false,bZ=false),RotationFilter=(bX=True,bY=True,bZ=True),ScaleFilter=(bX=True,bY=True,bZ=True))
    Parents=((Item=(Type=Bone,Name="dyn_L_sideChain_B_1"),Weight=1),(Item=(Type=Bone,Name="dyn_R_sideChain_B_1"),Weight=1),(Item=(Type=Bone,Name="pelvis"),Weight=0))
    Parents.2.Weight=@Clamp.Result
    Parents.1.Weight=@n19.Result
    Parents.0.Weight=@n20.Result
    Weight=&RigVMFunction_MathDoubleSub(A=1, B=&RigVMFunction_MathDoubleMax(A=@n21.Value, B=@n22.Value).Result).Result
   )
   Maximum = &RigVMFunction_MathDoubleMax(A=RigVMVariableNode() {variable=L footPelvis delta}.Value, B=RigVMVariableNode() {variable=R footPelvis delta}.Value)
   > &RigUnit_ModifyTransforms(
    ItemToModify=((Item=(Type=Bone,Name="dyn_back_chainMain_1"),Transform=(Rotation=(X=0,Y=-0.1736,Z=0,W=0.9848),Translation=(X=0,Y=0,Z=0),Scale3D=(X=1,Y=1,Z=1))))
    WeightMinimum=0
    Weight=@Maximum.Result
   )
   > &RigUnit_ModifyTransforms(
    ItemToModify="((Item=(Type=Bone,Name=\"dyn_back_chainMain_1\"),Transform=(Rotation=(X=0,Y=-0.0872,Z=0,W=0.9962),Translation=(X=-5,Y=0,Z=11),Scale3D=(X=1,Y=1,Z=1))),(Item=(Type=Bone,Name=\"dyn_back_chain_1\"),Transform=(Rotation=(X=0,Y=0.2588,Z=0,W=0.9659),Translation=(X=0,Y ...(truncated)"
    WeightMinimum=0
    Weight=&RigVMFunction_MathDoubleMax(A=RigVMVariableNode() {variable=L footPelvis delta evenCloser}.Value, B=RigVMVariableNode() {variable=R footPelvis delta evenCloser}.Value).Result
   )
   > &RigUnit_ModifyTransforms(
    ItemToModify=((Item=(Type=Bone,Name="dyn_back_chainMain_1"),Transform=(Rotation=(X=0,Y=-0.1736,Z=0,W=0.9848),Translation=(X=0,Y=0,Z=0),Scale3D=(X=1,Y=1,Z=1))))
    WeightMinimum=0
    Weight=&RigVMFunction_MathDoubleMax(A=RigVMVariableNode() {variable=L legBack alpha}.Value, B=RigVMVariableNode() {variable=R legBack alpha}.Value).Result
   )
   > &RigUnit_ParentConstraint(
    AdvancedSettings=(InterpolationType=Average,RotationOrderForFilter=XZY)
    Child=(Type=Bone,Name="dyn_L_sideChain_B_1")
    Filter=(TranslationFilter=(bX=true,bY=true,bZ=true),RotationFilter=(bX=True,bY=True,bZ=True),ScaleFilter=(bX=True,bY=True,bZ=True))
    Parents=((Item=(Type=Bone,Name="dyn_back_chain_1"),Weight=0.5),(Item=(Type=Bone,Name="dyn_L_sideChain_A_1"),Weight=0.5))
    Weight=0.7
   )
   > &RigUnit_ParentConstraint(
    AdvancedSettings=(InterpolationType=Average,RotationOrderForFilter=XZY)
    Child=(Type=Bone,Name="dyn_R_sideChain_B_1")
    Filter=(TranslationFilter=(bX=true,bY=true,bZ=true),RotationFilter=(bX=True,bY=True,bZ=True),ScaleFilter=(bX=True,bY=True,bZ=True))
    Parents=((Item=(Type=Bone,Name="dyn_back_chain_1"),Weight=0.5),(Item=(Type=Bone,Name="dyn_R_sideChain_A_1"),Weight=0.5))
    Weight=0.7
   )
   AlphaInterp = &RigVMFunction_AlphaInterp(InterpSpeedDecreasing=10, InterpSpeedIncreasing=10, Scale=1, bClampResult=true, Value=RigVMVariableNode() {variable=isAccelerating}.Value)
   > &RigUnit_ModifyTransforms(
    ItemToModify="((Item=(Type=Bone,Name=\"dyn_L_side_chainMain_1\"),Transform=(Rotation=(X=-0.0038,Y=-0.0871,Z=-0.0435,W=0.9952),Translation=(X=0,Y=0,Z=0),Scale3D=(X=1,Y=1,Z=1))),(Item=(Type=Bone,Name=\"dyn_R_side_chainMain_1\"),Transform=(Rotation=(X=0.0038,Y=-0.0871,Z=0.0435 ...(truncated)"
    WeightMinimum=0
    Weight=@AlphaInterp.Result
   )
   > &RigUnit_ParentConstraint(
    AdvancedSettings=(RotationOrderForFilter=XZY)
    Child=(Type=Bone,Name="dyn_L_hipChain_1")
    Filter=(TranslationFilter=(bX=true,bY=true,bZ=true),RotationFilter=(bX=True,bY=true,bZ=true),ScaleFilter=(bX=True,bY=True,bZ=True))
    Parents=((Item=(Type=Bone,Name="dyn_L_front_chainMain_A_1"),Weight=0.8),(Item=(Type=Bone,Name="dyn_L_front_chainMain_B_1"),Weight=0.2))
    Weight=0.6
   )
   > &RigUnit_ParentConstraint(
    AdvancedSettings=(RotationOrderForFilter=XZY)
    Child=(Type=Bone,Name="dyn_L_hipChain_2")
    Filter=(TranslationFilter=(bX=true,bY=true,bZ=true),RotationFilter=(bX=True,bY=true,bZ=true),ScaleFilter=(bX=True,bY=True,bZ=True))
    Parents=((Item=(Type=Bone,Name="dyn_L_front_chainMain_A_1"),Weight=0.8),(Item=(Type=Bone,Name="dyn_L_front_chainMain_B_1"),Weight=0.2))
    Weight=0.7
   )
   > &RigUnit_ParentConstraint(
    AdvancedSettings=(RotationOrderForFilter=XZY)
    Child=(Type=Bone,Name="dyn_R_hipChain_2")
    Filter=(TranslationFilter=(bX=true,bY=true,bZ=true),RotationFilter=(bX=True,bY=True,bZ=True),ScaleFilter=(bX=True,bY=True,bZ=True))
    Parents=((Item=(Type=Bone,Name="thigh_r"),Weight=1))
    Weight=0.8
   )
   > &RigUnit_ModifyTransforms(
    ItemToModify=((Item=(Type=Bone,Name="dyn_R_hipChain_1"),Transform=(Rotation=(X=-0.1736,Y=0,Z=0,W=0.9848),Translation=(X=0,Y=0,Z=0),Scale3D=(X=1,Y=1,Z=1))))
    WeightMinimum=0
    Weight=RigVMVariableNode() {variable=R thighForward alpha}.Value
   )
   > &RigUnit_ModifyTransforms(
    ItemToModify="((Item=(Type=Bone,Name=\"dyn_L_shoulderpad\"),Transform=(Rotation=(X=0,Y=-0.2588,Z=0,W=0.9659),Translation=(X=0,Y=1,Z=0),Scale3D=(X=1,Y=1,Z=1))),(Item=(Type=Bone,Name=\"dyn_L_plateAdjustment\"),Transform=(Rotation=(X=0,Y=0,Z=0,W=1),Translation=(X=0,Y=0,Z=0),Sc ...(truncated)"
    WeightMinimum=0
    Weight=RigVMVariableNode() {variable=L upperarm twist alpha}.Value
   )
   > &RigUnit_ModifyTransforms(
    ItemToModify="((Item=(Type=Bone,Name=\"dyn_R_shoulderpad\"),Transform=(Rotation=(X=0,Y=-0.2588,Z=0,W=0.9659),Translation=(X=0,Y=-1,Z=0),Scale3D=(X=1,Y=1,Z=1))),(Item=(Type=Bone,Name=\"dyn_R_plateAdjustment\"),Transform=(Rotation=(X=0,Y=0,Z=0,W=1),Translation=(X=0,Y=0,Z=0),S ...(truncated)"
    WeightMinimum=0
    Weight=RigVMVariableNode() {variable=R upperarm twist alpha}.Value
   )
   > &RigUnit_ParentConstraint(
    AdvancedSettings=(RotationOrderForFilter=XZY)
    Child=(Type=Bone,Name="dyn_L_pantLeg_inner_2")
    Filter=(TranslationFilter=(bX=true,bY=true,bZ=true),RotationFilter=(bX=True,bY=True,bZ=True),ScaleFilter=(bX=True,bY=True,bZ=True))
    Parents=((Item=(Type=Bone,Name="calf_l"),Weight=1))
    Weight=0.7
   )
   > &RigUnit_ParentConstraint(
    AdvancedSettings=(RotationOrderForFilter=XZY)
    Child=(Type=Bone,Name="dyn_R_pantLeg_inner_2")
    Filter=(TranslationFilter=(bX=true,bY=true,bZ=true),RotationFilter=(bX=True,bY=True,bZ=True),ScaleFilter=(bX=True,bY=True,bZ=True))
    Parents=((Item=(Type=Bone,Name="calf_r"),Weight=1))
    Weight=0.7
   )
  ]}
]
hierarchy [
 root {loc=<0 0 0>, rot=<0 0 0 1>}
 pelvis {parent=root, loc=<0 -0.7071 93.1191>, rot=<0.0127 0.707 -0.0127 -0.707>}
 spine_01 {parent=pelvis, loc=<6.7494 -0.6388 0>, rot=<0 0 0.034 -0.9994>}
 spine_02 {parent=spine_01, loc=<7.252 -0.035 0>, rot=<0 0 0.0238 -0.9997>}
 spine_03 {parent=spine_02, loc=<6.0106 0 0>, rot=<0 0 0.0149 -0.9999>}
 spine_04 {parent=spine_03, loc=<8.1023 0 0>, rot=<0 0 -0.0132 -0.9999>}
 spine_05 {parent=spine_04, loc=<10.0142 0 0>, rot=<0 0 -0.0015 -1>}
 attach_backpack {parent=spine_05, loc=<-0.123 -13.473 0>, rot=<-0.0093 -0.707 0.0093 -0.707>}
 neck_01 {parent=spine_05, loc=<11.3138 0.1529 0>, rot=<0 0 -0.2003 -0.9797>}
 neck_02 {parent=neck_01, loc=<4.5627 0 0>, rot=<0 0 0.0504 -0.9987>}
 head {parent=neck_02, loc=<5.3914 0 0>, rot=<0 0 0.115 -0.9934>}
 clavicle_l {parent=spine_05, loc=<4.3575 -0.1466 -2.3223>, rot=<0.0465 -0.7289 -0.009 -0.6829>}
 upperarm_l {parent=clavicle_l, loc=<15.5111 0 0>, rot=<0.0226 -0.3234 -0.0088 -0.9459>}
 lowerarm_l {parent=upperarm_l, loc=<26.6514 0 0>, rot=<0.0391 -0.0099 -0.2453 -0.9686>}
 hand_l {parent=lowerarm_l, loc=<22.3491 0 0>, rot=<-0.7081 -0.0507 0.0975 -0.6975>}
 thumb_01_l {parent=hand_l, loc=<1.3118 -0.3191 -2.459>, rot=<0.6546 -0.4042 0.0664 -0.6354>}
 thumb_02_l {parent=thumb_01_l, loc=<5.9336 -0.0448 0.0401>, rot=<-0.0037 0.0213 0.298 -0.9543>}
 thumb_03_l {parent=thumb_02_l, loc=<2.2223 0.3682 -0.0907>, rot=<0.0118 -0.013 0.0066 -0.9998>}
 index_metacarpal_l {parent=hand_l, loc=<1.6535 0.0183 -2.96>, rot=<0.0158 -0.0124 -0.0173 -0.9996>}
 index_01_l {parent=index_metacarpal_l, loc=<8.3008 0.0357 -0.0524>, rot=<-0.0044 0.0485 0.0924 -0.9945>}
 index_02_l {parent=index_01_l, loc=<3.3587 -0.3004 0.0452>, rot=<-0.002 0.0027 0.3225 -0.9466>}
 index_03_l {parent=index_02_l, loc=<2.5578 -0.123 -0.0252>, rot=<-0.002 0.0067 0.2558 -0.9667>}
 middle_metacarpal_l {parent=hand_l, loc=<2.1856 -0.1725 -0.5984>, rot=<0.006 0.0168 -0.0329 -0.9993>}
 middle_01_l {parent=middle_metacarpal_l, loc=<7.3275 -0.2447 0.0129>, rot=<-0.0208 0.0042 0.0858 -0.9961>}
 middle_02_l {parent=middle_01_l, loc=<4.0123 0.1401 0.1175>, rot=<-0.0218 0.0258 0.3512 -0.9357>}
 middle_03_l {parent=middle_02_l, loc=<2.6077 0.0534 0.1373>, rot=<0.0082 0.0188 0.1708 -0.9851>}
 ring_metacarpal_l {parent=hand_l, loc=<2.2074 -0.3625 1.3638>, rot=<-0.0284 0.0722 -0.0458 -0.9959>}
 ring_01_l {parent=ring_metacarpal_l, loc=<7.0295 -0.6197 -0.1446>, rot=<-0.0489 -0.0252 0.1892 -0.9804>}
 ring_02_l {parent=ring_01_l, loc=<3.6174 0.4647 -0.1027>, rot=<0 0 0.2806 -0.9598>}
 ring_03_l {parent=ring_02_l, loc=<2.4152 0.114 0.0199>, rot=<0 0 0.1189 -0.9929>}
 pinky_metacarpal_l {parent=hand_l, loc=<1.7616 -0.9753 3.3814>, rot=<-0.0397 0.0814 -0.069 -0.9935>}
 pinky_01_l {parent=pinky_metacarpal_l, loc=<7.1709 -0.4831 -0.1094>, rot=<-0.0654 -0.0035 0.2242 -0.9723>}
 pinky_02_l {parent=pinky_01_l, loc=<2.8274 0.0893 -0.1043>, rot=<0.0037 0.0011 0.2854 -0.9584>}
 pinky_03_l {parent=pinky_02_l, loc=<2.2046 0.0307 -0.0078>, rot=<-0.0002 0.0098 0.1445 -0.9895>}
 lowerarm_twist_02_l {parent=lowerarm_l, loc=<9.6199 0 0>, rot=<0 0 0 -1>}
 lowerarm_twist_01_l {parent=lowerarm_l, loc=<17.8549 0 0>, rot=<0 0 0 -1>}
 upperarm_twist_01_l {parent=upperarm_l, loc=<0.0041 0 0>, rot=<0 0 0 -1>}
 upperarm_twist_02_l {parent=upperarm_l, loc=<13.7941 0 0>, rot=<0 0 0 -1>}
 dyn_L_frontRibbon_base_2 {parent=upperarm_twist_02_l, loc=<1.7083 5.6691 0.6023>, rot=<-0.348 0.3887 0.6147 -0.5917>}
 dyn_L_frontRibbon_string_B_1 {parent=dyn_L_frontRibbon_base_2, loc=<-0.3967 -0.6077 -1.0754>, rot=<0.239 0.6293 -0.0468 -0.738>}
 dyn_L_frontRibbon_string_B_2 {parent=dyn_L_frontRibbon_string_B_1, loc=<-4.4864 0.1448 0.0241>, rot=<0 0 0 1>}
 dyn_L_backRibbon_base_2 {parent=upperarm_twist_02_l, loc=<1.9976 -6.9078 0.4957>, rot=<0.6294 -0.5746 0.325 -0.41>}
 dyn_L_backRibbon_string_B_1 {parent=dyn_L_backRibbon_base_2, loc=<0.3242 0.4691 1.2716>, rot=<0.73 0.0481 0.6417 0.2303>}
 dyn_L_backRibbon_string_B_2 {parent=dyn_L_backRibbon_string_B_1, loc=<4.4873 0 0>, rot=<0 0 0 1>}
 dyn_L_backRibbon_base_1 {parent=upperarm_l, loc=<3.7803 -7.6098 -0.9392>, rot=<-0.1 0.6147 0.7332 -0.2731>}
 dyn_L_backRibbon_A {parent=dyn_L_backRibbon_base_1, loc=<2.157 -0.2322 0.2412>, rot=<0.2401 -0.114 -0.1212 -0.9564>}
 dyn_L_backRibbon_B {parent=dyn_L_backRibbon_base_1, loc=<0.3393 0.4609 -2.2936>, rot=<0.3715 -0.38 -0.557 -0.6383>}
 dyn_L_backRibbon_string_A_1 {parent=dyn_L_backRibbon_base_1, loc=<-0.0494 -1.0235 -1.261>, rot=<0.1665 -0.8367 0.3776 -0.36>}
 dyn_L_backRibbon_string_A_2 {parent=dyn_L_backRibbon_string_A_1, loc=<4.9837 -0.5498 0.2401>, rot=<0 0 0 1>}
 dyn_L_backRibbon_string_A_3 {parent=dyn_L_backRibbon_string_A_1, loc=<6.5396 1.1405 0.1435>, rot=<0.5306 0.0405 -0.0327 -0.846>}
 dyn_L_frontRibbon_base_1 {parent=upperarm_l, loc=<3.1913 6.3384 -0.6682>, rot=<0.7331 -0.2483 0.0739 -0.6288>}
 dyn_L_frontRibbon_A {parent=dyn_L_frontRibbon_base_1, loc=<-2.5641 0.0589 -0.0359>, rot=<0.9671 0.0862 -0.0291 0.2376>}
 dyn_L_frontRibbon_B {parent=dyn_L_frontRibbon_base_1, loc=<-0.3709 -0.7517 2.3568>, rot=<0.6271 0.4958 -0.4092 0.4398>}
 dyn_L_frontRibbon_string_A_1 {parent=dyn_L_frontRibbon_base_1, loc=<0.1566 1.2203 1.2499>, rot=<-0.2966 0.03 0.915 -0.2718>}
 dyn_L_frontRibbon_string_A_2 {parent=dyn_L_frontRibbon_string_A_1, loc=<-6.437 0.5536 0.5276>, rot=<0 0 0 -1>}
 dyn_L_frontRibbon_string_A_3 {parent=dyn_L_frontRibbon_string_A_1, loc=<-4.7839 -0.5402 -0.2253>, rot=<-0.3194 0.0145 -0.0175 -0.9473>}
 dyn_L_shoulderpad {parent=upperarm_l, loc=<7.4533 0 0>, rot=<0.506 0.4863 0.4751 -0.5309>}
 dyn_L_plateAdjustment {parent=dyn_L_shoulderpad, loc=<3.7898 -6.285 -0.8541>, rot=<0.5 0.5 0.5 0.5>}
 attach_cape {parent=spine_05, loc=<9.017 -11.714 0>, rot=<0.0616 -0.7044 -0.0616 -0.7044>}
 clavicle_r {parent=spine_05, loc=<4.3575 -0.1466 2.3223>, rot=<0.7289 0.0465 -0.6829 0.009>}
 upperarm_r {parent=clavicle_r, loc=<-15.5111 0 0>, rot=<0.0226 -0.3234 -0.0088 -0.9459>}
 lowerarm_r {parent=upperarm_r, loc=<-26.6514 0 0>, rot=<0.0391 -0.0099 -0.2453 -0.9686>}
 hand_r {parent=lowerarm_r, loc=<-22.3491 0 0>, rot=<-0.7081 -0.0507 0.0975 -0.6975>}
 thumb_01_r {parent=hand_r, loc=<-1.3118 0.3191 2.459>, rot=<0.6546 -0.4042 0.0664 -0.6354>}
 thumb_02_r {parent=thumb_01_r, loc=<-5.9336 0.0448 -0.0401>, rot=<-0.0037 0.0213 0.298 -0.9543>}
 thumb_03_r {parent=thumb_02_r, loc=<-2.2223 -0.3682 0.0907>, rot=<0.0118 -0.013 0.0066 -0.9998>}
 index_metacarpal_r {parent=hand_r, loc=<-1.6535 -0.0183 2.96>, rot=<0.0158 -0.0124 -0.0173 -0.9996>}
 index_01_r {parent=index_metacarpal_r, loc=<-8.3008 -0.0357 0.0524>, rot=<-0.0044 0.0485 0.0924 -0.9945>}
 index_02_r {parent=index_01_r, loc=<-3.3587 0.3004 -0.0452>, rot=<-0.002 0.0027 0.3225 -0.9466>}
 index_03_r {parent=index_02_r, loc=<-2.5578 0.123 0.0252>, rot=<-0.002 0.0067 0.2558 -0.9667>}
 middle_metacarpal_r {parent=hand_r, loc=<-2.1856 0.1725 0.5984>, rot=<0.006 0.0168 -0.0329 -0.9993>}
 middle_01_r {parent=middle_metacarpal_r, loc=<-7.3275 0.2447 -0.0129>, rot=<-0.0208 0.0042 0.0858 -0.9961>}
 middle_02_r {parent=middle_01_r, loc=<-4.0123 -0.1401 -0.1175>, rot=<-0.0218 0.0258 0.3512 -0.9357>}
 middle_03_r {parent=middle_02_r, loc=<-2.6077 -0.0534 -0.1373>, rot=<0.0082 0.0188 0.1708 -0.9851>}
 ring_metacarpal_r {parent=hand_r, loc=<-2.2074 0.3625 -1.3638>, rot=<-0.0284 0.0722 -0.0458 -0.9959>}
 ring_01_r {parent=ring_metacarpal_r, loc=<-7.0295 0.6197 0.1446>, rot=<-0.0489 -0.0252 0.1892 -0.9804>}
 ring_02_r {parent=ring_01_r, loc=<-3.6174 -0.4647 0.1027>, rot=<0 0 0.2806 -0.9598>}
 ring_03_r {parent=ring_02_r, loc=<-2.4152 -0.114 -0.0199>, rot=<0 0 0.1189 -0.9929>}
 pinky_metacarpal_r {parent=hand_r, loc=<-1.7616 0.9753 -3.3814>, rot=<-0.0397 0.0814 -0.069 -0.9935>}
 pinky_01_r {parent=pinky_metacarpal_r, loc=<-7.1709 0.4831 0.1094>, rot=<-0.0654 -0.0035 0.2242 -0.9723>}
 pinky_02_r {parent=pinky_01_r, loc=<-2.8274 -0.0893 0.1043>, rot=<0.0037 0.0011 0.2854 -0.9584>}
 pinky_03_r {parent=pinky_02_r, loc=<-2.2046 -0.0307 0.0078>, rot=<-0.0002 0.0098 0.1445 -0.9895>}
 lowerarm_twist_02_r {parent=lowerarm_r, loc=<-9.6199 0 0>, rot=<0 0 0 -1>}
 lowerarm_twist_01_r {parent=lowerarm_r, loc=<-17.8549 0 0>, rot=<0 0 0 -1>}
 upperarm_twist_01_r {parent=upperarm_r, loc=<-0.0041 0 0>, rot=<0 0 0 -1>}
 upperarm_twist_02_r {parent=upperarm_r, loc=<-13.7941 0 0>, rot=<0 0 0 -1>}
 dyn_R_frontRibbon_base_2 {parent=upperarm_twist_02_r, loc=<-1.7083 -5.6691 -0.6023>, rot=<-0.348 0.3887 0.6147 -0.5917>}
 dyn_R_frontRibbon_string_B_1 {parent=dyn_R_frontRibbon_base_2, loc=<0.3967 0.6076 1.0751>, rot=<0.239 0.6293 -0.0468 -0.738>}
 dyn_R_frontRibbon_string_B_2 {parent=dyn_R_frontRibbon_string_B_1, loc=<4.487 -0.1449 -0.024>, rot=<0 0 0 1>}
 dyn_R_backRibbon_base_2 {parent=upperarm_twist_02_r, loc=<-1.9976 6.9078 -0.4957>, rot=<0.6294 -0.5746 0.325 -0.41>}
 dyn_R_backRibbon_string_B_1 {parent=dyn_R_backRibbon_base_2, loc=<-0.3241 -0.4691 -1.2714>, rot=<0.73 0.0481 0.6417 0.2303>}
 dyn_R_backRibbon_string_B_2 {parent=dyn_R_backRibbon_string_B_1, loc=<-4.4879 -0.0001 0>, rot=<0 0 0 1>}
 dyn_R_backRibbon_base_1 {parent=upperarm_r, loc=<-3.7803 7.6098 0.9392>, rot=<-0.1 0.6147 0.7332 -0.2731>}
 dyn_R_backRibbon_A {parent=dyn_R_backRibbon_base_1, loc=<-2.1571 0.2321 -0.2413>, rot=<0.2401 -0.114 -0.1212 -0.9564>}
 dyn_R_backRibbon_B {parent=dyn_R_backRibbon_base_1, loc=<-0.3392 -0.4608 2.2937>, rot=<0.3715 -0.38 -0.557 -0.6383>}
 dyn_R_backRibbon_string_A_1 {parent=dyn_R_backRibbon_base_1, loc=<0.0497 1.0237 1.2612>, rot=<0.1665 -0.8367 0.3776 -0.36>}
 dyn_R_backRibbon_string_A_2 {parent=dyn_R_backRibbon_string_A_1, loc=<-4.9837 0.5498 -0.2401>, rot=<0 0 0 -1>}
 dyn_R_backRibbon_string_A_3 {parent=dyn_R_backRibbon_string_A_1, loc=<-6.539 -1.1404 -0.1435>, rot=<0.5306 0.0405 -0.0327 -0.846>}
 dyn_R_frontRibbon_base_1 {parent=upperarm_r, loc=<-3.1913 -6.3384 0.6682>, rot=<0.7331 -0.2483 0.0739 -0.6288>}
 dyn_R_frontRibbon_A {parent=dyn_R_frontRibbon_base_1, loc=<2.5643 -0.0587 0.036>, rot=<0.9671 0.0862 -0.0291 0.2376>}
 dyn_R_frontRibbon_B {parent=dyn_R_frontRibbon_base_1, loc=<0.3711 0.7519 -2.3567>, rot=<0.6271 0.4958 -0.4092 0.4398>}
 dyn_R_frontRibbon_string_A_1 {parent=dyn_R_frontRibbon_base_1, loc=<-0.1569 -1.2204 -1.2501>, rot=<-0.2966 0.03 0.915 -0.2718>}
 dyn_R_frontRibbon_string_A_2 {parent=dyn_R_frontRibbon_string_A_1, loc=<6.4371 -0.5535 -0.5276>, rot=<0 0 0 -1>}
 dyn_R_frontRibbon_string_A_3 {parent=dyn_R_frontRibbon_string_A_1, loc=<4.7832 0.5402 0.2252>, rot=<-0.3194 0.0145 -0.0175 -0.9473>}
 dyn_R_shoulderpad {parent=upperarm_r, loc=<-7.453 0 0>, rot=<0.506 0.4863 0.4751 -0.5309>}
 dyn_R_plateAdjustment {parent=dyn_R_shoulderpad, loc=<-3.7896 6.2845 0.8541>, rot=<0.5 0.5 0.5 0.5>}
 dyn_R_hipChain_1 {parent=spine_01, loc=<3.4742 5.8376 7.755>, rot=<0.1541 0.4143 0.8968 0.0206>}
 dyn_R_hipChain_2 {parent=dyn_R_hipChain_1, loc=<8.2535 0 0>, rot=<-0.0363 0.0024 0.0126 -0.9993>}
 dyn_hipFroggy {parent=dyn_R_hipChain_1, loc=<10.8361 4.3394 1.0046>, rot=<0.8847 -0.1091 -0.1244 0.4358>}
 dyn_swords_1 {parent=spine_01, loc=<4.3937 2.6989 12.2515>, rot=<-0.1598 -0.9351 0.0633 0.3099>}
 dyn_swords_2 {parent=dyn_swords_1, loc=<7.2806 -0.478 -3.6536>, rot=<-0.1226 0.1187 -0.0147 -0.9852>}
 dyn_swords_3_A {parent=dyn_swords_2, loc=<4.4427 1.1325 0.0394>, rot=<0.0459 0.0993 -0.0579 -0.9923>}
 dyn_swords_3_B {parent=dyn_swords_2, loc=<4.239 -1.8844 0.3593>, rot=<-0.0352 0.1159 0.1533 -0.9807>}
 thigh_l {parent=pelvis, loc=<-2.5693 -0.0109 -7.7501>, rot=<0.0012 0.0572 0.0155 -0.9982>}
 calf_l {parent=thigh_l, loc=<-39.0484 0 0>, rot=<0.0006 -0.0161 -0.0391 -0.9991>}
 foot_l {parent=calf_l, loc=<-42.561 0.1077 0.4292>, rot=<0.0418 -0.0455 0.0343 -0.9975>}
 ball_l {parent=foot_l, loc=<-6.1599 13.8264 0.6806>, rot=<0.0001 -0.0001 -0.7063 -0.7079>}
 calf_twist_02_l {parent=calf_l, loc=<-18.8064 -0.0967 0.0397>, rot=<0.0237 0.0095 -0.0035 -0.9997>}
 calf_twist_01_l {parent=calf_l, loc=<-35.813 -0.0967 0.0397>, rot=<0.0237 0.0095 -0.0035 -0.9997>}
 dyn_L_pantLeg_back_1 {parent=calf_l, loc=<-4.0049 -14.7449 1.1501>, rot=<0.9932 -0.0085 0.0148 -0.1152>}
 dyn_L_pantLeg_back_2 {parent=dyn_L_pantLeg_back_1, loc=<-16.8467 0 -0.0001>, rot=<0 0 0 1>}
 dyn_L_pantLeg_outer_1 {parent=calf_l, loc=<-1.3265 -1.6325 -8.4517>, rot=<0.7503 0.0102 0.0559 -0.6586>}
 dyn_L_pantLeg_outer_2 {parent=dyn_L_pantLeg_outer_1, loc=<-17.1996 0 0>, rot=<0 0 0 1>}
 dyn_L_pantLeg_inner_1 {parent=thigh_l, loc=<-17.2967 -1.5516 6.4593>, rot=<-0.7169 -0.0266 0.0457 -0.6952>}
 dyn_L_pantLeg_inner_2 {parent=dyn_L_pantLeg_inner_1, loc=<-33.4071 0.1914 0.3104>, rot=<0 0 0 1>}
 dyn_L_pantLeg_front_1 {parent=thigh_l, loc=<-18.061 8.5518 -1.9219>, rot=<0.3145 -0.033 0.036 -0.948>}
 dyn_L_pantLeg_front_2 {parent=dyn_L_pantLeg_front_1, loc=<-33.4384 0.4039 1.7457>, rot=<0 0 0 1>}
 thigh_twist_01_l {parent=thigh_l, loc=<-21.6303 0 0>, rot=<0.0013 0 0 -1>}
 dyn_L_front_chainMain_A_1 {parent=thigh_l, loc=<-0.5422 10.281 3.2499>, rot=<0.0021 -0.8137 -0.577 -0.0701>}
 dyn_L_front_chainMain_A_2 {parent=dyn_L_front_chainMain_A_1, loc=<15.1316 -0.6337 0.3857>, rot=<0.0645 -0.0143 -0.016 -0.9977>}
 dyn_L_frontChain_A_1 {parent=dyn_L_front_chainMain_A_2, loc=<16.3577 0 0>, rot=<0.0232 0.0128 0.0022 -0.9996>}
 dyn_L_frontChain_A_2 {parent=dyn_L_frontChain_A_1, loc=<12.5257 0 0>, rot=<-0.0868 0.0155 0.0205 -0.9959>}
 dyn_L_frontChain_A_3 {parent=dyn_L_frontChain_A_2, loc=<9.334 0.0073 -0.4378>, rot=<-0.012 -0.0235 -0.1496 -0.9884>}
 thigh_r {parent=pelvis, loc=<-2.5693 -0.0109 7.7501>, rot=<0.0572 -0.0012 0.9982 0.0155>}
 calf_r {parent=thigh_r, loc=<39.0484 0 0>, rot=<0.0006 -0.0161 -0.0391 -0.9991>}
 foot_r {parent=calf_r, loc=<42.561 -0.1077 -0.4292>, rot=<0.0418 -0.0455 0.0343 -0.9975>}
 ball_r {parent=foot_r, loc=<6.1599 -13.8264 -0.6806>, rot=<0.0001 -0.0001 -0.7063 -0.7079>}
 calf_twist_02_r {parent=calf_r, loc=<18.8064 0.0967 -0.0397>, rot=<0.0237 0.0095 -0.0035 -0.9997>}
 calf_twist_01_r {parent=calf_r, loc=<35.813 0.0967 -0.0397>, rot=<0.0237 0.0095 -0.0035 -0.9997>}
 dyn_R_pantLeg_back_1 {parent=calf_r, loc=<4.0049 14.7449 -1.1501>, rot=<0.9932 -0.0085 0.0148 -0.1152>}
 dyn_R_pantLeg_back_2 {parent=dyn_R_pantLeg_back_1, loc=<16.8467 0 0.0001>, rot=<0 0 0 1>}
 dyn_R_pantLeg_outer_1 {parent=calf_r, loc=<1.3266 1.6325 8.4517>, rot=<0.7503 0.0102 0.0559 -0.6586>}
 dyn_R_pantLeg_outer_2 {parent=dyn_R_pantLeg_outer_1, loc=<17.1996 0 0>, rot=<0 0 0 1>}
 dyn_R_pantLeg_inner_1 {parent=thigh_r, loc=<17.2967 1.5516 -6.4593>, rot=<-0.7169 -0.0266 0.0457 -0.6952>}
 dyn_R_pantLeg_inner_2 {parent=dyn_R_pantLeg_inner_1, loc=<33.4071 -0.1914 -0.3104>, rot=<0 0 0 1>}
 dyn_R_pantLeg_front_1 {parent=thigh_r, loc=<18.061 -8.5518 1.9219>, rot=<0.3145 -0.033 0.036 -0.948>}
 dyn_R_pantLeg_front_2 {parent=dyn_R_pantLeg_front_1, loc=<33.4384 -0.4039 -1.7457>, rot=<0 0 0 1>}
 thigh_twist_01_r {parent=thigh_r, loc=<21.6303 0 0>, rot=<0.0013 0 0 -1>}
 dyn_L_hipChain_1 {parent=pelvis, loc=<4.1636 8.1549 -6.703>, rot=<-0.0754 -0.8388 -0.5218 -0.1361>}
 dyn_L_hipChain_2 {parent=dyn_L_hipChain_1, loc=<5.5805 -0.0372 -0.0179>, rot=<0.0582 0.0097 -0.0175 -0.9981>}
 dyn_L_hipChain_3 {parent=dyn_L_hipChain_2, loc=<5.6514 0 0.3536>, rot=<-0.0531 -0.0802 0.0119 -0.9953>}
 dyn_hipFishy {parent=dyn_L_hipChain_1, loc=<4.4512 -3.1736 1.2145>, rot=<0.0238 -0.0587 -0.0757 -0.9951>}
 dyn_back_chainMain_1 {parent=pelvis, loc=<-3.387 -12.059 -0.9631>, rot=<0.0672 -0.7105 0.6998 -0.0301>}
 dyn_back_chainMain_2 {parent=dyn_back_chainMain_1, loc=<15.7983 0 0.4143>, rot=<-0.0151 -0.0038 -0.0247 -0.9996>}
 dyn_back_chain_1 {parent=dyn_back_chainMain_2, loc=<15.367 0.0072 0.6031>, rot=<0.0226 0.0151 -0.0361 -0.999>}
 dyn_back_chain_2 {parent=dyn_back_chain_1, loc=<14.9151 0 0>, rot=<0.0575 0.0055 0.0363 -0.9977>}
 dyn_back_chain_3 {parent=dyn_back_chain_2, loc=<13.4512 0 0>, rot=<-0.0004 0.0072 0.2476 -0.9688>}
 dyn_R_side_chainMain_1 {parent=pelvis, loc=<-1.4326 -7.6462 11>, rot=<0.9535 0.0191 -0.1213 -0.2752>}
 dyn_R_side_chainMain_2 {parent=dyn_R_side_chainMain_1, loc=<-17.1614 0 0>, rot=<-0.1937 -0.0224 0.0475 -0.9797>}
 dyn_R_sideChain_A_1 {parent=dyn_R_side_chainMain_2, loc=<-14.4317 -6.8955 2.092>, rot=<0.1973 0.0235 -0.0486 -0.9789>}
 dyn_R_sideChain_A_2 {parent=dyn_R_sideChain_A_1, loc=<-12.996 0 0>, rot=<0.0558 0.0073 -0.0156 -0.9983>}
 dyn_R_sideChain_A_3 {parent=dyn_R_sideChain_A_2, loc=<-9.7171 0 0>, rot=<-0.0051 0.0057 -0.092 -0.9957>}
 dyn_R_sideChain_A_4 {parent=dyn_R_sideChain_A_3, loc=<-10.0723 0 0>, rot=<-0.0566 -0.0528 -0.3286 -0.9413>}
 dyn_R_sideChain_B_1 {parent=dyn_R_side_chainMain_2, loc=<-16.5409 5.8468 0.1555>, rot=<-0.0145 0.005 -0.0348 -0.9993>}
 dyn_R_sideChain_B_2 {parent=dyn_R_sideChain_B_1, loc=<-12.3701 0.1378 0.0325>, rot=<-0.1158 0.0023 0.0285 -0.9929>}
 dyn_R_sideChain_B_3 {parent=dyn_R_sideChain_B_2, loc=<-10.8364 -0.4869 -0.0957>, rot=<0.0835 0.04 -0.1079 -0.9898>}
 dyn_R_sideChain_B_4 {parent=dyn_R_sideChain_B_3, loc=<-10.1749 -0.6388 0.9964>, rot=<-0.1227 -0.1575 -0.3634 -0.91>}
 dyn_L_side_chainMain_1 {parent=pelvis, loc=<-1.433 -7.646 -11>, rot=<0.0191 -0.9535 0.2752 -0.1213>}
 dyn_L_side_chainMain_2 {parent=dyn_L_side_chainMain_1, loc=<17.165 0.0093 0.0148>, rot=<-0.1937 -0.0224 0.0475 -0.9797>}
 dyn_L_sideChain_A_1 {parent=dyn_L_side_chainMain_2, loc=<14.4317 6.8956 -2.092>, rot=<0.1973 0.0235 -0.0486 -0.9789>}
 dyn_L_sideChain_A_2 {parent=dyn_L_sideChain_A_1, loc=<12.9959 0 0>, rot=<0.0558 0.0073 -0.0156 -0.9983>}
 dyn_L_sideChain_A_3 {parent=dyn_L_sideChain_A_2, loc=<9.7172 0 0>, rot=<-0.0051 0.0057 -0.092 -0.9957>}
 dyn_L_sideChain_A_4 {parent=dyn_L_sideChain_A_3, loc=<10.0722 0 0>, rot=<-0.0566 -0.0528 -0.3286 -0.9413>}
 dyn_L_sideChain_B_1 {parent=dyn_L_side_chainMain_2, loc=<16.5408 -5.8467 -0.1555>, rot=<-0.0145 0.005 -0.0348 -0.9993>}
 dyn_L_sideChain_B_2 {parent=dyn_L_sideChain_B_1, loc=<12.3701 -0.1378 -0.0326>, rot=<-0.1158 0.0023 0.0285 -0.9929>}
 dyn_L_sideChain_B_3 {parent=dyn_L_sideChain_B_2, loc=<10.8364 0.4869 0.0958>, rot=<0.0835 0.04 -0.1079 -0.9898>}
 dyn_L_sideChain_B_4 {parent=dyn_L_sideChain_B_3, loc=<10.1749 0.6388 -0.9964>, rot=<-0.1227 -0.1575 -0.3634 -0.91>}
 dyn_L_front_chainMain_B_1 {parent=pelvis, loc=<-3.4546 2.0687 -12.2452>, rot=<-0.0471 -0.9712 -0.179 -0.1501>}
 dyn_L_front_chainMain_B_2 {parent=dyn_L_front_chainMain_B_1, loc=<15.2037 -0.2479 0.6792>, rot=<0.0643 -0.0152 -0.0026 -0.9978>}
 dyn_L_frontChain_B_1 {parent=dyn_L_front_chainMain_B_2, loc=<15.8477 -0.6563 0.0289>, rot=<0.0117 -0.0387 0.0462 -0.9981>}
 dyn_L_frontChain_B_2 {parent=dyn_L_frontChain_B_1, loc=<12.6147 -0.4676 0.0717>, rot=<0.0782 0.0576 0.0396 -0.9945>}
 dyn_L_frontChain_B_3 {parent=dyn_L_frontChain_B_2, loc=<9.2338 0.4749 0>, rot=<0.0373 0.0104 -0.1334 -0.9903>}
 dyn_L_frontChain_B_4 {parent=dyn_L_frontChain_B_3, loc=<9.2666 0 0>, rot=<-0.1355 -0.0509 -0.2705 -0.9518>}
]
variables [
 "L  leg wayOut" {type=double, default=0}
 L calfBent alpha {type=double, default=0}
 L footPelvis delta {type=double, default=0}
 L footPelvis delta evenCloser {type=double, default=0}
 L legBack alpha {type=float, default=0}
 L thighForward alpha {type=double, default=0}
 L thighInward alpha {type=double, default=0}
 L upperarm twist alpha {type=double, default=0}
 L_thigh_Z_value {type=double, default=0}
 R footPelvis delta {type=double, default=0}
 R footPelvis delta evenCloser {type=double, default=0}
 R leg wayOut {type=double, default=0}
 R legBack alpha {type=float, default=0}
 R thighForward alpha {type=double, default=0}
 R thighInward alpha {type=double, default=0}
 R upperarm twist alpha {type=double, default=0}
 R_thigh_Z_value {type=double, default=0}
 isAccelerating {type=double, default=0}
]
