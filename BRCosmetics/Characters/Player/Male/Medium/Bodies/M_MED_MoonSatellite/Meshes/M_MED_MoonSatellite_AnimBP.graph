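# Animation Blueprint for the MoonSatellite male medium cosmetic body, adding secondary motion to its ribbons, hip chains and swords.
# It copies the attached parent's pose into a Control Rig fed by isAccelerating.
# Weighted constraints anchor ribbon bases to arm, clavicle and spine bones, then chained spring-based AnimDynamics nodes simulate dangling bones.
# The output blends frontend-idle and non-frontend cached poses, additively rolls the sword bone and constrains dyn_hipFishy to the left hip chain.
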AnimBlueprint /BRCosmetics/Characters/Player/Male/Medium/Bodies/M_MED_MoonSatellite/Meshes/M_MED_MoonSatellite_AnimBP
target_skeleton=&/M_MED_MoonSatellite_Skeleton
graphs [
 AnimGraph {nodes [
   n1 = Use cached pose 'FrontendIdleCache'()
   n2 = Use cached pose 'nonFrontendIdleCache'()
   n3 = Get frontendIdleAlphaBool() -> (frontendIdleAlphaBool:bool)
   n4 = "Transform (Modify) Bone - Bone: dyn_swords_1"!(
    ComponentPose=Local To Component(LocalPose=Blend Poses by bool(BlendPose_0=@n1, BlendPose_1=@n2, bActiveValue=@n3) {blendType=Linear})
    bAlphaBoolEnabled:bool=True
    Rotation="0,0,-25"
   ) -> (Pose) {
    alphaInputType=Bool
    rotationMode=BMM_Additive
    rotationSpace=BCS_BoneSpace
    alphaBoolBlend {blendInTime=0.5, blendOption=Linear, blendOutTime=0.5}
    boneToModify {boneName=dyn_swords_1}
    rotation {roll=-25}
   }
   n5 = "Constraint - dyn_hipFishy (PP)"!(ComponentPose=@n4, Alpha=1, ConstraintWeights_0=0.7073, ConstraintWeights_1:float=0.2927) -> (Pose) {
    alphaBoolBlend {blendOption=Linear}
    boneToModify {boneName=dyn_hipFishy}
    constraintSetup [{targetBone {boneName=dyn_L_hipChain_1}}, {targetBone {boneName=dyn_L_hipChain_2}}]
    constraintWeights [1,1]
   }
   Output Pose(Result=Component To Local(ComponentPose=@n5)) {layerGroup=None}
   n6 = Control Rig!(Source=Copy Pose From Mesh() {bUseAttachedParent=true}, Alpha, isAccelerating:double=Get isAccelerating() -> (isAccelerating:double)) -> (Pose) {
    alphaBoolBlend {blendOption=Linear}
    controlRigAssetReference {blueprintRigClass=&/M_MED_MoonSatellite_CtrlRig.M_MED_MoonSatellite_CtrlRig_C}
   }
   n7 = "Constraint - dyn_L_frontRibbon_base_1 (PPP)"!(
    ComponentPose=Local To Component(LocalPose=@n6)
    Alpha=1
    ConstraintWeights_0=0.5736
    ConstraintWeights_1:float=0.2322
    ConstraintWeights_2:float=0.1942
   ) -> (Pose) {
    alphaBoolBlend {blendOption=Linear}
    boneToModify {boneName=dyn_L_frontRibbon_base_1}
    constraintSetup [{targetBone {boneName=upperarm_l}}, {targetBone {boneName=upperarm_twist_01_l}}, {targetBone {boneName=clavicle_l}}]
    constraintWeights [1,1,1]
   }
   n8 = "Constraint - dyn_L_frontRibbon_base_2 (PP)"!(ComponentPose=@n7, Alpha=1, ConstraintWeights_0=0.5103, ConstraintWeights_1:float=0.4897) -> (Pose) {
    alphaBoolBlend {blendOption=Linear}
    boneToModify {boneName=dyn_L_frontRibbon_base_2}
    constraintSetup [{targetBone {boneName=upperarm_l}}, {targetBone {boneName=upperarm_twist_02_l}}]
    constraintWeights [1,1]
   }
   n9 = "Constraint - dyn_L_backRibbon_base_1 (PPPPP)"!(
    ComponentPose=@n8
    Alpha=1
    ConstraintWeights_0=0.4074
    ConstraintWeights_1:float=0.2549
    ConstraintWeights_2:float=0.1959
    ConstraintWeights_3:float=0.0851
    ConstraintWeights_4:float=0.0567
   ) -> (Pose) {
    alphaBoolBlend {blendOption=Linear}
    boneToModify {boneName=dyn_L_backRibbon_base_1}
    constraintSetup [
     {targetBone {boneName=upperarm_l}}
     {targetBone {boneName=clavicle_l}}
     {targetBone {boneName=upperarm_twist_01_l}}
     {targetBone {boneName=upperarm_twist_02_l}}
     {targetBone {boneName=spine_04}}
    ]
    constraintWeights [1,1,1,1,1]
   }
   n10 = "Constraint - dyn_L_backRibbon_base_2 (PP)"!(ComponentPose=@n9, Alpha=1, ConstraintWeights_0=0.5038, ConstraintWeights_1:float=0.4962) -> (Pose) {
    alphaBoolBlend {blendOption=Linear}
    boneToModify {boneName=dyn_L_backRibbon_base_2}
    constraintSetup [{targetBone {boneName=upperarm_twist_02_l}}, {targetBone {boneName=upperarm_l}}]
    constraintWeights [1,1]
   }
   n11 = "Constraint - dyn_R_frontRibbon_base_1 (PPP)"!(ComponentPose=@n10, Alpha=1, ConstraintWeights_0=0.782, ConstraintWeights_1:float=0.1219, ConstraintWeights_2:float=0.0961) -> (Pose) {
    alphaBoolBlend {blendOption=Linear}
    boneToModify {boneName=dyn_R_frontRibbon_base_1}
    constraintSetup [{targetBone {boneName=upperarm_r}}, {targetBone {boneName=clavicle_r}}, {targetBone {boneName=upperarm_twist_01_r}}]
    constraintWeights [1,1,1]
   }
   n12 = "Constraint - dyn_R_frontRibbon_base_2 (PP)"!(ComponentPose=@n11, Alpha=1, ConstraintWeights_0=0.7349, ConstraintWeights_1:float=0.2651) -> (Pose) {
    alphaBoolBlend {blendOption=Linear}
    boneToModify {boneName=dyn_R_frontRibbon_base_2}
    constraintSetup [{targetBone {boneName=upperarm_r}}, {targetBone {boneName=upperarm_twist_02_r}}]
    constraintWeights [1,1]
   }
   n13 = "Constraint - dyn_R_backRibbon_base_1 (PPPP)"!(
    ComponentPose=@n12
    Alpha=1
    ConstraintWeights_0=0.5332
    ConstraintWeights_1:float=0.2504
    ConstraintWeights_2:float=0.152
    ConstraintWeights_3:float=0.0644
   ) -> (Pose) {
    alphaBoolBlend {blendOption=Linear}
    boneToModify {boneName=dyn_R_backRibbon_base_1}
    constraintSetup [
     {targetBone {boneName=upperarm_r}}
     {targetBone {boneName=clavicle_r}}
     {targetBone {boneName=upperarm_twist_01_r}}
     {targetBone {boneName=spine_04}}
    ]
    constraintWeights [1,1,1,1]
   }
   n14 = "Constraint - dyn_R_backRibbon_base_2 (PP)"!(ComponentPose=@n13, Alpha=1, ConstraintWeights_0=0.7146, ConstraintWeights_1:float=0.2854) -> (Pose) {
    alphaBoolBlend {blendOption=Linear}
    boneToModify {boneName=dyn_R_backRibbon_base_2}
    constraintSetup [{targetBone {boneName=upperarm_r}}, {targetBone {boneName=upperarm_twist_02_r}}]
    constraintWeights [1,1]
   }
   n15 = Anim Dynamics - Bone: dyn_L_frontRibbon_base_1!(SimulationSpace:AnimPhysSimSpaceType Enum, ComponentPose=@n14, Alpha=1) -> (Pose) {
    angularDampingOverride=1
    angularSpringConstant=1000
    bAngularSpring=true
    bOverrideAngularDamping=true
    bUseGravityOverride=true
    gravityScale=0.35
    simulationSpace=RootRelative
    alphaBoolBlend {blendOption=Linear}
    boundBone {boneName=dyn_L_frontRibbon_base_1}
    componentLinearAccScale=<1 1 1>
    componentLinearVelScale=<1 1 1>
    gravityOverride {z=-980}
    physicsBodyDefinitions [{
      boundBone {boneName=dyn_L_frontRibbon_base_1}
      boxExtents=<16 16 16>
      constraintSetup {angularLimitsMax=<0 10 10>, angularLimitsMin=<0 -10 -10>, angularTarget=<1 0 0>}
      localJointOffset=<0 0 0>
     }]
   }
   n16 = Anim Dynamics - Bone: dyn_L_frontRibbon_string_A_1!(SimulationSpace:AnimPhysSimSpaceType Enum, ComponentPose=@n15, Alpha=1) -> (Pose) {
    angularDampingOverride=1
    angularSpringConstant=100
    bAngularSpring=true
    bOverrideAngularDamping=true
    bUseGravityOverride=true
    gravityScale=0.35
    simulationSpace=RootRelative
    alphaBoolBlend {blendOption=Linear}
    boundBone {boneName=dyn_L_frontRibbon_string_A_1}
    componentAppliedLinearAccClamp=<2500 2500 2500>
    componentLinearAccScale=<1 1 1>
    componentLinearVelScale=<1 1 1>
    gravityOverride {z=-980}
    physicsBodyDefinitions [{
      boundBone {boneName=dyn_L_frontRibbon_string_A_1}
      boxExtents=<16 16 16>
      constraintSetup {angularLimitsMax=<0 70 45>, angularLimitsMin=<0 -20 -45>, angularTarget=<1 0 0>}
      localJointOffset=<-4 0 0>
     }]
   }
   n17 = Anim Dynamics - Bone: dyn_L_frontRibbon_string_A_2!(SimulationSpace:AnimPhysSimSpaceType Enum, ComponentPose=@n16, Alpha=1) -> (Pose) {
    angularDampingOverride=1
    angularSpringConstant=100
    bAngularSpring=true
    bOverrideAngularDamping=true
    bUseGravityOverride=true
    gravityScale=0.35
    simulationSpace=RootRelative
    alphaBoolBlend {blendOption=Linear}
    boundBone {boneName=dyn_L_frontRibbon_string_A_2}
    componentAppliedLinearAccClamp=<2500 2500 2500>
    componentLinearAccScale=<1 1 1>
    componentLinearVelScale=<1 1 1>
    gravityOverride {z=-980}
    physicsBodyDefinitions [{
      boundBone {boneName=dyn_L_frontRibbon_string_A_2}
      boxExtents=<16 16 16>
      constraintSetup {angularLimitsMax=<0 70 45>, angularLimitsMin=<0 -40 -45>, angularTarget=<1 0 0>}
      localJointOffset=<-4 0 0>
     }]
   }
   n18 = Anim Dynamics - Bone: dyn_L_frontRibbon_string_A_3!(SimulationSpace:AnimPhysSimSpaceType Enum, ComponentPose=@n17, Alpha=1) -> (Pose) {
    angularDampingOverride=1
    angularSpringConstant=100
    bAngularSpring=true
    bOverrideAngularDamping=true
    bUseGravityOverride=true
    gravityScale=0.35
    simulationSpace=RootRelative
    alphaBoolBlend {blendOption=Linear}
    boundBone {boneName=dyn_L_frontRibbon_string_A_3}
    componentAppliedLinearAccClamp=<2500 2500 2500>
    componentLinearAccScale=<1 1 1>
    componentLinearVelScale=<1 1 1>
    gravityOverride {z=-980}
    physicsBodyDefinitions [{
      boundBone {boneName=dyn_L_frontRibbon_string_A_3}
      boxExtents=<16 16 16>
      constraintSetup {angularLimitsMax=<0 70 45>, angularLimitsMin=<0 -70 -45>, angularTarget=<1 0 0>}
      localJointOffset=<-4 0 0>
     }]
   }
   n19 = Anim Dynamics - Bone: dyn_L_frontRibbon_A!(SimulationSpace:AnimPhysSimSpaceType Enum, ComponentPose=@n18, Alpha=1) -> (Pose) {
    angularDampingOverride=1
    angularSpringConstant=1000
    bAngularSpring=true
    bOverrideAngularDamping=true
    bUseGravityOverride=true
    gravityScale=0.35
    simulationSpace=RootRelative
    alphaBoolBlend {blendOption=Linear}
    boundBone {boneName=dyn_L_frontRibbon_A}
    componentAppliedLinearAccClamp=<2500 2500 2500>
    componentLinearAccScale=<1 1 1>
    componentLinearVelScale=<1 1 1>
    gravityOverride {z=-980}
    physicsBodyDefinitions [{
      boundBone {boneName=dyn_L_frontRibbon_A}
      boxExtents=<16 16 16>
      constraintSetup {angularLimitsMax=<0 15 15>, angularLimitsMin=<0 -15 -15>, angularTarget=<1 0 0>}
      localJointOffset=<-4 0 0>
     }]
   }
   n20 = Anim Dynamics - Bone: dyn_L_frontRibbon_B!(SimulationSpace:AnimPhysSimSpaceType Enum, ComponentPose=@n19, Alpha=1) -> (Pose) {
    angularDampingOverride=1
    angularSpringConstant=1000
    bAngularSpring=true
    bOverrideAngularDamping=true
    bUseGravityOverride=true
    gravityScale=0.35
    simulationSpace=RootRelative
    alphaBoolBlend {blendOption=Linear}
    boundBone {boneName=dyn_L_frontRibbon_B}
    componentAppliedLinearAccClamp=<2500 2500 2500>
    componentLinearAccScale=<1 1 1>
    componentLinearVelScale=<1 1 1>
    gravityOverride {z=-980}
    physicsBodyDefinitions [{
      boundBone {boneName=dyn_L_frontRibbon_B}
      boxExtents=<16 16 16>
      constraintSetup {angularLimitsMax=<0 15 15>, angularLimitsMin=<0 -15 -15>, angularTarget=<1 0 0>}
      localJointOffset=<-4 0 0>
     }]
   }
   n21 = Anim Dynamics - Bone: dyn_L_frontRibbon_base_2!(SimulationSpace:AnimPhysSimSpaceType Enum, ComponentPose=@n20, Alpha=1) -> (Pose) {
    angularDampingOverride=1
    angularSpringConstant=1000
    bAngularSpring=true
    bOverrideAngularDamping=true
    bUseGravityOverride=true
    gravityScale=0.35
    simulationSpace=RootRelative
    alphaBoolBlend {blendOption=Linear}
    boundBone {boneName=dyn_L_frontRibbon_base_2}
    componentLinearAccScale=<1 1 1>
    componentLinearVelScale=<1 1 1>
    gravityOverride {z=-980}
    physicsBodyDefinitions [{
      boundBone {boneName=dyn_L_frontRibbon_base_2}
      boxExtents=<16 16 16>
      constraintSetup {angularLimitsMax=<0 10 10>, angularLimitsMin=<0 -10 -10>, angularTarget=<1 0 0>}
      localJointOffset=<0 0 0>
     }]
   }
   n22 = Anim Dynamics - Bone: dyn_L_frontRibbon_string_B_1!(SimulationSpace:AnimPhysSimSpaceType Enum, ComponentPose=@n21, Alpha=1) -> (Pose) {
    angularDampingOverride=1
    angularSpringConstant=100
    bAngularSpring=true
    bOverrideAngularDamping=true
    bUseGravityOverride=true
    gravityScale=0.35
    simulationSpace=RootRelative
    alphaBoolBlend {blendOption=Linear}
    boundBone {boneName=dyn_L_frontRibbon_string_B_1}
    componentAppliedLinearAccClamp=<2500 2500 2500>
    componentLinearAccScale=<1 1 1>
    componentLinearVelScale=<1 1 1>
    gravityOverride {z=-980}
    physicsBodyDefinitions [{
      boundBone {boneName=dyn_L_frontRibbon_string_B_1}
      boxExtents=<16 16 16>
      constraintSetup {angularLimitsMax=<0 70 30>, angularLimitsMin=<0 -20 -30>, angularTarget=<1 0 0>}
      localJointOffset=<-4 0 0>
     }]
   }
   n23 = Anim Dynamics - Bone: dyn_L_frontRibbon_string_B_2!(SimulationSpace:AnimPhysSimSpaceType Enum, ComponentPose=@n22, Alpha=1) -> (Pose) {
    angularDampingOverride=1
    angularSpringConstant=100
    bAngularSpring=true
    bOverrideAngularDamping=true
    bUseGravityOverride=true
    gravityScale=0.35
    simulationSpace=RootRelative
    alphaBoolBlend {blendOption=Linear}
    boundBone {boneName=dyn_L_frontRibbon_string_B_2}
    componentAppliedLinearAccClamp=<2500 2500 2500>
    componentLinearAccScale=<1 1 1>
    componentLinearVelScale=<1 1 1>
    gravityOverride {z=-980}
    physicsBodyDefinitions [{
      boundBone {boneName=dyn_L_frontRibbon_string_B_2}
      boxExtents=<16 16 16>
      constraintSetup {angularLimitsMax=<0 70 30>, angularLimitsMin=<0 -30 -30>, angularTarget=<1 0 0>}
      localJointOffset=<-4 0 0>
     }]
   }
   n24 = Anim Dynamics - Bone: dyn_L_backRibbon_base_1!(SimulationSpace:AnimPhysSimSpaceType Enum, ComponentPose=@n23, Alpha=1) -> (Pose) {
    angularDampingOverride=1
    angularSpringConstant=1000
    bAngularSpring=true
    bOverrideAngularDamping=true
    bUseGravityOverride=true
    gravityScale=0.35
    simulationSpace=RootRelative
    alphaBoolBlend {blendOption=Linear}
    boundBone {boneName=dyn_L_backRibbon_base_1}
    componentLinearAccScale=<1 1 1>
    componentLinearVelScale=<1 1 1>
    gravityOverride {z=-980}
    physicsBodyDefinitions [{
      boundBone {boneName=dyn_L_backRibbon_base_1}
      boxExtents=<16 16 16>
      constraintSetup {angularLimitsMax=<0 10 10>, angularLimitsMin=<0 -10 -10>, angularTarget=<1 0 0>}
      localJointOffset=<0 0 0>
     }]
   }
   n25 = Anim Dynamics - Bone: dyn_L_backRibbon_string_A_1!(SimulationSpace:AnimPhysSimSpaceType Enum, ComponentPose=@n24, Alpha=1) -> (Pose) {
    angularDampingOverride=1
    angularSpringConstant=100
    bAngularSpring=true
    bOverrideAngularDamping=true
    bUseGravityOverride=true
    gravityScale=0.35
    simulationSpace=RootRelative
    alphaBoolBlend {blendOption=Linear}
    boundBone {boneName=dyn_L_backRibbon_string_A_1}
    componentAppliedLinearAccClamp=<2500 2500 2500>
    componentLinearAccScale=<1 1 1>
    componentLinearVelScale=<1 1 1>
    gravityOverride {z=-980}
    physicsBodyDefinitions [{
      boundBone {boneName=dyn_L_backRibbon_string_A_1}
      boxExtents=<16 16 16>
      constraintSetup {angularLimitsMax=<0 10 45>, angularLimitsMin=<0 -70 -45>, angularTarget=<1 0 0>}
      localJointOffset=<4 0 0>
     }]
   }
   n26 = Anim Dynamics - Bone: dyn_L_backRibbon_string_A_2!(SimulationSpace:AnimPhysSimSpaceType Enum, ComponentPose=@n25, Alpha=1) -> (Pose) {
    angularDampingOverride=1
    angularSpringConstant=100
    bAngularSpring=true
    bOverrideAngularDamping=true
    bUseGravityOverride=true
    gravityScale=0.35
    simulationSpace=RootRelative
    alphaBoolBlend {blendOption=Linear}
    boundBone {boneName=dyn_L_backRibbon_string_A_2}
    componentAppliedLinearAccClamp=<2500 2500 2500>
    componentLinearAccScale=<1 1 1>
    componentLinearVelScale=<1 1 1>
    gravityOverride {z=-980}
    physicsBodyDefinitions [{
      boundBone {boneName=dyn_L_backRibbon_string_A_2}
      boxExtents=<16 16 16>
      constraintSetup {angularLimitsMax=<0 10 45>, angularLimitsMin=<0 -70 -45>, angularTarget=<1 0 0>}
      localJointOffset=<4 0 0>
     }]
   }
   n27 = Anim Dynamics - Bone: dyn_L_backRibbon_string_A_3!(SimulationSpace:AnimPhysSimSpaceType Enum, ComponentPose=@n26, Alpha=1) -> (Pose) {
    angularDampingOverride=1
    angularSpringConstant=100
    bAngularSpring=true
    bOverrideAngularDamping=true
    bUseGravityOverride=true
    gravityScale=0.35
    simulationSpace=RootRelative
    alphaBoolBlend {blendOption=Linear}
    boundBone {boneName=dyn_L_backRibbon_string_A_3}
    componentAppliedLinearAccClamp=<2500 2500 2500>
    componentLinearAccScale=<1 1 1>
    componentLinearVelScale=<1 1 1>
    gravityOverride {z=-980}
    physicsBodyDefinitions [{
      boundBone {boneName=dyn_L_backRibbon_string_A_3}
      boxExtents=<16 16 16>
      constraintSetup {angularLimitsMax=<0 70 70>, angularLimitsMin=<0 -70 -70>, angularTarget=<1 0 0>}
      localJointOffset=<4 0 0>
     }]
   }
   n28 = Anim Dynamics - Bone: dyn_L_backRibbon_A!(SimulationSpace:AnimPhysSimSpaceType Enum, ComponentPose=@n27, Alpha=1) -> (Pose) {
    angularDampingOverride=1
    angularSpringConstant=1000
    bAngularSpring=true
    bOverrideAngularDamping=true
    bUseGravityOverride=true
    gravityScale=0.35
    simulationSpace=RootRelative
    alphaBoolBlend {blendOption=Linear}
    boundBone {boneName=dyn_L_backRibbon_A}
    componentAppliedLinearAccClamp=<2500 2500 2500>
    componentLinearAccScale=<1 1 1>
    componentLinearVelScale=<1 1 1>
    gravityOverride {z=-980}
    physicsBodyDefinitions [{
      boundBone {boneName=dyn_L_backRibbon_A}
      boxExtents=<16 16 16>
      constraintSetup {angularLimitsMax=<0 15 15>, angularLimitsMin=<0 -15 -15>, angularTarget=<1 0 0>}
      localJointOffset=<4 0 0>
     }]
   }
   n29 = Anim Dynamics - Bone: dyn_L_backRibbon_B!(SimulationSpace:AnimPhysSimSpaceType Enum, ComponentPose=@n28, Alpha=1) -> (Pose) {
    angularDampingOverride=1
    angularSpringConstant=1000
    bAngularSpring=true
    bOverrideAngularDamping=true
    bUseGravityOverride=true
    gravityScale=0.35
    simulationSpace=RootRelative
    alphaBoolBlend {blendOption=Linear}
    boundBone {boneName=dyn_L_backRibbon_B}
    componentAppliedLinearAccClamp=<2500 2500 2500>
    componentLinearAccScale=<1 1 1>
    componentLinearVelScale=<1 1 1>
    gravityOverride {z=-980}
    physicsBodyDefinitions [{
      boundBone {boneName=dyn_L_backRibbon_B}
      boxExtents=<16 16 16>
      constraintSetup {angularLimitsMax=<0 15 15>, angularLimitsMin=<0 -15 -15>, angularTarget=<1 0 0>}
      localJointOffset=<4 0 0>
     }]
   }
   n30 = Anim Dynamics - Bone: dyn_L_backRibbon_base_2!(SimulationSpace:AnimPhysSimSpaceType Enum, ComponentPose=@n29, Alpha=1) -> (Pose) {
    angularDampingOverride=1
    angularSpringConstant=1000
    bAngularSpring=true
    bOverrideAngularDamping=true
    bUseGravityOverride=true
    gravityScale=0.35
    simulationSpace=RootRelative
    alphaBoolBlend {blendOption=Linear}
    boundBone {boneName=dyn_L_backRibbon_base_2}
    componentLinearAccScale=<1 1 1>
    componentLinearVelScale=<1 1 1>
    gravityOverride {z=-980}
    physicsBodyDefinitions [{
      boundBone {boneName=dyn_L_backRibbon_base_2}
      boxExtents=<16 16 16>
      constraintSetup {angularLimitsMax=<0 10 10>, angularLimitsMin=<0 -10 -10>, angularTarget=<1 0 0>}
      localJointOffset=<0 0 0>
     }]
   }
   n31 = Anim Dynamics - Bone: dyn_L_backRibbon_string_B_2!(SimulationSpace:AnimPhysSimSpaceType Enum, ComponentPose=@n30, Alpha=1) -> (Pose) {
    angularDampingOverride=1
    angularSpringConstant=100
    bAngularSpring=true
    bOverrideAngularDamping=true
    bUseGravityOverride=true
    gravityScale=0.35
    simulationSpace=RootRelative
    alphaBoolBlend {blendOption=Linear}
    boundBone {boneName=dyn_L_backRibbon_string_B_2}
    componentAppliedLinearAccClamp=<2500 2500 2500>
    componentLinearAccScale=<1 1 1>
    componentLinearVelScale=<1 1 1>
    gravityOverride {z=-980}
    physicsBodyDefinitions [{
      boundBone {boneName=dyn_L_backRibbon_string_B_2}
      boxExtents=<16 16 16>
      constraintSetup {angularLimitsMax=<0 5 30>, angularLimitsMin=<0 -30 -30>, angularTarget=<1 0 0>}
      localJointOffset=<4 0 0>
     }]
   }
   n32 = Anim Dynamics - Bone: dyn_L_backRibbon_string_B_1!(SimulationSpace:AnimPhysSimSpaceType Enum, ComponentPose=@n31, Alpha=1) -> (Pose) {
    angularDampingOverride=1
    angularSpringConstant=100
    bAngularSpring=true
    bOverrideAngularDamping=true
    bUseGravityOverride=true
    gravityScale=0.35
    simulationSpace=RootRelative
    alphaBoolBlend {blendOption=Linear}
    boundBone {boneName=dyn_L_backRibbon_string_B_1}
    componentAppliedLinearAccClamp=<2500 2500 2500>
    componentLinearAccScale=<1 1 1>
    componentLinearVelScale=<1 1 1>
    gravityOverride {z=-980}
    physicsBodyDefinitions [{
      boundBone {boneName=dyn_L_backRibbon_string_B_1}
      boxExtents=<16 16 16>
      constraintSetup {angularLimitsMax=<0 5 30>, angularLimitsMin=<0 -70 -30>, angularTarget=<1 0 0>}
      localJointOffset=<4 0 0>
     }]
   }
   n33 = Anim Dynamics - Bone: dyn_R_frontRibbon_base_1!(SimulationSpace:AnimPhysSimSpaceType Enum, ComponentPose=@n32, Alpha=1) -> (Pose) {
    angularDampingOverride=1
    angularSpringConstant=1000
    bAngularSpring=true
    bOverrideAngularDamping=true
    bUseGravityOverride=true
    gravityScale=0.35
    simulationSpace=RootRelative
    alphaBoolBlend {blendOption=Linear}
    boundBone {boneName=dyn_R_frontRibbon_base_1}
    componentLinearAccScale=<1 1 1>
    componentLinearVelScale=<1 1 1>
    gravityOverride {z=-980}
    physicsBodyDefinitions [{
      boundBone {boneName=dyn_R_frontRibbon_base_1}
      boxExtents=<16 16 16>
      constraintSetup {angularLimitsMax=<0 10 10>, angularLimitsMin=<0 -10 -10>, angularTarget=<1 0 0>}
      localJointOffset=<0 0 0>
     }]
   }
   n34 = Anim Dynamics - Bone: dyn_R_frontRibbon_string_A_1!(SimulationSpace:AnimPhysSimSpaceType Enum, ComponentPose=@n33, Alpha=1) -> (Pose) {
    angularDampingOverride=1
    angularSpringConstant=100
    bAngularSpring=true
    bOverrideAngularDamping=true
    bUseGravityOverride=true
    gravityScale=0.35
    simulationSpace=RootRelative
    alphaBoolBlend {blendOption=Linear}
    boundBone {boneName=dyn_R_frontRibbon_string_A_1}
    componentAppliedLinearAccClamp=<2500 2500 2500>
    componentLinearAccScale=<1 1 1>
    componentLinearVelScale=<1 1 1>
    gravityOverride {z=-980}
    physicsBodyDefinitions [{
      boundBone {boneName=dyn_R_frontRibbon_string_A_1}
      boxExtents=<16 16 16>
      constraintSetup {angularLimitsMax=<0 70 45>, angularLimitsMin=<0 -20 -45>, angularTarget=<1 0 0>}
      localJointOffset=<4 0 0>
     }]
   }
   n35 = Anim Dynamics - Bone: dyn_R_frontRibbon_string_A_2!(SimulationSpace:AnimPhysSimSpaceType Enum, ComponentPose=@n34, Alpha=1) -> (Pose) {
    angularDampingOverride=1
    angularSpringConstant=100
    bAngularSpring=true
    bOverrideAngularDamping=true
    bUseGravityOverride=true
    gravityScale=0.35
    simulationSpace=RootRelative
    alphaBoolBlend {blendOption=Linear}
    boundBone {boneName=dyn_R_frontRibbon_string_A_2}
    componentAppliedLinearAccClamp=<2500 2500 2500>
    componentLinearAccScale=<1 1 1>
    componentLinearVelScale=<1 1 1>
    gravityOverride {z=-980}
    physicsBodyDefinitions [{
      boundBone {boneName=dyn_R_frontRibbon_string_A_2}
      boxExtents=<16 16 16>
      constraintSetup {angularLimitsMax=<0 70 45>, angularLimitsMin=<0 -40 -45>, angularTarget=<1 0 0>}
      localJointOffset=<4 0 0>
     }]
   }
   n36 = Anim Dynamics - Bone: dyn_R_frontRibbon_string_A_3!(SimulationSpace:AnimPhysSimSpaceType Enum, ComponentPose=@n35, Alpha=1) -> (Pose) {
    angularDampingOverride=1
    angularSpringConstant=100
    bAngularSpring=true
    bOverrideAngularDamping=true
    bUseGravityOverride=true
    gravityScale=0.35
    simulationSpace=RootRelative
    alphaBoolBlend {blendOption=Linear}
    boundBone {boneName=dyn_R_frontRibbon_string_A_3}
    componentAppliedLinearAccClamp=<2500 2500 2500>
    componentLinearAccScale=<1 1 1>
    componentLinearVelScale=<1 1 1>
    gravityOverride {z=-980}
    physicsBodyDefinitions [{
      boundBone {boneName=dyn_R_frontRibbon_string_A_3}
      boxExtents=<16 16 16>
      constraintSetup {angularLimitsMax=<0 70 45>, angularLimitsMin=<0 -70 -45>, angularTarget=<1 0 0>}
      localJointOffset=<4 0 0>
     }]
   }
   n37 = Anim Dynamics - Bone: dyn_R_frontRibbon_A!(SimulationSpace:AnimPhysSimSpaceType Enum, ComponentPose=@n36, Alpha=1) -> (Pose) {
    angularDampingOverride=1
    angularSpringConstant=1000
    bAngularSpring=true
    bOverrideAngularDamping=true
    bUseGravityOverride=true
    gravityScale=0.35
    simulationSpace=RootRelative
    alphaBoolBlend {blendOption=Linear}
    boundBone {boneName=dyn_R_frontRibbon_A}
    componentAppliedLinearAccClamp=<2500 2500 2500>
    componentLinearAccScale=<1 1 1>
    componentLinearVelScale=<1 1 1>
    gravityOverride {z=-980}
    physicsBodyDefinitions [{
      boundBone {boneName=dyn_R_frontRibbon_A}
      boxExtents=<16 16 16>
      constraintSetup {angularLimitsMax=<0 15 15>, angularLimitsMin=<0 -15 -15>, angularTarget=<1 0 0>}
      localJointOffset=<4 0 0>
     }]
   }
   n38 = Anim Dynamics - Bone: dyn_R_frontRibbon_B!(SimulationSpace:AnimPhysSimSpaceType Enum, ComponentPose=@n37, Alpha=1) -> (Pose) {
    angularDampingOverride=1
    angularSpringConstant=1000
    bAngularSpring=true
    bOverrideAngularDamping=true
    bUseGravityOverride=true
    gravityScale=0.35
    simulationSpace=RootRelative
    alphaBoolBlend {blendOption=Linear}
    boundBone {boneName=dyn_R_frontRibbon_B}
    componentAppliedLinearAccClamp=<2500 2500 2500>
    componentLinearAccScale=<1 1 1>
    componentLinearVelScale=<1 1 1>
    gravityOverride {z=-980}
    physicsBodyDefinitions [{
      boundBone {boneName=dyn_R_frontRibbon_B}
      boxExtents=<16 16 16>
      constraintSetup {angularLimitsMax=<0 15 15>, angularLimitsMin=<0 -15 -15>, angularTarget=<1 0 0>}
      localJointOffset=<4 0 0>
     }]
   }
   n39 = Anim Dynamics - Bone: dyn_R_frontRibbon_base_2!(SimulationSpace:AnimPhysSimSpaceType Enum, ComponentPose=@n38, Alpha=1) -> (Pose) {
    angularDampingOverride=1
    angularSpringConstant=1000
    bAngularSpring=true
    bOverrideAngularDamping=true
    bUseGravityOverride=true
    gravityScale=0.35
    simulationSpace=RootRelative
    alphaBoolBlend {blendOption=Linear}
    boundBone {boneName=dyn_R_frontRibbon_base_2}
    componentLinearAccScale=<1 1 1>
    componentLinearVelScale=<1 1 1>
    gravityOverride {z=-980}
    physicsBodyDefinitions [{
      boundBone {boneName=dyn_R_frontRibbon_base_2}
      boxExtents=<16 16 16>
      constraintSetup {angularLimitsMax=<0 20 10>, angularLimitsMin=<0 -20 -10>, angularTarget=<1 0 0>}
      localJointOffset=<0 0 0>
     }]
   }
   n40 = Anim Dynamics - Bone: dyn_R_frontRibbon_string_B_1!(SimulationSpace:AnimPhysSimSpaceType Enum, ComponentPose=@n39, Alpha=1) -> (Pose) {
    angularDampingOverride=1
    angularSpringConstant=100
    bAngularSpring=true
    bOverrideAngularDamping=true
    bUseGravityOverride=true
    gravityScale=0.35
    simulationSpace=RootRelative
    alphaBoolBlend {blendOption=Linear}
    boundBone {boneName=dyn_R_frontRibbon_string_B_1}
    componentAppliedLinearAccClamp=<2500 2500 2500>
    componentLinearAccScale=<1 1 1>
    componentLinearVelScale=<1 1 1>
    gravityOverride {z=-980}
    physicsBodyDefinitions [{
      boundBone {boneName=dyn_R_frontRibbon_string_B_1}
      boxExtents=<16 16 16>
      constraintSetup {angularLimitsMax=<0 70 30>, angularLimitsMin=<0 -20 -30>, angularTarget=<1 0 0>}
      localJointOffset=<4 0 0>
     }]
   }
   n41 = Anim Dynamics - Bone: dyn_R_frontRibbon_string_B_2!(SimulationSpace:AnimPhysSimSpaceType Enum, ComponentPose=@n40, Alpha=1) -> (Pose) {
    angularDampingOverride=1
    angularSpringConstant=100
    bAngularSpring=true
    bOverrideAngularDamping=true
    bUseGravityOverride=true
    gravityScale=0.35
    simulationSpace=RootRelative
    alphaBoolBlend {blendOption=Linear}
    boundBone {boneName=dyn_R_frontRibbon_string_B_2}
    componentAppliedLinearAccClamp=<2500 2500 2500>
    componentLinearAccScale=<1 1 1>
    componentLinearVelScale=<1 1 1>
    gravityOverride {z=-980}
    physicsBodyDefinitions [{
      boundBone {boneName=dyn_R_frontRibbon_string_B_2}
      boxExtents=<16 16 16>
      constraintSetup {angularLimitsMax=<0 70 30>, angularLimitsMin=<0 -30 -30>, angularTarget=<1 0 0>}
      localJointOffset=<4 0 0>
     }]
   }
   n42 = Anim Dynamics - Bone: dyn_R_backRibbon_base_1!(SimulationSpace:AnimPhysSimSpaceType Enum, ComponentPose=@n41, Alpha=1) -> (Pose) {
    angularDampingOverride=1
    angularSpringConstant=1000
    bAngularSpring=true
    bOverrideAngularDamping=true
    bUseGravityOverride=true
    gravityScale=0.35
    simulationSpace=RootRelative
    alphaBoolBlend {blendOption=Linear}
    boundBone {boneName=dyn_R_backRibbon_base_1}
    componentLinearAccScale=<1 1 1>
    componentLinearVelScale=<1 1 1>
    gravityOverride {z=-980}
    physicsBodyDefinitions [{
      boundBone {boneName=dyn_R_backRibbon_base_1}
      boxExtents=<16 16 16>
      constraintSetup {angularLimitsMax=<0 10 10>, angularLimitsMin=<0 -10 -10>, angularTarget=<1 0 0>}
      localJointOffset=<0 0 0>
     }]
   }
   n43 = Anim Dynamics - Bone: dyn_R_backRibbon_string_A_1!(SimulationSpace:AnimPhysSimSpaceType Enum, ComponentPose=@n42, Alpha=1) -> (Pose) {
    angularDampingOverride=1
    angularSpringConstant=100
    bAngularSpring=true
    bOverrideAngularDamping=true
    bUseGravityOverride=true
    gravityScale=0.35
    simulationSpace=RootRelative
    alphaBoolBlend {blendOption=Linear}
    boundBone {boneName=dyn_R_backRibbon_string_A_1}
    componentAppliedLinearAccClamp=<2500 2500 2500>
    componentLinearAccScale=<1 1 1>
    componentLinearVelScale=<1 1 1>
    gravityOverride {z=-980}
    physicsBodyDefinitions [{
      boundBone {boneName=dyn_R_backRibbon_string_A_1}
      boxExtents=<19 19 19>
      constraintSetup {angularLimitsMax=<0 10 45>, angularLimitsMin=<0 -70 -45>, angularTarget=<1 0 0>}
      localJointOffset=<-4 0 0>
     }]
   }
   n44 = Anim Dynamics - Bone: dyn_R_backRibbon_string_A_2!(SimulationSpace:AnimPhysSimSpaceType Enum, ComponentPose=@n43, Alpha=1) -> (Pose) {
    angularDampingOverride=1
    angularSpringConstant=100
    bAngularSpring=true
    bOverrideAngularDamping=true
    bUseGravityOverride=true
    gravityScale=0.35
    simulationSpace=RootRelative
    alphaBoolBlend {blendOption=Linear}
    boundBone {boneName=dyn_R_backRibbon_string_A_2}
    componentAppliedLinearAccClamp=<2500 2500 2500>
    componentLinearAccScale=<1 1 1>
    componentLinearVelScale=<1 1 1>
    gravityOverride {z=-980}
    physicsBodyDefinitions [{
      boundBone {boneName=dyn_R_backRibbon_string_A_2}
      boxExtents=<19 19 19>
      constraintSetup {angularLimitsMax=<0 10 45>, angularLimitsMin=<0 -70 -45>, angularTarget=<1 0 0>}
      localJointOffset=<-4 0 0>
     }]
   }
   n45 = Anim Dynamics - Bone: dyn_R_backRibbon_string_A_3!(SimulationSpace:AnimPhysSimSpaceType Enum, ComponentPose=@n44, Alpha=1) -> (Pose) {
    angularDampingOverride=1
    angularSpringConstant=100
    bAngularSpring=true
    bOverrideAngularDamping=true
    bUseGravityOverride=true
    gravityScale=0.35
    simulationSpace=RootRelative
    alphaBoolBlend {blendOption=Linear}
    boundBone {boneName=dyn_R_backRibbon_string_A_3}
    componentAppliedLinearAccClamp=<2500 2500 2500>
    componentLinearAccScale=<1 1 1>
    componentLinearVelScale=<1 1 1>
    gravityOverride {z=-980}
    physicsBodyDefinitions [{
      boundBone {boneName=dyn_R_backRibbon_string_A_3}
      boxExtents=<19 19 19>
      constraintSetup {angularLimitsMax=<0 70 70>, angularLimitsMin=<0 -70 -70>, angularTarget=<1 0 0>}
      localJointOffset=<-4 0 0>
     }]
   }
   n46 = Anim Dynamics - Bone: dyn_R_backRibbon_A!(SimulationSpace:AnimPhysSimSpaceType Enum, ComponentPose=@n45, Alpha=1) -> (Pose) {
    angularDampingOverride=1
    angularSpringConstant=1000
    bAngularSpring=true
    bOverrideAngularDamping=true
    bUseGravityOverride=true
    gravityScale=0.35
    simulationSpace=RootRelative
    alphaBoolBlend {blendOption=Linear}
    boundBone {boneName=dyn_R_backRibbon_A}
    componentAppliedLinearAccClamp=<2500 2500 2500>
    componentLinearAccScale=<1 1 1>
    componentLinearVelScale=<1 1 1>
    gravityOverride {z=-980}
    physicsBodyDefinitions [{
      boundBone {boneName=dyn_R_backRibbon_A}
      boxExtents=<16 16 16>
      constraintSetup {angularLimitsMax=<0 15 15>, angularLimitsMin=<0 -15 -15>, angularTarget=<1 0 0>}
      localJointOffset=<-4 0 0>
     }]
   }
   n47 = Anim Dynamics - Bone: dyn_R_backRibbon_B!(SimulationSpace:AnimPhysSimSpaceType Enum, ComponentPose=@n46, Alpha=1) -> (Pose) {
    angularDampingOverride=1
    angularSpringConstant=1000
    bAngularSpring=true
    bOverrideAngularDamping=true
    bUseGravityOverride=true
    gravityScale=0.35
    simulationSpace=RootRelative
    alphaBoolBlend {blendOption=Linear}
    boundBone {boneName=dyn_R_backRibbon_B}
    componentAppliedLinearAccClamp=<2500 2500 2500>
    componentLinearAccScale=<1 1 1>
    componentLinearVelScale=<1 1 1>
    gravityOverride {z=-980}
    physicsBodyDefinitions [{
      boundBone {boneName=dyn_R_backRibbon_B}
      boxExtents=<16 16 16>
      constraintSetup {angularLimitsMax=<0 15 15>, angularLimitsMin=<0 -15 -15>, angularTarget=<1 0 0>}
      localJointOffset=<-4 0 0>
     }]
   }
   n48 = Anim Dynamics - Bone: dyn_R_backRibbon_base_2!(SimulationSpace:AnimPhysSimSpaceType Enum, ComponentPose=@n47, Alpha=1) -> (Pose) {
    angularDampingOverride=1
    angularSpringConstant=1000
    bAngularSpring=true
    bOverrideAngularDamping=true
    bUseGravityOverride=true
    gravityScale=0.35
    simulationSpace=RootRelative
    alphaBoolBlend {blendOption=Linear}
    boundBone {boneName=dyn_R_backRibbon_base_2}
    componentLinearAccScale=<1 1 1>
    componentLinearVelScale=<1 1 1>
    gravityOverride {z=-980}
    physicsBodyDefinitions [{
      boundBone {boneName=dyn_R_backRibbon_base_2}
      boxExtents=<16 16 16>
      constraintSetup {angularLimitsMax=<0 10 10>, angularLimitsMin=<0 -10 -10>, angularTarget=<1 0 0>}
      localJointOffset=<0 0 0>
     }]
   }
   n49 = Anim Dynamics - Bone: dyn_R_backRibbon_string_B_1!(SimulationSpace:AnimPhysSimSpaceType Enum, ComponentPose=@n48, Alpha=1) -> (Pose) {
    angularDampingOverride=1
    angularSpringConstant=100
    bAngularSpring=true
    bOverrideAngularDamping=true
    bUseGravityOverride=true
    gravityScale=0.35
    simulationSpace=RootRelative
    alphaBoolBlend {blendOption=Linear}
    boundBone {boneName=dyn_R_backRibbon_string_B_1}
    componentAppliedLinearAccClamp=<2500 2500 2500>
    componentLinearAccScale=<1 1 1>
    componentLinearVelScale=<1 1 1>
    gravityOverride {z=-980}
    physicsBodyDefinitions [{
      boundBone {boneName=dyn_R_backRibbon_string_B_1}
      boxExtents=<16 16 16>
      constraintSetup {angularLimitsMax=<0 5 30>, angularLimitsMin=<0 -30 -30>, angularTarget=<1 0 0>}
      localJointOffset=<-4 0 0>
     }]
   }
   n50 = Anim Dynamics - Bone: dyn_R_backRibbon_string_B_2!(SimulationSpace:AnimPhysSimSpaceType Enum, ComponentPose=@n49, Alpha=1) -> (Pose) {
    angularDampingOverride=1
    angularSpringConstant=100
    bAngularSpring=true
    bOverrideAngularDamping=true
    bUseGravityOverride=true
    gravityScale=0.35
    simulationSpace=RootRelative
    alphaBoolBlend {blendOption=Linear}
    boundBone {boneName=dyn_R_backRibbon_string_B_2}
    componentAppliedLinearAccClamp=<2500 2500 2500>
    componentLinearAccScale=<1 1 1>
    componentLinearVelScale=<1 1 1>
    gravityOverride {z=-980}
    physicsBodyDefinitions [{
      boundBone {boneName=dyn_R_backRibbon_string_B_2}
      boxExtents=<16 16 16>
      constraintSetup {angularLimitsMax=<0 5 30>, angularLimitsMin=<0 -70 -30>, angularTarget=<1 0 0>}
      localJointOffset=<-4 0 0>
     }]
   }
   n51 = Anim Dynamics - Bone: dyn_L_hipChain_1!(ComponentPose=@n50, Alpha=1) -> (Pose) {
    angularDampingOverride=1
    angularSpringConstant=4000
    bAngularSpring=true
    bOverrideAngularDamping=true
    bUseGravityOverride=true
    gravityScale=0.35
    simulationSpace=RootRelative
    alphaBoolBlend {blendOption=Linear}
    alphaScaleBiasClamp {bInterpResult=true, interpSpeedDecreasing=5, interpSpeedIncreasing=5}
    boundBone {boneName=dyn_L_hipChain_1}
    componentAppliedLinearAccClamp=<2500 2500 2500>
    componentLinearAccScale=<1 1 1>
    componentLinearVelScale=<1 1 1>
    gravityOverride {z=-980}
    physicsBodyDefinitions [{
      boundBone {boneName=dyn_L_hipChain_1}
      boxExtents=<18 18 18>
      constraintSetup {angularLimitsMax=<5 0 10>, angularLimitsMin=<-5 -15 -10>, angularTarget=<1 0 0>}
      localJointOffset=<4 0 0>
     }]
   }
   n52 = Anim Dynamics - Bone: dyn_L_hipChain_2!(ComponentPose=@n51, Alpha=1) -> (Pose) {
    angularDampingOverride=1
    angularSpringConstant=4000
    bAngularSpring=true
    bOverrideAngularDamping=true
    bUseGravityOverride=true
    gravityScale=0.35
    simulationSpace=RootRelative
    alphaBoolBlend {blendOption=Linear}
    alphaScaleBiasClamp {bInterpResult=true, interpSpeedDecreasing=5, interpSpeedIncreasing=5}
    boundBone {boneName=dyn_L_hipChain_2}
    componentAppliedLinearAccClamp=<2500 2500 2500>
    componentLinearAccScale=<1 1 1>
    componentLinearVelScale=<1 1 1>
    gravityOverride {z=-980}
    physicsBodyDefinitions [{
      boundBone {boneName=dyn_L_hipChain_2}
      boxExtents=<18 18 18>
      constraintSetup {angularLimitsMax=<5 0 10>, angularLimitsMin=<-5 -15 -10>, angularTarget=<1 0 0>}
      localJointOffset=<4 0 0>
     }]
   }
   n53 = Anim Dynamics - Bone: dyn_L_hipChain_3!(ComponentPose=@n52, Alpha=1) -> (Pose) {
    angularDampingOverride=1
    angularSpringConstant=4000
    bAngularSpring=true
    bOverrideAngularDamping=true
    bUseGravityOverride=true
    gravityScale=0.35
    simulationSpace=RootRelative
    alphaBoolBlend {blendOption=Linear}
    alphaScaleBiasClamp {bInterpResult=true, interpSpeedDecreasing=5, interpSpeedIncreasing=5}
    boundBone {boneName=dyn_L_hipChain_3}
    componentAppliedLinearAccClamp=<2500 2500 2500>
    componentLinearAccScale=<1 1 1>
    componentLinearVelScale=<1 1 1>
    gravityOverride {z=-980}
    physicsBodyDefinitions [{
      boundBone {boneName=dyn_L_hipChain_3}
      boxExtents=<18 18 18>
      constraintSetup {angularLimitsMax=<5 0 10>, angularLimitsMin=<-5 -15 -10>, angularTarget=<1 0 0>}
      localJointOffset=<4 0 0>
     }]
   }
   n54 = Anim Dynamics - Bone: dyn_R_hipChain_1!(ComponentPose=@n53, Alpha=1) -> (Pose) {
    angularDampingOverride=1
    angularSpringConstant=4000
    bAngularSpring=true
    bOverrideAngularDamping=true
    bUseGravityOverride=true
    gravityScale=0.35
    simulationSpace=RootRelative
    alphaBoolBlend {blendOption=Linear}
    alphaScaleBiasClamp {bInterpResult=true, interpSpeedDecreasing=5, interpSpeedIncreasing=5}
    boundBone {boneName=dyn_R_hipChain_1}
    componentAppliedLinearAccClamp=<2500 2500 2500>
    componentLinearAccScale=<1 1 1>
    componentLinearVelScale=<1 1 1>
    gravityOverride {z=-980}
    physicsBodyDefinitions [{
      boundBone {boneName=dyn_R_hipChain_1}
      boxExtents=<18 18 18>
      constraintSetup {angularLimitsMax=<0 0 5>, angularLimitsMin=<0 -15 -5>, angularTarget=<1 0 0>}
      localJointOffset=<4 0 0>
     }]
   }
   n55 = Anim Dynamics - Bone: dyn_R_hipChain_2!(ComponentPose=@n54, Alpha=1) -> (Pose) {
    angularDampingOverride=1
    angularSpringConstant=4000
    bAngularSpring=true
    bOverrideAngularDamping=true
    bUseGravityOverride=true
    gravityScale=0.35
    simulationSpace=RootRelative
    alphaBoolBlend {blendOption=Linear}
    alphaScaleBiasClamp {bInterpResult=true, interpSpeedDecreasing=5, interpSpeedIncreasing=5}
    boundBone {boneName=dyn_R_hipChain_2}
    componentAppliedLinearAccClamp=<2500 2500 2500>
    componentLinearAccScale=<1 1 1>
    componentLinearVelScale=<1 1 1>
    gravityOverride {z=-980}
    physicsBodyDefinitions [{
      boundBone {boneName=dyn_R_hipChain_2}
      boxExtents=<18 18 18>
      constraintSetup {angularLimitsMax=<0 0 5>, angularLimitsMin=<0 -15 -5>, angularTarget=<1 0 0>}
      localJointOffset=<4 0 0>
     }]
   }
   n56 = Anim Dynamics - Bone: dyn_swords_1!(ComponentPose=@n55, Alpha=1) -> (Pose) {
    angularDampingOverride=1
    angularSpringConstant=10000
    bAngularSpring=true
    bOverrideAngularDamping=true
    bUseGravityOverride=true
    gravityScale=0.35
    simulationSpace=RootRelative
    alphaBoolBlend {blendOption=Linear}
    alphaScaleBiasClamp {bInterpResult=true, interpSpeedDecreasing=5, interpSpeedIncreasing=5}
    boundBone {boneName=dyn_swords_1}
    componentAppliedLinearAccClamp=<5000 5000 5000>
    componentLinearAccScale=<1 1 1>
    componentLinearVelScale=<1 1 1>
    gravityOverride {z=-980}
    physicsBodyDefinitions [{
      boundBone {boneName=dyn_swords_1}
      boxExtents=<25 25 25>
      constraintSetup {angularLimitsMax=<0 45 15>, angularLimitsMin=<-15 0 -15>, angularTarget=<1 0 0>}
      localJointOffset=<4 0 0>
     }]
   }
   n57 = Anim Dynamics - Bone: dyn_swords_2!(ComponentPose=@n56, Alpha=1) -> (Pose) {
    angularDampingOverride=1
    angularSpringConstant=20000
    bAngularSpring=true
    bOverrideAngularDamping=true
    bUseGravityOverride=true
    gravityScale=0.35
    simulationSpace=RootRelative
    alphaBoolBlend {blendOption=Linear}
    alphaScaleBiasClamp {bInterpResult=true, interpSpeedDecreasing=5, interpSpeedIncreasing=5}
    boundBone {boneName=dyn_swords_2}
    componentAppliedLinearAccClamp=<5000 5000 5000>
    componentLinearAccScale=<1 1 1>
    componentLinearVelScale=<1 1 1>
    gravityOverride {z=-980}
    physicsBodyDefinitions [{
      boundBone {boneName=dyn_swords_2}
      boxExtents=<20 20 20>
      constraintSetup {angularLimitsMax=<0 45 15>, angularLimitsMin=<-15 0 -15>, angularTarget=<1 0 0>}
      localJointOffset=<4 0 0>
     }]
   }
   n58 = Anim Dynamics - Bone: dyn_swords_3_A!(ComponentPose=@n57, Alpha=1) -> (Pose) {
    angularDampingOverride=1
    angularSpringConstant=4000
    bAngularSpring=true
    bOverrideAngularDamping=true
    bUseGravityOverride=true
    gravityScale=0.35
    simulationSpace=RootRelative
    alphaBoolBlend {blendOption=Linear}
    alphaScaleBiasClamp {bInterpResult=true, interpSpeedDecreasing=5, interpSpeedIncreasing=5}
    boundBone {boneName=dyn_swords_3_A}
    componentAppliedLinearAccClamp=<5000 5000 5000>
    componentLinearAccScale=<1 1 1>
    componentLinearVelScale=<1 1 1>
    gravityOverride {z=-980}
    physicsBodyDefinitions [{
      boundBone {boneName=dyn_swords_3_A}
      boxExtents=<23 23 23>
      constraintSetup {angularLimitsMax=<0 45 15>, angularLimitsMin=<-15 0 -15>, angularTarget=<1 0 0>}
      localJointOffset=<4 0 0>
     }]
   }
   n59 = Anim Dynamics - Bone: dyn_swords_3_B!(ComponentPose=@n58, Alpha=1) -> (Pose) {
    angularDampingOverride=1
    angularSpringConstant=4000
    bAngularSpring=true
    bOverrideAngularDamping=true
    bUseGravityOverride=true
    gravityScale=0.35
    simulationSpace=RootRelative
    alphaBoolBlend {blendOption=Linear}
    alphaScaleBiasClamp {bInterpResult=true, interpSpeedDecreasing=5, interpSpeedIncreasing=5}
    boundBone {boneName=dyn_swords_3_B}
    componentAppliedLinearAccClamp=<5000 5000 5000>
    componentLinearAccScale=<1 1 1>
    componentLinearVelScale=<1 1 1>
    gravityOverride {z=-980}
    physicsBodyDefinitions [{
      boundBone {boneName=dyn_swords_3_B}
      boxExtents=<23 23 23>
      constraintSetup {angularLimitsMax=<0 45 15>, angularLimitsMin=<-15 0 -15>, angularTarget=<1 0 0>}
      localJointOffset=<4 0 0>
     }]
   }
   Save cached pose 'setupCache'(Pose=Component To Local(ComponentPose=@n59)) {cacheName=setupCache}
   n60 = Anim Dynamics - Bone: dyn_L_pantLeg_front_1!(SimulationSpace:AnimPhysSimSpaceType Enum, ComponentPose=Local To Component(LocalPose=Use cached pose 'setupCache'()), Alpha=1) -> (Pose) {
    angularDampingOverride=1
    angularSpringConstant=1000
    bAngularSpring=true
    bOverrideAngularDamping=true
    bUseGravityOverride=true
    gravityScale=0.35
    simulationSpace=RootRelative
    alphaBoolBlend {blendOption=Linear}
    alphaScaleBiasClamp {scale=0.75}
    boundBone {boneName=dyn_L_pantLeg_front_1}
    componentAppliedLinearAccClamp=<2500 2500 2500>
    componentLinearAccScale=<1 1 1>
    componentLinearVelScale=<1 1 1>
    gravityOverride {z=-980}
    physicsBodyDefinitions [{
      boundBone {boneName=dyn_L_pantLeg_front_1}
      boxExtents=<25 25 25>
      constraintSetup {angularLimitsMax=<5 30 15>, angularLimitsMin=<-5 0 -5>, angularTarget=<1 0 0>}
      localJointOffset=<-4 0 0>
     }]
   }
   n61 = Anim Dynamics - Bone: dyn_L_pantLeg_front_2!(SimulationSpace:AnimPhysSimSpaceType Enum, ComponentPose=@n60, Alpha=1) -> (Pose) {
    angularDampingOverride=1
    angularSpringConstant=6000
    bAngularSpring=true
    bOverrideAngularDamping=true
    bUseGravityOverride=true
    gravityScale=0.35
    simulationSpace=RootRelative
    alphaBoolBlend {blendOption=Linear}
    alphaScaleBiasClamp {scale=0.75}
    boundBone {boneName=dyn_L_pantLeg_front_2}
    componentAppliedLinearAccClamp=<2500 2500 2500>
    componentLinearAccScale=<1 1 1>
    componentLinearVelScale=<1 1 1>
    gravityOverride {z=-980}
    physicsBodyDefinitions [{
      boundBone {boneName=dyn_L_pantLeg_front_2}
      boxExtents=<25 25 25>
      constraintSetup {angularLimitsMax=<5 30 60>, angularLimitsMin=<-5 0 -30>, angularTarget=<1 0 0>}
      localJointOffset=<-4 0 0>
     }]
   }
   n62 = Anim Dynamics - Bone: dyn_L_pantLeg_inner_1!(SimulationSpace:AnimPhysSimSpaceType Enum, ComponentPose=@n61, Alpha=1) -> (Pose) {
    angularDampingOverride=1
    angularSpringConstant=5000
    bAngularSpring=true
    bOverrideAngularDamping=true
    bUseGravityOverride=true
    gravityScale=0.35
    simulationSpace=RootRelative
    alphaBoolBlend {blendOption=Linear}
    boundBone {boneName=dyn_L_pantLeg_inner_1}
    componentAppliedLinearAccClamp=<2500 2500 2500>
    componentLinearAccScale=<1 1 1>
    componentLinearVelScale=<1 1 1>
    gravityOverride {z=-980}
    physicsBodyDefinitions [{
      boundBone {boneName=dyn_L_pantLeg_inner_1}
      boxExtents=<25 25 25>
      constraintSetup {angularLimitsMax=<5 10 10>, angularLimitsMin=<-5 -10 -35>, angularTarget=<1 0 0>}
      localJointOffset=<-4 0 0>
     }]
   }
   n63 = Anim Dynamics - Bone: dyn_L_pantLeg_inner_2!(SimulationSpace:AnimPhysSimSpaceType Enum, ComponentPose=@n62, Alpha=1) -> (Pose) {
    angularDampingOverride=1
    angularSpringConstant=10000
    bAngularSpring=true
    bOverrideAngularDamping=true
    bUseGravityOverride=true
    gravityScale=0.35
    simulationSpace=RootRelative
    alphaBoolBlend {blendOption=Linear}
    boundBone {boneName=dyn_L_pantLeg_inner_2}
    componentAppliedLinearAccClamp=<2500 2500 2500>
    componentLinearAccScale=<1 1 1>
    componentLinearVelScale=<1 1 1>
    gravityOverride {z=-980}
    physicsBodyDefinitions [{
      boundBone {boneName=dyn_L_pantLeg_inner_2}
      boxExtents=<25 25 25>
      constraintSetup {angularLimitsMax=<5 10 10>, angularLimitsMin=<-5 -10 -35>, angularTarget=<1 0 0>}
      localJointOffset=<-4 0 0>
     }]
   }
   n64 = Anim Dynamics - Bone: dyn_L_pantLeg_outer_1!(SimulationSpace:AnimPhysSimSpaceType Enum, ComponentPose=@n63, Alpha=1) -> (Pose) {
    angularDampingOverride=1
    angularSpringConstant=10000
    bAngularSpring=true
    bOverrideAngularDamping=true
    bUseGravityOverride=true
    gravityScale=0.35
    simulationSpace=RootRelative
    alphaBoolBlend {blendOption=Linear}
    boundBone {boneName=dyn_L_pantLeg_outer_1}
    componentAppliedLinearAccClamp=<2500 2500 2500>
    componentLinearAccScale=<1 1 1>
    componentLinearVelScale=<1 1 1>
    gravityOverride {z=-980}
    physicsBodyDefinitions [{
      boundBone {boneName=dyn_L_pantLeg_outer_1}
      boxExtents=<25 25 25>
      constraintSetup {angularLimitsMax=<5 30 10>, angularLimitsMin=<-5 -30 0>, angularTarget=<1 0 0>}
      localJointOffset=<-4 0 0>
     }]
   }
   n65 = Anim Dynamics - Bone: dyn_L_pantLeg_outer_2!(SimulationSpace:AnimPhysSimSpaceType Enum, ComponentPose=@n64, Alpha=1) -> (Pose) {
    angularDampingOverride=1
    angularSpringConstant=10000
    bAngularSpring=true
    bOverrideAngularDamping=true
    bUseGravityOverride=true
    gravityScale=0.35
    simulationSpace=RootRelative
    alphaBoolBlend {blendOption=Linear}
    boundBone {boneName=dyn_L_pantLeg_outer_2}
    componentAppliedLinearAccClamp=<2500 2500 2500>
    componentLinearAccScale=<1 1 1>
    componentLinearVelScale=<1 1 1>
    gravityOverride {z=-980}
    physicsBodyDefinitions [{
      boundBone {boneName=dyn_L_pantLeg_outer_2}
      boxExtents=<25 25 25>
      constraintSetup {angularLimitsMax=<5 30 10>, angularLimitsMin=<-5 -30 0>, angularTarget=<1 0 0>}
      localJointOffset=<-4 0 0>
     }]
   }
   n66 = Anim Dynamics - Bone: dyn_L_pantLeg_back_1!(SimulationSpace:AnimPhysSimSpaceType Enum, ComponentPose=@n65, Alpha=1) -> (Pose) {
    angularDampingOverride=1
    angularSpringConstant=10000
    bAngularSpring=true
    bOverrideAngularDamping=true
    bUseGravityOverride=true
    gravityScale=0.35
    simulationSpace=RootRelative
    alphaBoolBlend {blendOption=Linear}
    alphaScaleBiasClamp {scale=0.5}
    boundBone {boneName=dyn_L_pantLeg_back_1}
    componentAppliedLinearAccClamp=<2500 2500 2500>
    componentLinearAccScale=<1 1 1>
    componentLinearVelScale=<1 1 1>
    gravityOverride {z=-980}
    physicsBodyDefinitions [{
      boundBone {boneName=dyn_L_pantLeg_back_1}
      boxExtents=<25 25 25>
      constraintSetup {angularLimitsMax=<5 5 15>, angularLimitsMin=<-5 -5 0>, angularTarget=<1 0 0>}
      localJointOffset=<-4 0 0>
     }]
   }
   n67 = Anim Dynamics - Bone: dyn_L_pantLeg_back_2!(SimulationSpace:AnimPhysSimSpaceType Enum, ComponentPose=@n66, Alpha=1) -> (Pose) {
    angularDampingOverride=1
    angularSpringConstant=10000
    bAngularSpring=true
    bOverrideAngularDamping=true
    bUseGravityOverride=true
    gravityScale=0.35
    simulationSpace=RootRelative
    alphaBoolBlend {blendOption=Linear}
    alphaScaleBiasClamp {scale=0.5}
    boundBone {boneName=dyn_L_pantLeg_back_2}
    componentAppliedLinearAccClamp=<2500 2500 2500>
    componentLinearAccScale=<1 1 1>
    componentLinearVelScale=<1 1 1>
    gravityOverride {z=-980}
    physicsBodyDefinitions [{
      boundBone {boneName=dyn_L_pantLeg_back_2}
      boxExtents=<25 25 25>
      constraintSetup {angularLimitsMax=<5 5 15>, angularLimitsMin=<-5 -5 0>, angularTarget=<1 0 0>}
      localJointOffset=<-4 0 0>
     }]
   }
   n68 = Anim Dynamics - Bone: dyn_R_pantLeg_front_1!(SimulationSpace:AnimPhysSimSpaceType Enum, ComponentPose=@n67, Alpha=1) -> (Pose) {
    angularDampingOverride=1
    angularSpringConstant=1000
    bAngularSpring=true
    bOverrideAngularDamping=true
    bUseGravityOverride=true
    gravityScale=0.35
    simulationSpace=RootRelative
    alphaBoolBlend {blendOption=Linear}
    alphaScaleBiasClamp {scale=0.75}
    boundBone {boneName=dyn_R_pantLeg_front_1}
    componentAppliedLinearAccClamp=<2500 2500 2500>
    componentLinearAccScale=<1 1 1>
    componentLinearVelScale=<1 1 1>
    gravityOverride {z=-980}
    physicsBodyDefinitions [{
      boundBone {boneName=dyn_R_pantLeg_front_1}
      boxExtents=<23 23 23>
      constraintSetup {angularLimitsMax=<5 30 15>, angularLimitsMin=<-5 -5 -5>, angularTarget=<1 0 0>}
      localJointOffset=<4 0 0>
     }]
   }
   n69 = Anim Dynamics - Bone: dyn_R_pantLeg_front_2!(SimulationSpace:AnimPhysSimSpaceType Enum, ComponentPose=@n68, Alpha=1) -> (Pose) {
    angularDampingOverride=1
    angularSpringConstant=6000
    bAngularSpring=true
    bOverrideAngularDamping=true
    bUseGravityOverride=true
    gravityScale=0.35
    simulationSpace=RootRelative
    alphaBoolBlend {blendOption=Linear}
    alphaScaleBiasClamp {scale=0.75}
    boundBone {boneName=dyn_R_pantLeg_front_2}
    componentAppliedLinearAccClamp=<2500 2500 2500>
    componentLinearAccScale=<1 1 1>
    componentLinearVelScale=<1 1 1>
    gravityOverride {z=-980}
    physicsBodyDefinitions [{
      boundBone {boneName=dyn_R_pantLeg_front_2}
      boxExtents=<23 23 23>
      constraintSetup {angularLimitsMax=<5 30 60>, angularLimitsMin=<-5 -5 -30>, angularTarget=<1 0 0>}
      localJointOffset=<4 0 0>
     }]
   }
   n70 = Anim Dynamics - Bone: dyn_R_pantLeg_inner_1!(SimulationSpace:AnimPhysSimSpaceType Enum, ComponentPose=@n69, Alpha=1) -> (Pose) {
    angularDampingOverride=1
    angularSpringConstant=5000
    bAngularSpring=true
    bOverrideAngularDamping=true
    bUseGravityOverride=true
    gravityScale=0.35
    simulationSpace=RootRelative
    alphaBoolBlend {blendOption=Linear}
    boundBone {boneName=dyn_R_pantLeg_inner_1}
    componentAppliedLinearAccClamp=<2500 2500 2500>
    componentLinearAccScale=<1 1 1>
    componentLinearVelScale=<1 1 1>
    gravityOverride {z=-980}
    physicsBodyDefinitions [{
      boundBone {boneName=dyn_R_pantLeg_inner_1}
      boxExtents=<23 23 23>
      constraintSetup {angularLimitsMax=<5 10 10>, angularLimitsMin=<-5 -10 -35>, angularTarget=<1 0 0>}
      localJointOffset=<4 0 0>
     }]
   }
   n71 = Anim Dynamics - Bone: dyn_R_pantLeg_inner_2!(SimulationSpace:AnimPhysSimSpaceType Enum, ComponentPose=@n70, Alpha=1) -> (Pose) {
    angularDampingOverride=1
    angularSpringConstant=10000
    bAngularSpring=true
    bOverrideAngularDamping=true
    bUseGravityOverride=true
    gravityScale=0.35
    simulationSpace=RootRelative
    alphaBoolBlend {blendOption=Linear}
    boundBone {boneName=dyn_R_pantLeg_inner_2}
    componentAppliedLinearAccClamp=<2500 2500 2500>
    componentLinearAccScale=<1 1 1>
    componentLinearVelScale=<1 1 1>
    gravityOverride {z=-980}
    physicsBodyDefinitions [{
      boundBone {boneName=dyn_R_pantLeg_inner_2}
      boxExtents=<23 23 23>
      constraintSetup {angularLimitsMax=<5 10 10>, angularLimitsMin=<-5 -10 -35>, angularTarget=<1 0 0>}
      localJointOffset=<4 0 0>
     }]
   }
   n72 = Anim Dynamics - Bone: dyn_R_pantLeg_outer_1!(SimulationSpace:AnimPhysSimSpaceType Enum, ComponentPose=@n71, Alpha=1) -> (Pose) {
    angularDampingOverride=1
    angularSpringConstant=10000
    bAngularSpring=true
    bOverrideAngularDamping=true
    bUseGravityOverride=true
    gravityScale=0.35
    simulationSpace=RootRelative
    alphaBoolBlend {blendOption=Linear}
    boundBone {boneName=dyn_R_pantLeg_outer_1}
    componentAppliedLinearAccClamp=<2500 2500 2500>
    componentLinearAccScale=<1 1 1>
    componentLinearVelScale=<1 1 1>
    gravityOverride {z=-980}
    physicsBodyDefinitions [{
      boundBone {boneName=dyn_R_pantLeg_outer_1}
      boxExtents=<23 23 23>
      constraintSetup {angularLimitsMax=<5 30 0>, angularLimitsMin=<-5 -30 -10>, angularTarget=<1 0 0>}
      localJointOffset=<4 0 0>
     }]
   }
   n73 = Anim Dynamics - Bone: dyn_R_pantLeg_outer_2!(SimulationSpace:AnimPhysSimSpaceType Enum, ComponentPose=@n72, Alpha=1) -> (Pose) {
    angularDampingOverride=1
    angularSpringConstant=10000
    bAngularSpring=true
    bOverrideAngularDamping=true
    bUseGravityOverride=true
    gravityScale=0.35
    simulationSpace=RootRelative
    alphaBoolBlend {blendOption=Linear}
    boundBone {boneName=dyn_R_pantLeg_outer_2}
    componentAppliedLinearAccClamp=<2500 2500 2500>
    componentLinearAccScale=<1 1 1>
    componentLinearVelScale=<1 1 1>
    gravityOverride {z=-980}
    physicsBodyDefinitions [{
      boundBone {boneName=dyn_R_pantLeg_outer_2}
      boxExtents=<23 23 23>
      constraintSetup {angularLimitsMax=<5 30 0>, angularLimitsMin=<-5 -30 -10>, angularTarget=<1 0 0>}
      localJointOffset=<4 0 0>
     }]
   }
   n74 = Anim Dynamics - Bone: dyn_R_pantLeg_back_1!(SimulationSpace:AnimPhysSimSpaceType Enum, ComponentPose=@n73, Alpha=1) -> (Pose) {
    angularDampingOverride=1
    angularSpringConstant=10000
    bAngularSpring=true
    bOverrideAngularDamping=true
    bUseGravityOverride=true
    gravityScale=0.35
    simulationSpace=RootRelative
    alphaBoolBlend {blendOption=Linear}
    alphaScaleBiasClamp {scale=0.5}
    boundBone {boneName=dyn_R_pantLeg_back_1}
    componentAppliedLinearAccClamp=<2500 2500 2500>
    componentLinearAccScale=<1 1 1>
    componentLinearVelScale=<1 1 1>
    gravityOverride {z=-980}
    physicsBodyDefinitions [{
      boundBone {boneName=dyn_R_pantLeg_back_1}
      boxExtents=<23 23 23>
      constraintSetup {angularLimitsMax=<5 5 15>, angularLimitsMin=<-5 -5 0>, angularTarget=<1 0 0>}
      localJointOffset=<4 0 0>
     }]
   }
   n75 = Anim Dynamics - Bone: dyn_R_pantLeg_back_2!(SimulationSpace:AnimPhysSimSpaceType Enum, ComponentPose=@n74, Alpha=1) -> (Pose) {
    angularDampingOverride=1
    angularSpringConstant=10000
    bAngularSpring=true
    bOverrideAngularDamping=true
    bUseGravityOverride=true
    gravityScale=0.35
    simulationSpace=RootRelative
    alphaBoolBlend {blendOption=Linear}
    alphaScaleBiasClamp {scale=0.5}
    boundBone {boneName=dyn_R_pantLeg_back_2}
    componentAppliedLinearAccClamp=<2500 2500 2500>
    componentLinearAccScale=<1 1 1>
    componentLinearVelScale=<1 1 1>
    gravityOverride {z=-980}
    physicsBodyDefinitions [{
      boundBone {boneName=dyn_R_pantLeg_back_2}
      boxExtents=<23 23 23>
      constraintSetup {angularLimitsMax=<5 5 15>, angularLimitsMin=<-5 -5 0>, angularTarget=<1 0 0>}
      localJointOffset=<4 0 0>
     }]
   }
   n76 = Anim Dynamics - Bone: dyn_L_front_chainMain_A_1!(SimulationSpace:AnimPhysSimSpaceType Enum, ComponentPose=@n75, Alpha=1) -> (Pose) {
    angularDampingOverride=1
    angularSpringConstant=3000
    bAngularSpring=true
    bOverrideAngularDamping=true
    bUseGravityOverride=true
    bUsePlanarLimit=false
    gravityScale=0.35
    simulationSpace=RootRelative
    alphaBoolBlend {blendOption=Linear}
    alphaScaleBiasClamp {bInterpResult=true}
    boundBone {boneName=dyn_L_front_chainMain_A_1}
    componentAppliedLinearAccClamp=<2500 2500 2500>
    componentLinearAccScale=<1 1 1>
    componentLinearVelScale=<1 1 1>
    gravityOverride {z=-980}
    physicsBodyDefinitions [{
      boundBone {boneName=dyn_L_front_chainMain_A_1}
      boxExtents=<30 30 30>
      constraintSetup {angularLimitsMax=<0 0 0>, angularLimitsMin=<0 -15 0>, angularTarget=<1 0 0>}
      localJointOffset=<4 0 0>
     }]
   }
   n77 = Anim Dynamics - Bone: dyn_L_front_chainMain_A_2!(SimulationSpace:AnimPhysSimSpaceType Enum, ComponentPose=@n76, Alpha=1) -> (Pose) {
    angularDampingOverride=1
    angularSpringConstant=3000
    bAngularSpring=true
    bOverrideAngularDamping=true
    bUseGravityOverride=true
    bUsePlanarLimit=false
    gravityScale=0.35
    simulationSpace=RootRelative
    alphaBoolBlend {blendOption=Linear}
    alphaScaleBiasClamp {bInterpResult=true}
    boundBone {boneName=dyn_L_front_chainMain_A_2}
    componentAppliedLinearAccClamp=<2500 2500 2500>
    componentLinearAccScale=<1 1 1>
    componentLinearVelScale=<1 1 1>
    gravityOverride {z=-980}
    physicsBodyDefinitions [{
      boundBone {boneName=dyn_L_front_chainMain_A_2}
      boxExtents=<30 30 30>
      constraintSetup {angularLimitsMax=<0 0 0>, angularLimitsMin=<0 -15 0>, angularTarget=<1 0 0>}
      localJointOffset=<4 0 0>
     }]
   }
   n78 = Anim Dynamics - Bone: dyn_L_frontChain_A_1!(SimulationSpace:AnimPhysSimSpaceType Enum, ComponentPose=@n77, Alpha=1) -> (Pose) {
    angularDampingOverride=1
    angularSpringConstant=3000
    bAngularSpring=true
    bOverrideAngularDamping=true
    bUseGravityOverride=true
    bUsePlanarLimit=false
    gravityScale=0.35
    simulationSpace=RootRelative
    alphaBoolBlend {blendOption=Linear}
    alphaScaleBiasClamp {bInterpResult=true}
    boundBone {boneName=dyn_L_frontChain_A_1}
    componentAppliedLinearAccClamp=<2500 2500 2500>
    componentLinearAccScale=<1 1 1>
    componentLinearVelScale=<1 1 1>
    gravityOverride {z=-980}
    physicsBodyDefinitions [{
      boundBone {boneName=dyn_L_frontChain_A_1}
      boxExtents=<30 30 30>
      constraintSetup {angularLimitsMax=<0 0 5>, angularLimitsMin=<0 -15 -5>, angularTarget=<1 0 0>}
      localJointOffset=<4 0 0>
     }]
   }
   n79 = Anim Dynamics - Bone: dyn_L_frontChain_A_2!(SimulationSpace:AnimPhysSimSpaceType Enum, ComponentPose=@n78, Alpha=1) -> (Pose) {
    angularDampingOverride=1
    angularSpringConstant=3000
    bAngularSpring=true
    bOverrideAngularDamping=true
    bUseGravityOverride=true
    bUsePlanarLimit=false
    gravityScale=0.35
    simulationSpace=RootRelative
    alphaBoolBlend {blendOption=Linear}
    alphaScaleBiasClamp {bInterpResult=true}
    boundBone {boneName=dyn_L_frontChain_A_2}
    componentAppliedLinearAccClamp=<2500 2500 2500>
    componentLinearAccScale=<1 1 1>
    componentLinearVelScale=<1 1 1>
    gravityOverride {z=-980}
    physicsBodyDefinitions [{
      boundBone {boneName=dyn_L_frontChain_A_2}
      boxExtents=<26 26 26>
      constraintSetup {angularLimitsMax=<10 5 15>, angularLimitsMin=<-10 -15 -15>, angularTarget=<1 0 0>}
      localJointOffset=<4 0 0>
     }]
   }
   n80 = Anim Dynamics - Bone: dyn_L_frontChain_A_3!(SimulationSpace:AnimPhysSimSpaceType Enum, ComponentPose=@n79, Alpha=1) -> (Pose) {
    angularDampingOverride=1
    angularSpringConstant=3000
    bAngularSpring=true
    bOverrideAngularDamping=true
    bUseGravityOverride=true
    bUsePlanarLimit=false
    gravityScale=0.35
    simulationSpace=RootRelative
    alphaBoolBlend {blendOption=Linear}
    alphaScaleBiasClamp {bInterpResult=true}
    boundBone {boneName=dyn_L_frontChain_A_3}
    componentAppliedLinearAccClamp=<2500 2500 2500>
    componentLinearAccScale=<1 1 1>
    componentLinearVelScale=<1 1 1>
    gravityOverride {z=-980}
    physicsBodyDefinitions [{
      boundBone {boneName=dyn_L_frontChain_A_3}
      boxExtents=<24 24 24>
      constraintSetup {angularLimitsMax=<10 0 15>, angularLimitsMin=<-10 -15 -15>, angularTarget=<1 0 0>}
      localJointOffset=<4 0 0>
     }]
   }
   n81 = Anim Dynamics - Bone: dyn_L_front_chainMain_B_1!(SimulationSpace:AnimPhysSimSpaceType Enum, ComponentPose=@n80, Alpha=1) -> (Pose) {
    angularDampingOverride=1
    angularSpringConstant=3000
    bAngularSpring=true
    bOverrideAngularDamping=true
    bUseGravityOverride=true
    gravityScale=0.35
    simulationSpace=RootRelative
    alphaBoolBlend {blendOption=Linear}
    alphaScaleBiasClamp {bInterpResult=true}
    boundBone {boneName=dyn_L_front_chainMain_B_1}
    componentAppliedLinearAccClamp=<2500 2500 2500>
    componentLinearAccScale=<1 1 1>
    componentLinearVelScale=<1 1 1>
    gravityOverride {z=-980}
    physicsBodyDefinitions [{
      boundBone {boneName=dyn_L_front_chainMain_B_1}
      boxExtents=<30 30 30>
      constraintSetup {angularLimitsMax=<0 0 0>, angularLimitsMin=<0 -15 0>, angularTarget=<1 0 0>}
      localJointOffset=<4 0 0>
     }]
   }
   n82 = Anim Dynamics - Bone: dyn_L_front_chainMain_B_2!(SimulationSpace:AnimPhysSimSpaceType Enum, ComponentPose=@n81, Alpha=1) -> (Pose) {
    angularDampingOverride=1
    angularSpringConstant=3000
    bAngularSpring=true
    bOverrideAngularDamping=true
    bUseGravityOverride=true
    gravityScale=0.35
    simulationSpace=RootRelative
    alphaBoolBlend {blendOption=Linear}
    alphaScaleBiasClamp {bInterpResult=true}
    boundBone {boneName=dyn_L_front_chainMain_B_2}
    componentAppliedLinearAccClamp=<2500 2500 2500>
    componentLinearAccScale=<1 1 1>
    componentLinearVelScale=<1 1 1>
    gravityOverride {z=-980}
    physicsBodyDefinitions [{
      boundBone {boneName=dyn_L_front_chainMain_B_2}
      boxExtents=<30 30 30>
      constraintSetup {angularLimitsMax=<0 0 0>, angularLimitsMin=<0 -15 0>, angularTarget=<1 0 0>}
      localJointOffset=<4 0 0>
     }]
   }
   n83 = Anim Dynamics - Bone: dyn_L_frontChain_B_1!(SimulationSpace:AnimPhysSimSpaceType Enum, ComponentPose=@n82, Alpha=1) -> (Pose) {
    angularDampingOverride=1
    angularSpringConstant=3000
    bAngularSpring=true
    bOverrideAngularDamping=true
    bUseGravityOverride=true
    gravityScale=0.35
    simulationSpace=RootRelative
    alphaBoolBlend {blendOption=Linear}
    alphaScaleBiasClamp {bInterpResult=true}
    boundBone {boneName=dyn_L_frontChain_B_1}
    componentAppliedLinearAccClamp=<2500 2500 2500>
    componentLinearAccScale=<1 1 1>
    componentLinearVelScale=<1 1 1>
    gravityOverride {z=-980}
    physicsBodyDefinitions [{
      boundBone {boneName=dyn_L_frontChain_B_1}
      boxExtents=<30 30 30>
      constraintSetup {angularLimitsMax=<0 5 5>, angularLimitsMin=<0 -15 -5>, angularTarget=<1 0 0>}
      localJointOffset=<4 0 0>
     }]
   }
   n84 = Anim Dynamics - Bone: dyn_L_frontChain_B_2!(SimulationSpace:AnimPhysSimSpaceType Enum, ComponentPose=@n83, Alpha=1) -> (Pose) {
    angularDampingOverride=1
    angularSpringConstant=3000
    bAngularSpring=true
    bOverrideAngularDamping=true
    bUseGravityOverride=true
    gravityScale=0.35
    simulationSpace=RootRelative
    alphaBoolBlend {blendOption=Linear}
    alphaScaleBiasClamp {bInterpResult=true}
    boundBone {boneName=dyn_L_frontChain_B_2}
    componentAppliedLinearAccClamp=<2500 2500 2500>
    componentLinearAccScale=<1 1 1>
    componentLinearVelScale=<1 1 1>
    gravityOverride {z=-980}
    physicsBodyDefinitions [{
      boundBone {boneName=dyn_L_frontChain_B_2}
      boxExtents=<26 26 26>
      constraintSetup {angularLimitsMax=<10 5 15>, angularLimitsMin=<-10 -15 -15>, angularTarget=<1 0 0>}
      localJointOffset=<4 0 0>
     }]
   }
   n85 = Anim Dynamics - Bone: dyn_L_frontChain_B_3!(SimulationSpace:AnimPhysSimSpaceType Enum, ComponentPose=@n84, Alpha=1) -> (Pose) {
    angularDampingOverride=1
    angularSpringConstant=3000
    bAngularSpring=true
    bOverrideAngularDamping=true
    bUseGravityOverride=true
    gravityScale=0.35
    simulationSpace=RootRelative
    alphaBoolBlend {blendOption=Linear}
    alphaScaleBiasClamp {bInterpResult=true}
    boundBone {boneName=dyn_L_frontChain_B_3}
    componentAppliedLinearAccClamp=<2500 2500 2500>
    componentLinearAccScale=<1 1 1>
    componentLinearVelScale=<1 1 1>
    gravityOverride {z=-980}
    physicsBodyDefinitions [{
      boundBone {boneName=dyn_L_frontChain_B_3}
      boxExtents=<24 24 24>
      constraintSetup {angularLimitsMax=<10 0 30>, angularLimitsMin=<-10 -15 -30>, angularTarget=<1 0 0>}
      localJointOffset=<4 0 0>
     }]
   }
   n86 = Anim Dynamics - Bone: dyn_L_frontChain_B_4!(SimulationSpace:AnimPhysSimSpaceType Enum, ComponentPose=@n85, Alpha=1) -> (Pose) {
    angularDampingOverride=1
    angularSpringConstant=3000
    bAngularSpring=true
    bOverrideAngularDamping=true
    bUseGravityOverride=true
    gravityScale=0.35
    simulationSpace=RootRelative
    alphaBoolBlend {blendOption=Linear}
    alphaScaleBiasClamp {bInterpResult=true}
    boundBone {boneName=dyn_L_frontChain_B_4}
    componentAppliedLinearAccClamp=<2500 2500 2500>
    componentLinearAccScale=<1 1 1>
    componentLinearVelScale=<1 1 1>
    gravityOverride {z=-980}
    physicsBodyDefinitions [{
      boundBone {boneName=dyn_L_frontChain_B_4}
      boxExtents=<24 24 24>
      constraintSetup {angularLimitsMax=<10 0 30>, angularLimitsMin=<-10 -15 -30>, angularTarget=<1 0 0>}
      localJointOffset=<4 0 0>
     }]
   }
   n87 = Anim Dynamics - Bone: dyn_L_side_chainMain_1!(SimulationSpace:AnimPhysSimSpaceType Enum, ComponentPose=@n86, Alpha=1) -> (Pose) {
    angularDampingOverride=1
    angularSpringConstant=3000
    bAngularSpring=true
    bOverrideAngularDamping=true
    bUseGravityOverride=true
    gravityScale=0.35
    simulationSpace=RootRelative
    alphaBoolBlend {blendOption=Linear}
    alphaScaleBiasClamp {bInterpResult=true}
    boundBone {boneName=dyn_L_side_chainMain_1}
    componentAppliedLinearAccClamp=<2500 2500 2500>
    componentLinearAccScale=<1 1 1>
    componentLinearVelScale=<1 1 1>
    gravityOverride {z=-980}
    physicsBodyDefinitions [{
      boundBone {boneName=dyn_L_side_chainMain_1}
      boxExtents=<30 30 30>
      constraintSetup {angularLimitsMax=<0 0 5>, angularLimitsMin=<0 -15 -5>, angularTarget=<1 0 0>}
      localJointOffset=<4 0 0>
     }]
   }
   n88 = Anim Dynamics - Bone: dyn_L_side_chainMain_2!(SimulationSpace:AnimPhysSimSpaceType Enum, ComponentPose=@n87, Alpha=1) -> (Pose) {
    angularDampingOverride=1
    angularSpringConstant=3000
    bAngularSpring=true
    bOverrideAngularDamping=true
    bUseGravityOverride=true
    gravityScale=0.35
    simulationSpace=RootRelative
    alphaBoolBlend {blendOption=Linear}
    alphaScaleBiasClamp {bInterpResult=true}
    boundBone {boneName=dyn_L_side_chainMain_2}
    componentAppliedLinearAccClamp=<2500 2500 2500>
    componentLinearAccScale=<1 1 1>
    componentLinearVelScale=<1 1 1>
    gravityOverride {z=-980}
    physicsBodyDefinitions [{
      boundBone {boneName=dyn_L_side_chainMain_2}
      boxExtents=<30 30 30>
      constraintSetup {angularLimitsMax=<0 0 5>, angularLimitsMin=<0 -15 -5>, angularTarget=<1 0 0>}
      localJointOffset=<4 0 0>
     }]
   }
   n89 = Anim Dynamics - Bone: dyn_L_sideChain_A_1!(SimulationSpace:AnimPhysSimSpaceType Enum, ComponentPose=@n88, Alpha=1) -> (Pose) {
    angularDampingOverride=1
    angularSpringConstant=3000
    bAngularSpring=true
    bOverrideAngularDamping=true
    bUseGravityOverride=true
    gravityScale=0.35
    simulationSpace=RootRelative
    alphaBoolBlend {blendOption=Linear}
    alphaScaleBiasClamp {bInterpResult=true}
    boundBone {boneName=dyn_L_sideChain_A_1}
    componentAppliedLinearAccClamp=<2500 2500 2500>
    componentLinearAccScale=<1 1 1>
    componentLinearVelScale=<1 1 1>
    gravityOverride {z=-980}
    physicsBodyDefinitions [{
      boundBone {boneName=dyn_L_sideChain_A_1}
      boxExtents=<30 30 30>
      constraintSetup {angularLimitsMax=<10 5 15>, angularLimitsMin=<-10 -15 -15>, angularTarget=<1 0 0>}
      localJointOffset=<4 0 0>
     }]
   }
   n90 = Anim Dynamics - Bone: dyn_L_sideChain_A_2!(SimulationSpace:AnimPhysSimSpaceType Enum, ComponentPose=@n89, Alpha=1) -> (Pose) {
    angularDampingOverride=1
    angularSpringConstant=3000
    bAngularSpring=true
    bOverrideAngularDamping=true
    bUseGravityOverride=true
    gravityScale=0.35
    simulationSpace=RootRelative
    alphaBoolBlend {blendOption=Linear}
    alphaScaleBiasClamp {bInterpResult=true}
    boundBone {boneName=dyn_L_sideChain_A_2}
    componentAppliedLinearAccClamp=<2500 2500 2500>
    componentLinearAccScale=<1 1 1>
    componentLinearVelScale=<1 1 1>
    gravityOverride {z=-980}
    physicsBodyDefinitions [{
      boundBone {boneName=dyn_L_sideChain_A_2}
      boxExtents=<26 26 26>
      constraintSetup {angularLimitsMax=<10 5 15>, angularLimitsMin=<-10 -15 -15>, angularTarget=<1 0 0>}
      localJointOffset=<4 0 0>
     }]
   }
   n91 = Anim Dynamics - Bone: dyn_L_sideChain_A_3!(SimulationSpace:AnimPhysSimSpaceType Enum, ComponentPose=@n90, Alpha=1) -> (Pose) {
    angularDampingOverride=1
    angularSpringConstant=3000
    bAngularSpring=true
    bOverrideAngularDamping=true
    bUseGravityOverride=true
    gravityScale=0.35
    simulationSpace=RootRelative
    alphaBoolBlend {blendOption=Linear}
    alphaScaleBiasClamp {bInterpResult=true}
    boundBone {boneName=dyn_L_sideChain_A_3}
    componentAppliedLinearAccClamp=<2500 2500 2500>
    componentLinearAccScale=<1 1 1>
    componentLinearVelScale=<1 1 1>
    gravityOverride {z=-980}
    physicsBodyDefinitions [{
      boundBone {boneName=dyn_L_sideChain_A_3}
      boxExtents=<24 24 24>
      constraintSetup {angularLimitsMax=<10 5 15>, angularLimitsMin=<-10 -15 -15>, angularTarget=<1 0 0>}
      localJointOffset=<4 0 0>
     }]
   }
   n92 = Anim Dynamics - Bone: dyn_L_sideChain_A_4!(SimulationSpace:AnimPhysSimSpaceType Enum, ComponentPose=@n91, Alpha=1) -> (Pose) {
    angularDampingOverride=1
    angularSpringConstant=3000
    bAngularSpring=true
    bOverrideAngularDamping=true
    bUseGravityOverride=true
    gravityScale=0.35
    simulationSpace=RootRelative
    alphaBoolBlend {blendOption=Linear}
    alphaScaleBiasClamp {bInterpResult=true}
    boundBone {boneName=dyn_L_sideChain_A_4}
    componentAppliedLinearAccClamp=<2500 2500 2500>
    componentLinearAccScale=<1 1 1>
    componentLinearVelScale=<1 1 1>
    gravityOverride {z=-980}
    physicsBodyDefinitions [{
      boundBone {boneName=dyn_L_sideChain_A_4}
      boxExtents=<24 24 24>
      constraintSetup {angularLimitsMax=<10 5 15>, angularLimitsMin=<-10 -15 -15>, angularTarget=<1 0 0>}
      localJointOffset=<4 0 0>
     }]
   }
   n93 = Anim Dynamics - Bone: dyn_L_sideChain_B_1!(SimulationSpace:AnimPhysSimSpaceType Enum, ComponentPose=@n92, Alpha=1) -> (Pose) {
    angularDampingOverride=1
    angularSpringConstant=3000
    bAngularSpring=true
    bOverrideAngularDamping=true
    bUseGravityOverride=true
    gravityScale=0.35
    simulationSpace=RootRelative
    alphaBoolBlend {blendOption=Linear}
    alphaScaleBiasClamp {bInterpResult=true}
    boundBone {boneName=dyn_L_sideChain_B_1}
    componentAppliedLinearAccClamp=<2500 2500 2500>
    componentLinearAccScale=<1 1 1>
    componentLinearVelScale=<1 1 1>
    gravityOverride {z=-980}
    physicsBodyDefinitions [{
      boundBone {boneName=dyn_L_sideChain_B_1}
      boxExtents=<30 30 30>
      constraintSetup {angularLimitsMax=<0 0 15>, angularLimitsMin=<0 -15 -15>, angularTarget=<1 0 0>}
      localJointOffset=<4 0 0>
     }]
   }
   n94 = Anim Dynamics - Bone: dyn_L_sideChain_B_2!(SimulationSpace:AnimPhysSimSpaceType Enum, ComponentPose=@n93, Alpha=1) -> (Pose) {
    angularDampingOverride=1
    angularSpringConstant=3000
    bAngularSpring=true
    bOverrideAngularDamping=true
    bUseGravityOverride=true
    gravityScale=0.35
    simulationSpace=RootRelative
    alphaBoolBlend {blendOption=Linear}
    alphaScaleBiasClamp {bInterpResult=true}
    boundBone {boneName=dyn_L_sideChain_B_2}
    componentAppliedLinearAccClamp=<2500 2500 2500>
    componentLinearAccScale=<1 1 1>
    componentLinearVelScale=<1 1 1>
    gravityOverride {z=-980}
    physicsBodyDefinitions [{
      boundBone {boneName=dyn_L_sideChain_B_2}
      boxExtents=<26 26 26>
      constraintSetup {angularLimitsMax=<0 5 15>, angularLimitsMin=<0 -15 -15>, angularTarget=<1 0 0>}
      localJointOffset=<4 0 0>
     }]
   }
   n95 = Anim Dynamics - Bone: dyn_L_sideChain_B_3!(SimulationSpace:AnimPhysSimSpaceType Enum, ComponentPose=@n94, Alpha=1) -> (Pose) {
    angularDampingOverride=1
    angularSpringConstant=3000
    bAngularSpring=true
    bOverrideAngularDamping=true
    bUseGravityOverride=true
    gravityScale=0.35
    simulationSpace=RootRelative
    alphaBoolBlend {blendOption=Linear}
    alphaScaleBiasClamp {bInterpResult=true}
    boundBone {boneName=dyn_L_sideChain_B_3}
    componentAppliedLinearAccClamp=<2500 2500 2500>
    componentLinearAccScale=<1 1 1>
    componentLinearVelScale=<1 1 1>
    gravityOverride {z=-980}
    physicsBodyDefinitions [{
      boundBone {boneName=dyn_L_sideChain_B_3}
      boxExtents=<24 24 24>
      constraintSetup {angularLimitsMax=<10 5 15>, angularLimitsMin=<-10 -15 -15>, angularTarget=<1 0 0>}
      localJointOffset=<4 0 0>
     }]
   }
   n96 = Anim Dynamics - Bone: dyn_L_sideChain_B_4!(SimulationSpace:AnimPhysSimSpaceType Enum, ComponentPose=@n95, Alpha=1) -> (Pose) {
    angularDampingOverride=1
    angularSpringConstant=3000
    bAngularSpring=true
    bOverrideAngularDamping=true
    bUseGravityOverride=true
    gravityScale=0.35
    simulationSpace=RootRelative
    alphaBoolBlend {blendOption=Linear}
    alphaScaleBiasClamp {bInterpResult=true}
    boundBone {boneName=dyn_L_sideChain_B_4}
    componentAppliedLinearAccClamp=<2500 2500 2500>
    componentLinearAccScale=<1 1 1>
    componentLinearVelScale=<1 1 1>
    gravityOverride {z=-980}
    physicsBodyDefinitions [{
      boundBone {boneName=dyn_L_sideChain_B_4}
      boxExtents=<24 24 24>
      constraintSetup {angularLimitsMax=<10 5 15>, angularLimitsMin=<-10 -15 -15>, angularTarget=<1 0 0>}
      localJointOffset=<4 0 0>
     }]
   }
   n97 = Anim Dynamics - Bone: dyn_back_chainMain_1!(SimulationSpace:AnimPhysSimSpaceType Enum, ComponentPose=@n96, Alpha=1) -> (Pose) {
    angularDampingOverride=1
    angularSpringConstant=3000
    bAngularSpring=true
    bOverrideAngularDamping=true
    bUseGravityOverride=true
    gravityScale=0.35
    simulationSpace=RootRelative
    alphaBoolBlend {blendOption=Linear}
    alphaScaleBiasClamp {bInterpResult=true}
    boundBone {boneName=dyn_back_chainMain_1}
    componentAppliedLinearAccClamp=<2500 2500 2500>
    componentLinearAccScale=<1 1 1>
    componentLinearVelScale=<1 1 1>
    gravityOverride {z=-980}
    physicsBodyDefinitions [{
      boundBone {boneName=dyn_back_chainMain_1}
      boxExtents=<30 30 30>
      constraintSetup {angularLimitsMax=<0 0 5>, angularLimitsMin=<0 -15 -5>, angularTarget=<1 0 0>}
      localJointOffset=<4 0 0>
     }]
   }
   n98 = Anim Dynamics - Bone: dyn_back_chainMain_2!(SimulationSpace:AnimPhysSimSpaceType Enum, ComponentPose=@n97, Alpha=1) -> (Pose) {
    angularDampingOverride=1
    angularSpringConstant=3000
    bAngularSpring=true
    bOverrideAngularDamping=true
    bUseGravityOverride=true
    gravityScale=0.35
    simulationSpace=RootRelative
    alphaBoolBlend {blendOption=Linear}
    alphaScaleBiasClamp {bInterpResult=true}
    boundBone {boneName=dyn_back_chainMain_2}
    componentAppliedLinearAccClamp=<2500 2500 2500>
    componentLinearAccScale=<1 1 1>
    componentLinearVelScale=<1 1 1>
    gravityOverride {z=-980}
    physicsBodyDefinitions [{
      boundBone {boneName=dyn_back_chainMain_2}
      boxExtents=<30 30 30>
      constraintSetup {angularLimitsMax=<0 0 15>, angularLimitsMin=<0 -15 -15>, angularTarget=<1 0 0>}
      localJointOffset=<4 0 0>
     }]
   }
   n99 = Anim Dynamics - Bone: dyn_back_chain_1!(SimulationSpace:AnimPhysSimSpaceType Enum, ComponentPose=@n98, Alpha=1) -> (Pose) {
    angularDampingOverride=1
    angularSpringConstant=3000
    bAngularSpring=true
    bOverrideAngularDamping=true
    bUseGravityOverride=true
    gravityScale=0.35
    simulationSpace=RootRelative
    alphaBoolBlend {blendOption=Linear}
    alphaScaleBiasClamp {bInterpResult=true}
    boundBone {boneName=dyn_back_chain_1}
    componentAppliedLinearAccClamp=<2500 2500 2500>
    componentLinearAccScale=<1 1 1>
    componentLinearVelScale=<1 1 1>
    gravityOverride {z=-980}
    physicsBodyDefinitions [{
      boundBone {boneName=dyn_back_chain_1}
      boxExtents=<30 30 30>
      constraintSetup {angularLimitsMax=<10 5 15>, angularLimitsMin=<-10 -15 -15>, angularTarget=<1 0 0>}
      localJointOffset=<4 0 0>
     }]
   }
   n100 = Anim Dynamics - Bone: dyn_back_chain_2!(SimulationSpace:AnimPhysSimSpaceType Enum, ComponentPose=@n99, Alpha=1) -> (Pose) {
    angularDampingOverride=1
    angularSpringConstant=3000
    bAngularSpring=true
    bOverrideAngularDamping=true
    bUseGravityOverride=true
    gravityScale=0.35
    simulationSpace=RootRelative
    alphaBoolBlend {blendOption=Linear}
    alphaScaleBiasClamp {bInterpResult=true}
    boundBone {boneName=dyn_back_chain_2}
    componentAppliedLinearAccClamp=<2500 2500 2500>
    componentLinearAccScale=<1 1 1>
    componentLinearVelScale=<1 1 1>
    gravityOverride {z=-980}
    physicsBodyDefinitions [{
      boundBone {boneName=dyn_back_chain_2}
      boxExtents=<26 26 26>
      constraintSetup {angularLimitsMax=<10 5 15>, angularLimitsMin=<-10 -15 -15>, angularTarget=<1 0 0>}
      localJointOffset=<4 0 0>
     }]
   }
   n101 = Anim Dynamics - Bone: dyn_back_chain_3!(SimulationSpace:AnimPhysSimSpaceType Enum, ComponentPose=@n100, Alpha=1) -> (Pose) {
    angularDampingOverride=1
    angularSpringConstant=3000
    bAngularSpring=true
    bOverrideAngularDamping=true
    bUseGravityOverride=true
    gravityScale=0.35
    simulationSpace=RootRelative
    alphaBoolBlend {blendOption=Linear}
    alphaScaleBiasClamp {bInterpResult=true}
    boundBone {boneName=dyn_back_chain_3}
    componentAppliedLinearAccClamp=<2500 2500 2500>
    componentLinearAccScale=<1 1 1>
    componentLinearVelScale=<1 1 1>
    gravityOverride {z=-980}
    physicsBodyDefinitions [{
      boundBone {boneName=dyn_back_chain_3}
      boxExtents=<24 24 24>
      constraintSetup {angularLimitsMax=<10 5 15>, angularLimitsMin=<-10 -15 -15>, angularTarget=<1 0 0>}
      localJointOffset=<4 0 0>
     }]
   }
   n102 = Anim Dynamics - Bone: dyn_R_side_chainMain_1!(SimulationSpace:AnimPhysSimSpaceType Enum, ComponentPose=@n101, Alpha=1) -> (Pose) {
    angularDampingOverride=1
    angularSpringConstant=3000
    bAngularSpring=true
    bOverrideAngularDamping=true
    bUseGravityOverride=true
    gravityScale=0.35
    simulationSpace=RootRelative
    alphaBoolBlend {blendOption=Linear}
    alphaScaleBiasClamp {bInterpResult=true}
    boundBone {boneName=dyn_R_side_chainMain_1}
    componentAppliedLinearAccClamp=<2500 2500 2500>
    componentLinearAccScale=<1 1 1>
    componentLinearVelScale=<1 1 1>
    gravityOverride {z=-980}
    physicsBodyDefinitions [{
      boundBone {boneName=dyn_R_side_chainMain_1}
      boxExtents=<30 30 30>
      constraintSetup {angularLimitsMax=<0 0 5>, angularLimitsMin=<0 -15 -5>, angularTarget=<1 0 0>}
      localJointOffset=<-4 0 0>
     }]
   }
   n103 = Anim Dynamics - Bone: dyn_R_side_chainMain_2!(SimulationSpace:AnimPhysSimSpaceType Enum, ComponentPose=@n102, Alpha=1) -> (Pose) {
    angularDampingOverride=1
    angularSpringConstant=3000
    bAngularSpring=true
    bOverrideAngularDamping=true
    bUseGravityOverride=true
    gravityScale=0.35
    simulationSpace=RootRelative
    alphaBoolBlend {blendOption=Linear}
    alphaScaleBiasClamp {bInterpResult=true}
    boundBone {boneName=dyn_R_side_chainMain_2}
    componentAppliedLinearAccClamp=<2500 2500 2500>
    componentLinearAccScale=<1 1 1>
    componentLinearVelScale=<1 1 1>
    gravityOverride {z=-980}
    physicsBodyDefinitions [{
      boundBone {boneName=dyn_R_side_chainMain_2}
      boxExtents=<30 30 30>
      constraintSetup {angularLimitsMax=<0 0 15>, angularLimitsMin=<0 -15 -5>, angularTarget=<1 0 0>}
      localJointOffset=<-4 0 0>
     }]
   }
   n104 = Anim Dynamics - Bone: dyn_R_sideChain_A_1!(SimulationSpace:AnimPhysSimSpaceType Enum, ComponentPose=@n103, Alpha=1) -> (Pose) {
    angularDampingOverride=1
    angularSpringConstant=3000
    bAngularSpring=true
    bOverrideAngularDamping=true
    bUseGravityOverride=true
    gravityScale=0.35
    simulationSpace=RootRelative
    alphaBoolBlend {blendOption=Linear}
    alphaScaleBiasClamp {bInterpResult=true}
    boundBone {boneName=dyn_R_sideChain_A_1}
    componentAppliedLinearAccClamp=<2500 2500 2500>
    componentLinearAccScale=<1 1 1>
    componentLinearVelScale=<1 1 1>
    gravityOverride {z=-980}
    physicsBodyDefinitions [{
      boundBone {boneName=dyn_R_sideChain_A_1}
      boxExtents=<30 30 30>
      constraintSetup {angularLimitsMax=<10 0 15>, angularLimitsMin=<-10 -15 -15>, angularTarget=<1 0 0>}
      localJointOffset=<-4 0 0>
     }]
   }
   n105 = Anim Dynamics - Bone: dyn_R_sideChain_A_2!(SimulationSpace:AnimPhysSimSpaceType Enum, ComponentPose=@n104, Alpha=1) -> (Pose) {
    angularDampingOverride=1
    angularSpringConstant=3000
    bAngularSpring=true
    bOverrideAngularDamping=true
    bUseGravityOverride=true
    gravityScale=0.35
    simulationSpace=RootRelative
    alphaBoolBlend {blendOption=Linear}
    alphaScaleBiasClamp {bInterpResult=true}
    boundBone {boneName=dyn_R_sideChain_A_2}
    componentAppliedLinearAccClamp=<2500 2500 2500>
    componentLinearAccScale=<1 1 1>
    componentLinearVelScale=<1 1 1>
    gravityOverride {z=-980}
    physicsBodyDefinitions [{
      boundBone {boneName=dyn_R_sideChain_A_2}
      boxExtents=<26 26 26>
      constraintSetup {angularLimitsMax=<10 3 15>, angularLimitsMin=<-10 -15 -15>, angularTarget=<1 0 0>}
      localJointOffset=<-4 0 0>
     }]
   }
   n106 = Anim Dynamics - Bone: dyn_R_sideChain_A_3!(SimulationSpace:AnimPhysSimSpaceType Enum, ComponentPose=@n105, Alpha=1) -> (Pose) {
    angularDampingOverride=1
    angularSpringConstant=3000
    bAngularSpring=true
    bOverrideAngularDamping=true
    bUseGravityOverride=true
    gravityScale=0.35
    simulationSpace=RootRelative
    alphaBoolBlend {blendOption=Linear}
    alphaScaleBiasClamp {bInterpResult=true}
    boundBone {boneName=dyn_R_sideChain_A_3}
    componentAppliedLinearAccClamp=<2500 2500 2500>
    componentLinearAccScale=<1 1 1>
    componentLinearVelScale=<1 1 1>
    gravityOverride {z=-980}
    physicsBodyDefinitions [{
      boundBone {boneName=dyn_R_sideChain_A_3}
      boxExtents=<24 24 24>
      constraintSetup {angularLimitsMax=<10 5 15>, angularLimitsMin=<-10 -15 -15>, angularTarget=<1 0 0>}
      localJointOffset=<-4 0 0>
     }]
   }
   n107 = Anim Dynamics - Bone: dyn_R_sideChain_A_4!(SimulationSpace:AnimPhysSimSpaceType Enum, ComponentPose=@n106, Alpha=1) -> (Pose) {
    angularDampingOverride=1
    angularSpringConstant=3000
    bAngularSpring=true
    bOverrideAngularDamping=true
    bUseGravityOverride=true
    gravityScale=0.35
    simulationSpace=RootRelative
    alphaBoolBlend {blendOption=Linear}
    alphaScaleBiasClamp {bInterpResult=true}
    boundBone {boneName=dyn_R_sideChain_A_4}
    componentAppliedLinearAccClamp=<2500 2500 2500>
    componentLinearAccScale=<1 1 1>
    componentLinearVelScale=<1 1 1>
    gravityOverride {z=-980}
    physicsBodyDefinitions [{
      boundBone {boneName=dyn_R_sideChain_A_4}
      boxExtents=<24 24 24>
      constraintSetup {angularLimitsMax=<10 5 15>, angularLimitsMin=<-10 -15 -15>, angularTarget=<1 0 0>}
      localJointOffset=<-4 0 0>
     }]
   }
   n108 = Anim Dynamics - Bone: dyn_R_sideChain_B_1!(SimulationSpace:AnimPhysSimSpaceType Enum, ComponentPose=@n107, Alpha=1) -> (Pose) {
    angularDampingOverride=1
    angularSpringConstant=3000
    bAngularSpring=true
    bOverrideAngularDamping=true
    bUseGravityOverride=true
    gravityScale=0.35
    simulationSpace=RootRelative
    alphaBoolBlend {blendOption=Linear}
    alphaScaleBiasClamp {bInterpResult=true}
    boundBone {boneName=dyn_R_sideChain_B_1}
    componentAppliedLinearAccClamp=<2500 2500 2500>
    componentLinearAccScale=<1 1 1>
    componentLinearVelScale=<1 1 1>
    gravityOverride {z=-980}
    physicsBodyDefinitions [{
      boundBone {boneName=dyn_R_sideChain_B_1}
      boxExtents=<30 30 30>
      constraintSetup {angularLimitsMax=<0 0 15>, angularLimitsMin=<0 -15 -15>, angularTarget=<1 0 0>}
      localJointOffset=<-4 0 0>
     }]
   }
   n109 = Anim Dynamics - Bone: dyn_R_sideChain_B_2!(SimulationSpace:AnimPhysSimSpaceType Enum, ComponentPose=@n108, Alpha=1) -> (Pose) {
    angularDampingOverride=1
    angularSpringConstant=3000
    bAngularSpring=true
    bOverrideAngularDamping=true
    bUseGravityOverride=true
    gravityScale=0.35
    simulationSpace=RootRelative
    alphaBoolBlend {blendOption=Linear}
    alphaScaleBiasClamp {bInterpResult=true}
    boundBone {boneName=dyn_R_sideChain_B_2}
    componentAppliedLinearAccClamp=<2500 2500 2500>
    componentLinearAccScale=<1 1 1>
    componentLinearVelScale=<1 1 1>
    gravityOverride {z=-980}
    physicsBodyDefinitions [{
      boundBone {boneName=dyn_R_sideChain_B_2}
      boxExtents=<26 26 26>
      constraintSetup {angularLimitsMax=<0 5 15>, angularLimitsMin=<0 -15 -15>, angularTarget=<1 0 0>}
      localJointOffset=<-4 0 0>
     }]
   }
   n110 = Anim Dynamics - Bone: dyn_R_sideChain_B_3!(SimulationSpace:AnimPhysSimSpaceType Enum, ComponentPose=@n109, Alpha=1) -> (Pose) {
    angularDampingOverride=1
    angularSpringConstant=3000
    bAngularSpring=true
    bOverrideAngularDamping=true
    bUseGravityOverride=true
    gravityScale=0.35
    simulationSpace=RootRelative
    alphaBoolBlend {blendOption=Linear}
    alphaScaleBiasClamp {bInterpResult=true}
    boundBone {boneName=dyn_R_sideChain_B_3}
    componentAppliedLinearAccClamp=<2500 2500 2500>
    componentLinearAccScale=<1 1 1>
    componentLinearVelScale=<1 1 1>
    gravityOverride {z=-980}
    physicsBodyDefinitions [{
      boundBone {boneName=dyn_R_sideChain_B_3}
      boxExtents=<24 24 24>
      constraintSetup {angularLimitsMax=<10 5 15>, angularLimitsMin=<-10 -15 -15>, angularTarget=<1 0 0>}
      localJointOffset=<-4 0 0>
     }]
   }
   n111 = Anim Dynamics - Bone: dyn_R_sideChain_B_4!(SimulationSpace:AnimPhysSimSpaceType Enum, ComponentPose=@n110, Alpha=1) -> (Pose) {
    angularDampingOverride=1
    angularSpringConstant=3000
    bAngularSpring=true
    bOverrideAngularDamping=true
    bUseGravityOverride=true
    gravityScale=0.35
    simulationSpace=RootRelative
    alphaBoolBlend {blendOption=Linear}
    alphaScaleBiasClamp {bInterpResult=true}
    boundBone {boneName=dyn_R_sideChain_B_4}
    componentAppliedLinearAccClamp=<2500 2500 2500>
    componentLinearAccScale=<1 1 1>
    componentLinearVelScale=<1 1 1>
    gravityOverride {z=-980}
    physicsBodyDefinitions [{
      boundBone {boneName=dyn_R_sideChain_B_4}
      boxExtents=<24 24 24>
      constraintSetup {angularLimitsMax=<10 5 15>, angularLimitsMin=<-10 -15 -15>, angularTarget=<1 0 0>}
      localJointOffset=<-4 0 0>
     }]
   }
   Save cached pose 'nonFrontendIdleCache'(Pose=Component To Local(ComponentPose=@n111)) {cacheName=nonFrontendIdleCache}
   n112 = Anim Dynamics - Bone: dyn_L_pantLeg_front_2(ComponentPose=Local To Component(LocalPose=Use cached pose 'setupCache'()), Alpha=1) {
    angularDampingOverride=1
    angularSpringConstant=6000
    bAngularSpring=true
    bOverrideAngularDamping=true
    bUseGravityOverride=true
    gravityScale=0.35
    simulationSpace=RootRelative
    alphaBoolBlend {blendOption=Linear}
    alphaScaleBiasClamp {scale=0.75}
    boundBone {boneName=dyn_L_pantLeg_front_2}
    componentAppliedLinearAccClamp=<2500 2500 2500>
    componentLinearAccScale=<1 1 1>
    componentLinearVelScale=<1 1 1>
    gravityOverride {z=-980}
    physicsBodyDefinitions [{
      boundBone {boneName=dyn_L_pantLeg_front_2}
      boxExtents=<25 25 25>
      constraintSetup {angularLimitsMax=<5 30 30>, angularLimitsMin=<-5 0 -30>, angularTarget=<1 0 0>}
      localJointOffset=<-4 0 0>
     }]
   }
   n113 = Anim Dynamics - Bone: dyn_L_pantLeg_inner_1(ComponentPose=@n112, Alpha=1) {
    angularDampingOverride=1
    angularSpringConstant=5000
    bAngularSpring=true
    bOverrideAngularDamping=true
    bUseGravityOverride=true
    gravityScale=0.35
    simulationSpace=RootRelative
    alphaBoolBlend {blendOption=Linear}
    boundBone {boneName=dyn_L_pantLeg_inner_1}
    componentAppliedLinearAccClamp=<2500 2500 2500>
    componentLinearAccScale=<1 1 1>
    componentLinearVelScale=<1 1 1>
    gravityOverride {z=-980}
    physicsBodyDefinitions [{
      boundBone {boneName=dyn_L_pantLeg_inner_1}
      boxExtents=<25 25 25>
      constraintSetup {angularLimitsMax=<5 10 10>, angularLimitsMin=<-5 -10 -35>, angularTarget=<1 0 0>}
      localJointOffset=<-4 0 0>
     }]
   }
   n114 = Anim Dynamics - Bone: dyn_L_pantLeg_inner_2(ComponentPose=@n113, Alpha=1) {
    angularDampingOverride=1
    angularSpringConstant=10000
    bAngularSpring=true
    bOverrideAngularDamping=true
    bUseGravityOverride=true
    gravityScale=0.35
    simulationSpace=RootRelative
    alphaBoolBlend {blendOption=Linear}
    boundBone {boneName=dyn_L_pantLeg_inner_2}
    componentAppliedLinearAccClamp=<2500 2500 2500>
    componentLinearAccScale=<1 1 1>
    componentLinearVelScale=<1 1 1>
    gravityOverride {z=-980}
    physicsBodyDefinitions [{
      boundBone {boneName=dyn_L_pantLeg_inner_2}
      boxExtents=<25 25 25>
      constraintSetup {angularLimitsMax=<5 10 10>, angularLimitsMin=<-5 -10 -35>, angularTarget=<1 0 0>}
      localJointOffset=<-4 0 0>
     }]
   }
   n115 = Anim Dynamics - Bone: dyn_L_pantLeg_outer_1(ComponentPose=@n114, Alpha=1) {
    angularDampingOverride=1
    angularSpringConstant=10000
    bAngularSpring=true
    bOverrideAngularDamping=true
    bUseGravityOverride=true
    gravityScale=0.35
    simulationSpace=RootRelative
    alphaBoolBlend {blendOption=Linear}
    boundBone {boneName=dyn_L_pantLeg_outer_1}
    componentAppliedLinearAccClamp=<2500 2500 2500>
    componentLinearAccScale=<1 1 1>
    componentLinearVelScale=<1 1 1>
    gravityOverride {z=-980}
    physicsBodyDefinitions [{
      boundBone {boneName=dyn_L_pantLeg_outer_1}
      boxExtents=<25 25 25>
      constraintSetup {angularLimitsMax=<5 30 10>, angularLimitsMin=<-5 -30 0>, angularTarget=<1 0 0>}
      localJointOffset=<-4 0 0>
     }]
   }
   n116 = Anim Dynamics - Bone: dyn_L_pantLeg_outer_2(ComponentPose=@n115, Alpha=1) {
    angularDampingOverride=1
    angularSpringConstant=10000
    bAngularSpring=true
    bOverrideAngularDamping=true
    bUseGravityOverride=true
    gravityScale=0.35
    simulationSpace=RootRelative
    alphaBoolBlend {blendOption=Linear}
    boundBone {boneName=dyn_L_pantLeg_outer_2}
    componentAppliedLinearAccClamp=<2500 2500 2500>
    componentLinearAccScale=<1 1 1>
    componentLinearVelScale=<1 1 1>
    gravityOverride {z=-980}
    physicsBodyDefinitions [{
      boundBone {boneName=dyn_L_pantLeg_outer_2}
      boxExtents=<25 25 25>
      constraintSetup {angularLimitsMax=<5 30 10>, angularLimitsMin=<-5 -30 0>, angularTarget=<1 0 0>}
      localJointOffset=<-4 0 0>
     }]
   }
   n117 = Anim Dynamics - Bone: dyn_L_pantLeg_back_1(ComponentPose=@n116, Alpha=1) {
    angularDampingOverride=1
    angularSpringConstant=10000
    bAngularSpring=true
    bOverrideAngularDamping=true
    bUseGravityOverride=true
    gravityScale=0.35
    simulationSpace=RootRelative
    alphaBoolBlend {blendOption=Linear}
    alphaScaleBiasClamp {scale=0.5}
    boundBone {boneName=dyn_L_pantLeg_back_1}
    componentAppliedLinearAccClamp=<2500 2500 2500>
    componentLinearAccScale=<1 1 1>
    componentLinearVelScale=<1 1 1>
    gravityOverride {z=-980}
    physicsBodyDefinitions [{
      boundBone {boneName=dyn_L_pantLeg_back_1}
      boxExtents=<25 25 25>
      constraintSetup {angularLimitsMax=<5 5 15>, angularLimitsMin=<-5 -5 0>, angularTarget=<1 0 0>}
      localJointOffset=<-4 0 0>
     }]
   }
   n118 = Anim Dynamics - Bone: dyn_L_pantLeg_back_2(ComponentPose=@n117, Alpha=1) {
    angularDampingOverride=1
    angularSpringConstant=10000
    bAngularSpring=true
    bOverrideAngularDamping=true
    bUseGravityOverride=true
    gravityScale=0.35
    simulationSpace=RootRelative
    alphaBoolBlend {blendOption=Linear}
    alphaScaleBiasClamp {scale=0.5}
    boundBone {boneName=dyn_L_pantLeg_back_2}
    componentAppliedLinearAccClamp=<2500 2500 2500>
    componentLinearAccScale=<1 1 1>
    componentLinearVelScale=<1 1 1>
    gravityOverride {z=-980}
    physicsBodyDefinitions [{
      boundBone {boneName=dyn_L_pantLeg_back_2}
      boxExtents=<25 25 25>
      constraintSetup {angularLimitsMax=<5 5 15>, angularLimitsMin=<-5 -5 0>, angularTarget=<1 0 0>}
      localJointOffset=<-4 0 0>
     }]
   }
   n119 = Anim Dynamics - Bone: dyn_R_pantLeg_front_2(ComponentPose=@n118, Alpha=1) {
    angularDampingOverride=1
    angularSpringConstant=6000
    bAngularSpring=true
    bOverrideAngularDamping=true
    bUseGravityOverride=true
    gravityScale=0.35
    simulationSpace=RootRelative
    alphaBoolBlend {blendOption=Linear}
    alphaScaleBiasClamp {scale=0.75}
    boundBone {boneName=dyn_R_pantLeg_front_2}
    componentAppliedLinearAccClamp=<2500 2500 2500>
    componentLinearAccScale=<1 1 1>
    componentLinearVelScale=<1 1 1>
    gravityOverride {z=-980}
    physicsBodyDefinitions [{
      boundBone {boneName=dyn_R_pantLeg_front_2}
      boxExtents=<23 23 23>
      constraintSetup {angularLimitsMax=<5 30 30>, angularLimitsMin=<-5 -5 -30>, angularTarget=<1 0 0>}
      localJointOffset=<4 0 0>
     }]
   }
   n120 = Anim Dynamics - Bone: dyn_R_pantLeg_inner_1(ComponentPose=@n119, Alpha=1) {
    angularDampingOverride=1
    angularSpringConstant=5000
    bAngularSpring=true
    bOverrideAngularDamping=true
    bUseGravityOverride=true
    gravityScale=0.35
    simulationSpace=RootRelative
    alphaBoolBlend {blendOption=Linear}
    boundBone {boneName=dyn_R_pantLeg_inner_1}
    componentAppliedLinearAccClamp=<2500 2500 2500>
    componentLinearAccScale=<1 1 1>
    componentLinearVelScale=<1 1 1>
    gravityOverride {z=-980}
    physicsBodyDefinitions [{
      boundBone {boneName=dyn_R_pantLeg_inner_1}
      boxExtents=<23 23 23>
      constraintSetup {angularLimitsMax=<5 10 10>, angularLimitsMin=<-5 -10 -35>, angularTarget=<1 0 0>}
      localJointOffset=<4 0 0>
     }]
   }
   n121 = Anim Dynamics - Bone: dyn_R_pantLeg_inner_2(ComponentPose=@n120, Alpha=1) {
    angularDampingOverride=1
    angularSpringConstant=10000
    bAngularSpring=true
    bOverrideAngularDamping=true
    bUseGravityOverride=true
    gravityScale=0.35
    simulationSpace=RootRelative
    alphaBoolBlend {blendOption=Linear}
    boundBone {boneName=dyn_R_pantLeg_inner_2}
    componentAppliedLinearAccClamp=<2500 2500 2500>
    componentLinearAccScale=<1 1 1>
    componentLinearVelScale=<1 1 1>
    gravityOverride {z=-980}
    physicsBodyDefinitions [{
      boundBone {boneName=dyn_R_pantLeg_inner_2}
      boxExtents=<23 23 23>
      constraintSetup {angularLimitsMax=<5 10 10>, angularLimitsMin=<-5 -10 -35>, angularTarget=<1 0 0>}
      localJointOffset=<4 0 0>
     }]
   }
   n122 = Anim Dynamics - Bone: dyn_R_pantLeg_outer_1(ComponentPose=@n121, Alpha=1) {
    angularDampingOverride=1
    angularSpringConstant=10000
    bAngularSpring=true
    bOverrideAngularDamping=true
    bUseGravityOverride=true
    gravityScale=0.35
    simulationSpace=RootRelative
    alphaBoolBlend {blendOption=Linear}
    boundBone {boneName=dyn_R_pantLeg_outer_1}
    componentAppliedLinearAccClamp=<2500 2500 2500>
    componentLinearAccScale=<1 1 1>
    componentLinearVelScale=<1 1 1>
    gravityOverride {z=-980}
    physicsBodyDefinitions [{
      boundBone {boneName=dyn_R_pantLeg_outer_1}
      boxExtents=<23 23 23>
      constraintSetup {angularLimitsMax=<5 30 0>, angularLimitsMin=<-5 -30 -10>, angularTarget=<1 0 0>}
      localJointOffset=<4 0 0>
     }]
   }
   n123 = Anim Dynamics - Bone: dyn_R_pantLeg_outer_2(ComponentPose=@n122, Alpha=1) {
    angularDampingOverride=1
    angularSpringConstant=10000
    bAngularSpring=true
    bOverrideAngularDamping=true
    bUseGravityOverride=true
    gravityScale=0.35
    simulationSpace=RootRelative
    alphaBoolBlend {blendOption=Linear}
    boundBone {boneName=dyn_R_pantLeg_outer_2}
    componentAppliedLinearAccClamp=<2500 2500 2500>
    componentLinearAccScale=<1 1 1>
    componentLinearVelScale=<1 1 1>
    gravityOverride {z=-980}
    physicsBodyDefinitions [{
      boundBone {boneName=dyn_R_pantLeg_outer_2}
      boxExtents=<23 23 23>
      constraintSetup {angularLimitsMax=<5 30 0>, angularLimitsMin=<-5 -30 -10>, angularTarget=<1 0 0>}
      localJointOffset=<4 0 0>
     }]
   }
   n124 = Anim Dynamics - Bone: dyn_R_pantLeg_back_1(ComponentPose=@n123, Alpha=1) {
    angularDampingOverride=1
    angularSpringConstant=10000
    bAngularSpring=true
    bOverrideAngularDamping=true
    bUseGravityOverride=true
    gravityScale=0.35
    simulationSpace=RootRelative
    alphaBoolBlend {blendOption=Linear}
    alphaScaleBiasClamp {scale=0.5}
    boundBone {boneName=dyn_R_pantLeg_back_1}
    componentAppliedLinearAccClamp=<2500 2500 2500>
    componentLinearAccScale=<1 1 1>
    componentLinearVelScale=<1 1 1>
    gravityOverride {z=-980}
    physicsBodyDefinitions [{
      boundBone {boneName=dyn_R_pantLeg_back_1}
      boxExtents=<23 23 23>
      constraintSetup {angularLimitsMax=<5 5 15>, angularLimitsMin=<-5 -5 0>, angularTarget=<1 0 0>}
      localJointOffset=<4 0 0>
     }]
   }
   n125 = Anim Dynamics - Bone: dyn_R_pantLeg_back_2(ComponentPose=@n124, Alpha=1) {
    angularDampingOverride=1
    angularSpringConstant=10000
    bAngularSpring=true
    bOverrideAngularDamping=true
    bUseGravityOverride=true
    gravityScale=0.35
    simulationSpace=RootRelative
    alphaBoolBlend {blendOption=Linear}
    alphaScaleBiasClamp {scale=0.5}
    boundBone {boneName=dyn_R_pantLeg_back_2}
    componentAppliedLinearAccClamp=<2500 2500 2500>
    componentLinearAccScale=<1 1 1>
    componentLinearVelScale=<1 1 1>
    gravityOverride {z=-980}
    physicsBodyDefinitions [{
      boundBone {boneName=dyn_R_pantLeg_back_2}
      boxExtents=<23 23 23>
      constraintSetup {angularLimitsMax=<5 5 15>, angularLimitsMin=<-5 -5 0>, angularTarget=<1 0 0>}
      localJointOffset=<4 0 0>
     }]
   }
   n126 = Anim Dynamics - Bone: dyn_L_front_chainMain_A_1(ComponentPose=@n125, Alpha=1) {
    angularDampingOverride=1
    angularSpringConstant=3000
    bAngularSpring=true
    bOverrideAngularDamping=true
    bUseGravityOverride=true
    bUsePlanarLimit=false
    gravityScale=0.35
    simulationSpace=RootRelative
    alphaBoolBlend {blendOption=Linear}
    alphaScaleBiasClamp {bInterpResult=true, scale=0.75}
    boundBone {boneName=dyn_L_front_chainMain_A_1}
    componentAppliedLinearAccClamp=<2500 2500 2500>
    componentLinearAccScale=<1 1 1>
    componentLinearVelScale=<1 1 1>
    gravityOverride {z=-980}
    physicsBodyDefinitions [{
      boundBone {boneName=dyn_L_front_chainMain_A_1}
      boxExtents=<30 30 30>
      constraintSetup {angularLimitsMax=<0 0 30>, angularLimitsMin=<0 -5 -10>, angularTarget=<1 0 0>}
      localJointOffset=<4 0 0>
     }]
   }
   n127 = Anim Dynamics - Bone: dyn_L_front_chainMain_A_2(ComponentPose=@n126, Alpha=1) {
    angularDampingOverride=1
    angularSpringConstant=3000
    bAngularSpring=true
    bOverrideAngularDamping=true
    bUseGravityOverride=true
    bUsePlanarLimit=false
    gravityScale=0.35
    simulationSpace=RootRelative
    alphaBoolBlend {blendOption=Linear}
    alphaScaleBiasClamp {bInterpResult=true, scale=0.75}
    boundBone {boneName=dyn_L_front_chainMain_A_2}
    componentAppliedLinearAccClamp=<2500 2500 2500>
    componentLinearAccScale=<1 1 1>
    componentLinearVelScale=<1 1 1>
    gravityOverride {z=-980}
    physicsBodyDefinitions [{
      boundBone {boneName=dyn_L_front_chainMain_A_2}
      boxExtents=<30 30 30>
      constraintSetup {angularLimitsMax=<0 0 10>, angularLimitsMin=<0 -5 -1>, angularTarget=<1 0 0>}
      localJointOffset=<4 0 0>
     }]
   }
   n128 = Anim Dynamics - Bone: dyn_L_frontChain_A_1(ComponentPose=@n127, Alpha=1) {
    angularDampingOverride=1
    angularSpringConstant=3000
    bAngularSpring=true
    bOverrideAngularDamping=true
    bUseGravityOverride=true
    bUsePlanarLimit=false
    gravityScale=0.35
    simulationSpace=RootRelative
    alphaBoolBlend {blendOption=Linear}
    alphaScaleBiasClamp {bInterpResult=true}
    boundBone {boneName=dyn_L_frontChain_A_1}
    componentAppliedLinearAccClamp=<2500 2500 2500>
    componentLinearAccScale=<1 1 1>
    componentLinearVelScale=<1 1 1>
    gravityOverride {z=-980}
    physicsBodyDefinitions [{
      boundBone {boneName=dyn_L_frontChain_A_1}
      boxExtents=<30 30 30>
      constraintSetup {angularLimitsMax=<0 0 5>, angularLimitsMin=<0 0 -5>, angularTarget=<1 0 0>}
      localJointOffset=<4 0 0>
     }]
   }
   n129 = Anim Dynamics - Bone: dyn_L_frontChain_A_2(ComponentPose=@n128, Alpha=1) {
    angularDampingOverride=1
    angularSpringConstant=3000
    bAngularSpring=true
    bOverrideAngularDamping=true
    bUseGravityOverride=true
    bUsePlanarLimit=false
    gravityScale=0.35
    simulationSpace=RootRelative
    alphaBoolBlend {blendOption=Linear}
    alphaScaleBiasClamp {bInterpResult=true}
    boundBone {boneName=dyn_L_frontChain_A_2}
    componentAppliedLinearAccClamp=<2500 2500 2500>
    componentLinearAccScale=<1 1 1>
    componentLinearVelScale=<1 1 1>
    gravityOverride {z=-980}
    physicsBodyDefinitions [{
      boundBone {boneName=dyn_L_frontChain_A_2}
      boxExtents=<26 26 26>
      constraintSetup {angularLimitsMax=<10 5 15>, angularLimitsMin=<-10 -15 -15>, angularTarget=<1 0 0>}
      localJointOffset=<4 0 0>
     }]
   }
   n130 = Anim Dynamics - Bone: dyn_L_frontChain_A_3(ComponentPose=@n129, Alpha=1) {
    angularDampingOverride=1
    angularSpringConstant=3000
    bAngularSpring=true
    bOverrideAngularDamping=true
    bUseGravityOverride=true
    bUsePlanarLimit=false
    gravityScale=0.35
    simulationSpace=RootRelative
    alphaBoolBlend {blendOption=Linear}
    alphaScaleBiasClamp {bInterpResult=true}
    boundBone {boneName=dyn_L_frontChain_A_3}
    componentAppliedLinearAccClamp=<2500 2500 2500>
    componentLinearAccScale=<1 1 1>
    componentLinearVelScale=<1 1 1>
    gravityOverride {z=-980}
    physicsBodyDefinitions [{
      boundBone {boneName=dyn_L_frontChain_A_3}
      boxExtents=<24 24 24>
      constraintSetup {angularLimitsMax=<10 0 15>, angularLimitsMin=<-10 -15 -15>, angularTarget=<1 0 0>}
      localJointOffset=<4 0 0>
     }]
   }
   n131 = Anim Dynamics - Bone: dyn_L_front_chainMain_B_1(ComponentPose=@n130, Alpha=1) {
    angularDampingOverride=1
    angularSpringConstant=3000
    bAngularSpring=true
    bOverrideAngularDamping=true
    bUseGravityOverride=true
    gravityScale=0.35
    simulationSpace=RootRelative
    alphaBoolBlend {blendOption=Linear}
    alphaScaleBiasClamp {bInterpResult=true}
    boundBone {boneName=dyn_L_front_chainMain_B_1}
    componentAppliedLinearAccClamp=<2500 2500 2500>
    componentLinearAccScale=<1 1 1>
    componentLinearVelScale=<1 1 1>
    gravityOverride {z=-980}
    physicsBodyDefinitions [{
      boundBone {boneName=dyn_L_front_chainMain_B_1}
      boxExtents=<30 30 30>
      constraintSetup {angularLimitsMax=<0 0 10>, angularLimitsMin=<0 -5 0>, angularTarget=<1 0 0>}
      localJointOffset=<4 0 0>
     }]
   }
   n132 = Anim Dynamics - Bone: dyn_L_front_chainMain_B_2(ComponentPose=@n131, Alpha=1) {
    angularDampingOverride=1
    angularSpringConstant=3000
    bAngularSpring=true
    bOverrideAngularDamping=true
    bUseGravityOverride=true
    gravityScale=0.35
    simulationSpace=RootRelative
    alphaBoolBlend {blendOption=Linear}
    alphaScaleBiasClamp {bInterpResult=true}
    boundBone {boneName=dyn_L_front_chainMain_B_2}
    componentAppliedLinearAccClamp=<2500 2500 2500>
    componentLinearAccScale=<1 1 1>
    componentLinearVelScale=<1 1 1>
    gravityOverride {z=-980}
    physicsBodyDefinitions [{
      boundBone {boneName=dyn_L_front_chainMain_B_2}
      boxExtents=<30 30 30>
      constraintSetup {angularLimitsMax=<0 0 0>, angularLimitsMin=<0 -15 0>, angularTarget=<1 0 0>}
      localJointOffset=<4 0 0>
     }]
   }
   n133 = Anim Dynamics - Bone: dyn_L_frontChain_B_1(ComponentPose=@n132, Alpha=1) {
    angularDampingOverride=1
    angularSpringConstant=3000
    bAngularSpring=true
    bOverrideAngularDamping=true
    bUseGravityOverride=true
    gravityScale=0.35
    simulationSpace=RootRelative
    alphaBoolBlend {blendOption=Linear}
    alphaScaleBiasClamp {bInterpResult=true}
    boundBone {boneName=dyn_L_frontChain_B_1}
    componentAppliedLinearAccClamp=<2500 2500 2500>
    componentLinearAccScale=<1 1 1>
    componentLinearVelScale=<1 1 1>
    gravityOverride {z=-980}
    physicsBodyDefinitions [{
      boundBone {boneName=dyn_L_frontChain_B_1}
      boxExtents=<30 30 30>
      constraintSetup {angularLimitsMax=<0 5 5>, angularLimitsMin=<0 -15 -5>, angularTarget=<1 0 0>}
      localJointOffset=<4 0 0>
     }]
   }
   n134 = Anim Dynamics - Bone: dyn_L_frontChain_B_2(ComponentPose=@n133, Alpha=1) {
    angularDampingOverride=1
    angularSpringConstant=3000
    bAngularSpring=true
    bOverrideAngularDamping=true
    bUseGravityOverride=true
    gravityScale=0.35
    simulationSpace=RootRelative
    alphaBoolBlend {blendOption=Linear}
    alphaScaleBiasClamp {bInterpResult=true}
    boundBone {boneName=dyn_L_frontChain_B_2}
    componentAppliedLinearAccClamp=<2500 2500 2500>
    componentLinearAccScale=<1 1 1>
    componentLinearVelScale=<1 1 1>
    gravityOverride {z=-980}
    physicsBodyDefinitions [{
      boundBone {boneName=dyn_L_frontChain_B_2}
      boxExtents=<26 26 26>
      constraintSetup {angularLimitsMax=<10 5 15>, angularLimitsMin=<-10 -15 -15>, angularTarget=<1 0 0>}
      localJointOffset=<4 0 0>
     }]
   }
   n135 = Anim Dynamics - Bone: dyn_L_frontChain_B_3(ComponentPose=@n134, Alpha=1) {
    angularDampingOverride=1
    angularSpringConstant=3000
    bAngularSpring=true
    bOverrideAngularDamping=true
    bUseGravityOverride=true
    gravityScale=0.35
    simulationSpace=RootRelative
    alphaBoolBlend {blendOption=Linear}
    alphaScaleBiasClamp {bInterpResult=true}
    boundBone {boneName=dyn_L_frontChain_B_3}
    componentAppliedLinearAccClamp=<2500 2500 2500>
    componentLinearAccScale=<1 1 1>
    componentLinearVelScale=<1 1 1>
    gravityOverride {z=-980}
    physicsBodyDefinitions [{
      boundBone {boneName=dyn_L_frontChain_B_3}
      boxExtents=<24 24 24>
      constraintSetup {angularLimitsMax=<10 0 30>, angularLimitsMin=<-10 -15 -30>, angularTarget=<1 0 0>}
      localJointOffset=<4 0 0>
     }]
   }
   n136 = Anim Dynamics - Bone: dyn_L_frontChain_B_4(ComponentPose=@n135, Alpha=1) {
    angularDampingOverride=1
    angularSpringConstant=3000
    bAngularSpring=true
    bOverrideAngularDamping=true
    bUseGravityOverride=true
    gravityScale=0.35
    simulationSpace=RootRelative
    alphaBoolBlend {blendOption=Linear}
    alphaScaleBiasClamp {bInterpResult=true}
    boundBone {boneName=dyn_L_frontChain_B_4}
    componentAppliedLinearAccClamp=<2500 2500 2500>
    componentLinearAccScale=<1 1 1>
    componentLinearVelScale=<1 1 1>
    gravityOverride {z=-980}
    physicsBodyDefinitions [{
      boundBone {boneName=dyn_L_frontChain_B_4}
      boxExtents=<24 24 24>
      constraintSetup {angularLimitsMax=<10 0 30>, angularLimitsMin=<-10 -15 -30>, angularTarget=<1 0 0>}
      localJointOffset=<4 0 0>
     }]
   }
   n137 = Anim Dynamics - Bone: dyn_L_side_chainMain_1(ComponentPose=@n136, Alpha=1) {
    angularDampingOverride=1
    angularSpringConstant=3000
    bAngularSpring=true
    bOverrideAngularDamping=true
    bUseGravityOverride=true
    gravityScale=0.35
    simulationSpace=RootRelative
    alphaBoolBlend {blendOption=Linear}
    alphaScaleBiasClamp {bInterpResult=true}
    boundBone {boneName=dyn_L_side_chainMain_1}
    componentAppliedLinearAccClamp=<2500 2500 2500>
    componentLinearAccScale=<1 1 1>
    componentLinearVelScale=<1 1 1>
    gravityOverride {z=-980}
    physicsBodyDefinitions [{
      boundBone {boneName=dyn_L_side_chainMain_1}
      boxExtents=<30 30 30>
      constraintSetup {angularLimitsMax=<0 0 5>, angularLimitsMin=<0 -15 -5>, angularTarget=<1 0 0>}
      localJointOffset=<4 0 0>
     }]
   }
   n138 = Anim Dynamics - Bone: dyn_L_side_chainMain_2(ComponentPose=@n137, Alpha=1) {
    angularDampingOverride=1
    angularSpringConstant=3000
    bAngularSpring=true
    bOverrideAngularDamping=true
    bUseGravityOverride=true
    gravityScale=0.35
    simulationSpace=RootRelative
    alphaBoolBlend {blendOption=Linear}
    alphaScaleBiasClamp {bInterpResult=true}
    boundBone {boneName=dyn_L_side_chainMain_2}
    componentAppliedLinearAccClamp=<2500 2500 2500>
    componentLinearAccScale=<1 1 1>
    componentLinearVelScale=<1 1 1>
    gravityOverride {z=-980}
    physicsBodyDefinitions [{
      boundBone {boneName=dyn_L_side_chainMain_2}
      boxExtents=<30 30 30>
      constraintSetup {angularLimitsMax=<0 0 5>, angularLimitsMin=<0 -15 -5>, angularTarget=<1 0 0>}
      localJointOffset=<4 0 0>
     }]
   }
   n139 = Anim Dynamics - Bone: dyn_L_sideChain_A_1(ComponentPose=@n138, Alpha=1) {
    angularDampingOverride=1
    angularSpringConstant=3000
    bAngularSpring=true
    bOverrideAngularDamping=true
    bUseGravityOverride=true
    gravityScale=0.35
    simulationSpace=RootRelative
    alphaBoolBlend {blendOption=Linear}
    alphaScaleBiasClamp {bInterpResult=true}
    boundBone {boneName=dyn_L_sideChain_A_1}
    componentAppliedLinearAccClamp=<2500 2500 2500>
    componentLinearAccScale=<1 1 1>
    componentLinearVelScale=<1 1 1>
    gravityOverride {z=-980}
    physicsBodyDefinitions [{
      boundBone {boneName=dyn_L_sideChain_A_1}
      boxExtents=<30 30 30>
      constraintSetup {angularLimitsMax=<10 5 15>, angularLimitsMin=<-10 -15 -15>, angularTarget=<1 0 0>}
      localJointOffset=<4 0 0>
     }]
   }
   n140 = Anim Dynamics - Bone: dyn_L_sideChain_A_2(ComponentPose=@n139, Alpha=1) {
    angularDampingOverride=1
    angularSpringConstant=3000
    bAngularSpring=true
    bOverrideAngularDamping=true
    bUseGravityOverride=true
    gravityScale=0.35
    simulationSpace=RootRelative
    alphaBoolBlend {blendOption=Linear}
    alphaScaleBiasClamp {bInterpResult=true}
    boundBone {boneName=dyn_L_sideChain_A_2}
    componentAppliedLinearAccClamp=<2500 2500 2500>
    componentLinearAccScale=<1 1 1>
    componentLinearVelScale=<1 1 1>
    gravityOverride {z=-980}
    physicsBodyDefinitions [{
      boundBone {boneName=dyn_L_sideChain_A_2}
      boxExtents=<26 26 26>
      constraintSetup {angularLimitsMax=<10 5 15>, angularLimitsMin=<-10 -15 -15>, angularTarget=<1 0 0>}
      localJointOffset=<4 0 0>
     }]
   }
   n141 = Anim Dynamics - Bone: dyn_L_sideChain_A_3(ComponentPose=@n140, Alpha=1) {
    angularDampingOverride=1
    angularSpringConstant=3000
    bAngularSpring=true
    bOverrideAngularDamping=true
    bUseGravityOverride=true
    gravityScale=0.35
    simulationSpace=RootRelative
    alphaBoolBlend {blendOption=Linear}
    alphaScaleBiasClamp {bInterpResult=true}
    boundBone {boneName=dyn_L_sideChain_A_3}
    componentAppliedLinearAccClamp=<2500 2500 2500>
    componentLinearAccScale=<1 1 1>
    componentLinearVelScale=<1 1 1>
    gravityOverride {z=-980}
    physicsBodyDefinitions [{
      boundBone {boneName=dyn_L_sideChain_A_3}
      boxExtents=<24 24 24>
      constraintSetup {angularLimitsMax=<10 5 15>, angularLimitsMin=<-10 -15 -15>, angularTarget=<1 0 0>}
      localJointOffset=<4 0 0>
     }]
   }
   n142 = Anim Dynamics - Bone: dyn_L_sideChain_A_4(ComponentPose=@n141, Alpha=1) {
    angularDampingOverride=1
    angularSpringConstant=3000
    bAngularSpring=true
    bOverrideAngularDamping=true
    bUseGravityOverride=true
    gravityScale=0.35
    simulationSpace=RootRelative
    alphaBoolBlend {blendOption=Linear}
    alphaScaleBiasClamp {bInterpResult=true}
    boundBone {boneName=dyn_L_sideChain_A_4}
    componentAppliedLinearAccClamp=<2500 2500 2500>
    componentLinearAccScale=<1 1 1>
    componentLinearVelScale=<1 1 1>
    gravityOverride {z=-980}
    physicsBodyDefinitions [{
      boundBone {boneName=dyn_L_sideChain_A_4}
      boxExtents=<24 24 24>
      constraintSetup {angularLimitsMax=<10 5 15>, angularLimitsMin=<-10 -15 -15>, angularTarget=<1 0 0>}
      localJointOffset=<4 0 0>
     }]
   }
   n143 = Anim Dynamics - Bone: dyn_L_sideChain_B_1(ComponentPose=@n142, Alpha=1) {
    angularDampingOverride=1
    angularSpringConstant=3000
    bAngularSpring=true
    bOverrideAngularDamping=true
    bUseGravityOverride=true
    gravityScale=0.35
    simulationSpace=RootRelative
    alphaBoolBlend {blendOption=Linear}
    alphaScaleBiasClamp {bInterpResult=true}
    boundBone {boneName=dyn_L_sideChain_B_1}
    componentAppliedLinearAccClamp=<2500 2500 2500>
    componentLinearAccScale=<1 1 1>
    componentLinearVelScale=<1 1 1>
    gravityOverride {z=-980}
    physicsBodyDefinitions [{
      boundBone {boneName=dyn_L_sideChain_B_1}
      boxExtents=<30 30 30>
      constraintSetup {angularLimitsMax=<0 0 15>, angularLimitsMin=<0 -15 -15>, angularTarget=<1 0 0>}
      localJointOffset=<4 0 0>
     }]
   }
   n144 = Anim Dynamics - Bone: dyn_L_sideChain_B_2(ComponentPose=@n143, Alpha=1) {
    angularDampingOverride=1
    angularSpringConstant=3000
    bAngularSpring=true
    bOverrideAngularDamping=true
    bUseGravityOverride=true
    gravityScale=0.35
    simulationSpace=RootRelative
    alphaBoolBlend {blendOption=Linear}
    alphaScaleBiasClamp {bInterpResult=true}
    boundBone {boneName=dyn_L_sideChain_B_2}
    componentAppliedLinearAccClamp=<2500 2500 2500>
    componentLinearAccScale=<1 1 1>
    componentLinearVelScale=<1 1 1>
    gravityOverride {z=-980}
    physicsBodyDefinitions [{
      boundBone {boneName=dyn_L_sideChain_B_2}
      boxExtents=<26 26 26>
      constraintSetup {angularLimitsMax=<0 5 15>, angularLimitsMin=<0 -15 -15>, angularTarget=<1 0 0>}
      localJointOffset=<4 0 0>
     }]
   }
   n145 = Anim Dynamics - Bone: dyn_L_sideChain_B_3(ComponentPose=@n144, Alpha=1) {
    angularDampingOverride=1
    angularSpringConstant=3000
    bAngularSpring=true
    bOverrideAngularDamping=true
    bUseGravityOverride=true
    gravityScale=0.35
    simulationSpace=RootRelative
    alphaBoolBlend {blendOption=Linear}
    alphaScaleBiasClamp {bInterpResult=true}
    boundBone {boneName=dyn_L_sideChain_B_3}
    componentAppliedLinearAccClamp=<2500 2500 2500>
    componentLinearAccScale=<1 1 1>
    componentLinearVelScale=<1 1 1>
    gravityOverride {z=-980}
    physicsBodyDefinitions [{
      boundBone {boneName=dyn_L_sideChain_B_3}
      boxExtents=<24 24 24>
      constraintSetup {angularLimitsMax=<10 5 15>, angularLimitsMin=<-10 -15 -15>, angularTarget=<1 0 0>}
      localJointOffset=<4 0 0>
     }]
   }
   n146 = Anim Dynamics - Bone: dyn_L_sideChain_B_4(ComponentPose=@n145, Alpha=1) {
    angularDampingOverride=1
    angularSpringConstant=3000
    bAngularSpring=true
    bOverrideAngularDamping=true
    bUseGravityOverride=true
    gravityScale=0.35
    simulationSpace=RootRelative
    alphaBoolBlend {blendOption=Linear}
    alphaScaleBiasClamp {bInterpResult=true}
    boundBone {boneName=dyn_L_sideChain_B_4}
    componentAppliedLinearAccClamp=<2500 2500 2500>
    componentLinearAccScale=<1 1 1>
    componentLinearVelScale=<1 1 1>
    gravityOverride {z=-980}
    physicsBodyDefinitions [{
      boundBone {boneName=dyn_L_sideChain_B_4}
      boxExtents=<24 24 24>
      constraintSetup {angularLimitsMax=<10 5 15>, angularLimitsMin=<-10 -15 -15>, angularTarget=<1 0 0>}
      localJointOffset=<4 0 0>
     }]
   }
   n147 = Anim Dynamics - Bone: dyn_back_chainMain_1(ComponentPose=@n146, Alpha=1) {
    angularDampingOverride=1
    angularSpringConstant=3000
    bAngularSpring=true
    bOverrideAngularDamping=true
    bUseGravityOverride=true
    gravityScale=0.35
    simulationSpace=RootRelative
    alphaBoolBlend {blendOption=Linear}
    alphaScaleBiasClamp {bInterpResult=true}
    boundBone {boneName=dyn_back_chainMain_1}
    componentAppliedLinearAccClamp=<2500 2500 2500>
    componentLinearAccScale=<1 1 1>
    componentLinearVelScale=<1 1 1>
    gravityOverride {z=-980}
    physicsBodyDefinitions [{
      boundBone {boneName=dyn_back_chainMain_1}
      boxExtents=<30 30 30>
      constraintSetup {angularLimitsMax=<0 0 5>, angularLimitsMin=<0 -15 -5>, angularTarget=<1 0 0>}
      localJointOffset=<4 0 0>
     }]
   }
   n148 = Anim Dynamics - Bone: dyn_back_chainMain_2(ComponentPose=@n147, Alpha=1) {
    angularDampingOverride=1
    angularSpringConstant=3000
    bAngularSpring=true
    bOverrideAngularDamping=true
    bUseGravityOverride=true
    gravityScale=0.35
    simulationSpace=RootRelative
    alphaBoolBlend {blendOption=Linear}
    alphaScaleBiasClamp {bInterpResult=true}
    boundBone {boneName=dyn_back_chainMain_2}
    componentAppliedLinearAccClamp=<2500 2500 2500>
    componentLinearAccScale=<1 1 1>
    componentLinearVelScale=<1 1 1>
    gravityOverride {z=-980}
    physicsBodyDefinitions [{
      boundBone {boneName=dyn_back_chainMain_2}
      boxExtents=<30 30 30>
      constraintSetup {angularLimitsMax=<0 0 15>, angularLimitsMin=<0 -15 -15>, angularTarget=<1 0 0>}
      localJointOffset=<4 0 0>
     }]
   }
   n149 = Anim Dynamics - Bone: dyn_back_chain_1(ComponentPose=@n148, Alpha=1) {
    angularDampingOverride=1
    angularSpringConstant=3000
    bAngularSpring=true
    bOverrideAngularDamping=true
    bUseGravityOverride=true
    gravityScale=0.35
    simulationSpace=RootRelative
    alphaBoolBlend {blendOption=Linear}
    alphaScaleBiasClamp {bInterpResult=true}
    boundBone {boneName=dyn_back_chain_1}
    componentAppliedLinearAccClamp=<2500 2500 2500>
    componentLinearAccScale=<1 1 1>
    componentLinearVelScale=<1 1 1>
    gravityOverride {z=-980}
    physicsBodyDefinitions [{
      boundBone {boneName=dyn_back_chain_1}
      boxExtents=<30 30 30>
      constraintSetup {angularLimitsMax=<10 5 15>, angularLimitsMin=<-10 -15 -15>, angularTarget=<1 0 0>}
      localJointOffset=<4 0 0>
     }]
   }
   n150 = Anim Dynamics - Bone: dyn_back_chain_2(ComponentPose=@n149, Alpha=1) {
    angularDampingOverride=1
    angularSpringConstant=3000
    bAngularSpring=true
    bOverrideAngularDamping=true
    bUseGravityOverride=true
    gravityScale=0.35
    simulationSpace=RootRelative
    alphaBoolBlend {blendOption=Linear}
    alphaScaleBiasClamp {bInterpResult=true}
    boundBone {boneName=dyn_back_chain_2}
    componentAppliedLinearAccClamp=<2500 2500 2500>
    componentLinearAccScale=<1 1 1>
    componentLinearVelScale=<1 1 1>
    gravityOverride {z=-980}
    physicsBodyDefinitions [{
      boundBone {boneName=dyn_back_chain_2}
      boxExtents=<26 26 26>
      constraintSetup {angularLimitsMax=<10 5 15>, angularLimitsMin=<-10 -15 -15>, angularTarget=<1 0 0>}
      localJointOffset=<4 0 0>
     }]
   }
   n151 = Anim Dynamics - Bone: dyn_back_chain_3(ComponentPose=@n150, Alpha=1) {
    angularDampingOverride=1
    angularSpringConstant=3000
    bAngularSpring=true
    bOverrideAngularDamping=true
    bUseGravityOverride=true
    gravityScale=0.35
    simulationSpace=RootRelative
    alphaBoolBlend {blendOption=Linear}
    alphaScaleBiasClamp {bInterpResult=true}
    boundBone {boneName=dyn_back_chain_3}
    componentAppliedLinearAccClamp=<2500 2500 2500>
    componentLinearAccScale=<1 1 1>
    componentLinearVelScale=<1 1 1>
    gravityOverride {z=-980}
    physicsBodyDefinitions [{
      boundBone {boneName=dyn_back_chain_3}
      boxExtents=<24 24 24>
      constraintSetup {angularLimitsMax=<10 5 15>, angularLimitsMin=<-10 -15 -15>, angularTarget=<1 0 0>}
      localJointOffset=<4 0 0>
     }]
   }
   n152 = Anim Dynamics - Bone: dyn_R_side_chainMain_1(ComponentPose=@n151, Alpha=1) {
    angularDampingOverride=1
    angularSpringConstant=3000
    bAngularSpring=true
    bOverrideAngularDamping=true
    bUseGravityOverride=true
    gravityScale=0.35
    simulationSpace=RootRelative
    alphaBoolBlend {blendOption=Linear}
    alphaScaleBiasClamp {bInterpResult=true}
    boundBone {boneName=dyn_R_side_chainMain_1}
    componentAppliedLinearAccClamp=<2500 2500 2500>
    componentLinearAccScale=<1 1 1>
    componentLinearVelScale=<1 1 1>
    gravityOverride {z=-980}
    physicsBodyDefinitions [{
      boundBone {boneName=dyn_R_side_chainMain_1}
      boxExtents=<30 30 30>
      constraintSetup {angularLimitsMax=<0 0 5>, angularLimitsMin=<0 -15 -5>, angularTarget=<1 0 0>}
      localJointOffset=<-4 0 0>
     }]
   }
   n153 = Anim Dynamics - Bone: dyn_R_side_chainMain_2(ComponentPose=@n152, Alpha=1) {
    angularDampingOverride=1
    angularSpringConstant=3000
    bAngularSpring=true
    bOverrideAngularDamping=true
    bUseGravityOverride=true
    gravityScale=0.35
    simulationSpace=RootRelative
    alphaBoolBlend {blendOption=Linear}
    alphaScaleBiasClamp {bInterpResult=true}
    boundBone {boneName=dyn_R_side_chainMain_2}
    componentAppliedLinearAccClamp=<2500 2500 2500>
    componentLinearAccScale=<1 1 1>
    componentLinearVelScale=<1 1 1>
    gravityOverride {z=-980}
    physicsBodyDefinitions [{
      boundBone {boneName=dyn_R_side_chainMain_2}
      boxExtents=<30 30 30>
      constraintSetup {angularLimitsMax=<0 0 15>, angularLimitsMin=<0 -15 -5>, angularTarget=<1 0 0>}
      localJointOffset=<-4 0 0>
     }]
   }
   n154 = Anim Dynamics - Bone: dyn_R_sideChain_A_1(ComponentPose=@n153, Alpha=1) {
    angularDampingOverride=1
    angularSpringConstant=3000
    bAngularSpring=true
    bOverrideAngularDamping=true
    bUseGravityOverride=true
    gravityScale=0.35
    simulationSpace=RootRelative
    alphaBoolBlend {blendOption=Linear}
    alphaScaleBiasClamp {bInterpResult=true}
    boundBone {boneName=dyn_R_sideChain_A_1}
    componentAppliedLinearAccClamp=<2500 2500 2500>
    componentLinearAccScale=<1 1 1>
    componentLinearVelScale=<1 1 1>
    gravityOverride {z=-980}
    physicsBodyDefinitions [{
      boundBone {boneName=dyn_R_sideChain_A_1}
      boxExtents=<30 30 30>
      constraintSetup {angularLimitsMax=<10 0 15>, angularLimitsMin=<-10 -15 -15>, angularTarget=<1 0 0>}
      localJointOffset=<-4 0 0>
     }]
   }
   n155 = Anim Dynamics - Bone: dyn_R_sideChain_A_2(ComponentPose=@n154, Alpha=1) {
    angularDampingOverride=1
    angularSpringConstant=3000
    bAngularSpring=true
    bOverrideAngularDamping=true
    bUseGravityOverride=true
    gravityScale=0.35
    simulationSpace=RootRelative
    alphaBoolBlend {blendOption=Linear}
    alphaScaleBiasClamp {bInterpResult=true}
    boundBone {boneName=dyn_R_sideChain_A_2}
    componentAppliedLinearAccClamp=<2500 2500 2500>
    componentLinearAccScale=<1 1 1>
    componentLinearVelScale=<1 1 1>
    gravityOverride {z=-980}
    physicsBodyDefinitions [{
      boundBone {boneName=dyn_R_sideChain_A_2}
      boxExtents=<26 26 26>
      constraintSetup {angularLimitsMax=<10 3 15>, angularLimitsMin=<-10 -15 -15>, angularTarget=<1 0 0>}
      localJointOffset=<-4 0 0>
     }]
   }
   n156 = Anim Dynamics - Bone: dyn_R_sideChain_A_3(ComponentPose=@n155, Alpha=1) {
    angularDampingOverride=1
    angularSpringConstant=3000
    bAngularSpring=true
    bOverrideAngularDamping=true
    bUseGravityOverride=true
    gravityScale=0.35
    simulationSpace=RootRelative
    alphaBoolBlend {blendOption=Linear}
    alphaScaleBiasClamp {bInterpResult=true}
    boundBone {boneName=dyn_R_sideChain_A_3}
    componentAppliedLinearAccClamp=<2500 2500 2500>
    componentLinearAccScale=<1 1 1>
    componentLinearVelScale=<1 1 1>
    gravityOverride {z=-980}
    physicsBodyDefinitions [{
      boundBone {boneName=dyn_R_sideChain_A_3}
      boxExtents=<24 24 24>
      constraintSetup {angularLimitsMax=<10 5 15>, angularLimitsMin=<-10 -15 -15>, angularTarget=<1 0 0>}
      localJointOffset=<-4 0 0>
     }]
   }
   n157 = Anim Dynamics - Bone: dyn_R_sideChain_A_4(ComponentPose=@n156, Alpha=1) {
    angularDampingOverride=1
    angularSpringConstant=3000
    bAngularSpring=true
    bOverrideAngularDamping=true
    bUseGravityOverride=true
    gravityScale=0.35
    simulationSpace=RootRelative
    alphaBoolBlend {blendOption=Linear}
    alphaScaleBiasClamp {bInterpResult=true}
    boundBone {boneName=dyn_R_sideChain_A_4}
    componentAppliedLinearAccClamp=<2500 2500 2500>
    componentLinearAccScale=<1 1 1>
    componentLinearVelScale=<1 1 1>
    gravityOverride {z=-980}
    physicsBodyDefinitions [{
      boundBone {boneName=dyn_R_sideChain_A_4}
      boxExtents=<24 24 24>
      constraintSetup {angularLimitsMax=<10 5 15>, angularLimitsMin=<-10 -15 -15>, angularTarget=<1 0 0>}
      localJointOffset=<-4 0 0>
     }]
   }
   n158 = Anim Dynamics - Bone: dyn_R_sideChain_B_1(ComponentPose=@n157, Alpha=1) {
    angularDampingOverride=1
    angularSpringConstant=3000
    bAngularSpring=true
    bOverrideAngularDamping=true
    bUseGravityOverride=true
    gravityScale=0.35
    simulationSpace=RootRelative
    alphaBoolBlend {blendOption=Linear}
    alphaScaleBiasClamp {bInterpResult=true}
    boundBone {boneName=dyn_R_sideChain_B_1}
    componentAppliedLinearAccClamp=<2500 2500 2500>
    componentLinearAccScale=<1 1 1>
    componentLinearVelScale=<1 1 1>
    gravityOverride {z=-980}
    physicsBodyDefinitions [{
      boundBone {boneName=dyn_R_sideChain_B_1}
      boxExtents=<30 30 30>
      constraintSetup {angularLimitsMax=<0 0 15>, angularLimitsMin=<0 -15 -15>, angularTarget=<1 0 0>}
      localJointOffset=<-4 0 0>
     }]
   }
   n159 = Anim Dynamics - Bone: dyn_R_sideChain_B_2(ComponentPose=@n158, Alpha=1) {
    angularDampingOverride=1
    angularSpringConstant=3000
    bAngularSpring=true
    bOverrideAngularDamping=true
    bUseGravityOverride=true
    gravityScale=0.35
    simulationSpace=RootRelative
    alphaBoolBlend {blendOption=Linear}
    alphaScaleBiasClamp {bInterpResult=true}
    boundBone {boneName=dyn_R_sideChain_B_2}
    componentAppliedLinearAccClamp=<2500 2500 2500>
    componentLinearAccScale=<1 1 1>
    componentLinearVelScale=<1 1 1>
    gravityOverride {z=-980}
    physicsBodyDefinitions [{
      boundBone {boneName=dyn_R_sideChain_B_2}
      boxExtents=<26 26 26>
      constraintSetup {angularLimitsMax=<0 5 15>, angularLimitsMin=<0 -15 -15>, angularTarget=<1 0 0>}
      localJointOffset=<-4 0 0>
     }]
   }
   n160 = Anim Dynamics - Bone: dyn_R_sideChain_B_3(ComponentPose=@n159, Alpha=1) {
    angularDampingOverride=1
    angularSpringConstant=3000
    bAngularSpring=true
    bOverrideAngularDamping=true
    bUseGravityOverride=true
    gravityScale=0.35
    simulationSpace=RootRelative
    alphaBoolBlend {blendOption=Linear}
    alphaScaleBiasClamp {bInterpResult=true}
    boundBone {boneName=dyn_R_sideChain_B_3}
    componentAppliedLinearAccClamp=<2500 2500 2500>
    componentLinearAccScale=<1 1 1>
    componentLinearVelScale=<1 1 1>
    gravityOverride {z=-980}
    physicsBodyDefinitions [{
      boundBone {boneName=dyn_R_sideChain_B_3}
      boxExtents=<24 24 24>
      constraintSetup {angularLimitsMax=<10 5 15>, angularLimitsMin=<-10 -15 -15>, angularTarget=<1 0 0>}
      localJointOffset=<-4 0 0>
     }]
   }
   n161 = Anim Dynamics - Bone: dyn_R_sideChain_B_4(ComponentPose=@n160, Alpha=1) {
    angularDampingOverride=1
    angularSpringConstant=3000
    bAngularSpring=true
    bOverrideAngularDamping=true
    bUseGravityOverride=true
    gravityScale=0.35
    simulationSpace=RootRelative
    alphaBoolBlend {blendOption=Linear}
    alphaScaleBiasClamp {bInterpResult=true}
    boundBone {boneName=dyn_R_sideChain_B_4}
    componentAppliedLinearAccClamp=<2500 2500 2500>
    componentLinearAccScale=<1 1 1>
    componentLinearVelScale=<1 1 1>
    gravityOverride {z=-980}
    physicsBodyDefinitions [{
      boundBone {boneName=dyn_R_sideChain_B_4}
      boxExtents=<24 24 24>
      constraintSetup {angularLimitsMax=<10 5 15>, angularLimitsMin=<-10 -15 -15>, angularTarget=<1 0 0>}
      localJointOffset=<-4 0 0>
     }]
   }
   n162 = "Transform (Modify) Bone - Bone: dyn_L_front_chainMain_A_1"!(ComponentPose=@n161, Alpha=1, Rotation="0,0,10") -> (Pose) {
    rotationMode=BMM_Additive
    rotationSpace=BCS_BoneSpace
    alphaBoolBlend {blendInTime=0.5, blendOption=Linear, blendOutTime=0.5}
    boneToModify {boneName=dyn_L_front_chainMain_A_1}
    rotation {roll=10}
   }
   Save cached pose 'FrontendIdleCache'(Pose=Component To Local(ComponentPose=@n162)) {cacheName=FrontendIdleCache}
  ]}
 EventGraph {nodes [
   Event BlueprintUpdateAnimation()
   n1 = BreakVector(InVec=Get PawnVelocity())
   n2 = "float > float"(A="Absolute (Float)"(A=@n1.Z), B=10)
   n3 = "float > float"(A="Absolute (Float)"(A=@n1), B=10)
   n4 = "float > float"(A="Absolute (Float)"(A=@n1.Y), B=10)
   > Set isAccelerating(isAccelerating:double="To Float (Boolean)"(InBool=OR Boolean!(A=@n3, B=@n4, C:bool=@n2) -> (ReturnValue))) -> (Output_Get:double)
   n5 = Get bIsPlayingEmote()
   > n6 = Set frontendIdleAlphaBool(frontendIdleAlphaBool:bool=AND Boolean(A=Get bIsInFrontEnd(), B=NOT Boolean(A=@n5))) -> (Output_Get:bool)
   > Set frontendIdleAlpha(frontendIdleAlpha:float="To Float (Boolean)"(InBool=@n6)) -> (Output_Get:float)
   n7 = "float < float"(A="Absolute (Float)"(A=@n1), B=1000)
   n8 = "float < float"(A="Absolute (Float)"(A=@n1.Y), B=1000)
   n9 = "float < float"(A="Absolute (Float)"(A=@n1.Z), B=1000)
   > Set highVelocityTurnOffAnimDyn(highVelocityTurnOffAnimDyn:double="To Float (Boolean)"(InBool=AND Boolean!(A=@n7, B=@n8, C:bool=@n9) -> (ReturnValue))) -> (Output_Get:double)
   n10 = Get bIsInFrontEnd()
   n11 = Get PawnRotationRate() -> (PawnRotationRate:Rotator)
   n12 = "float >= float"(A="Absolute (Float)"(A=BreakRotator(InRot=@n11).Yaw), B=1.5)
   n13 = Get bIsInFrontEnd()
   > Set frontendAnimDynAlpha(frontendAnimDynAlpha:double=SelectFloat(A=0.5, B=SelectFloat(A=0.7, B=1, bPickA=@n10), bPickA=AND Boolean!(A=@n13, B=@n12) -> (ReturnValue))) -> (Output_Get:double)
   > n14 = Branch(Condition=OR Boolean!(A=Get bIsInFrontEnd(), B=Get bIsDBNOCarried()) -> (ReturnValue))
   > Set simSpaceChange(simSpaceChange:AnimPhysSimSpaceType Enum=World) -> (Output_Get:AnimPhysSimSpaceType Enum)
   Set simSpaceChange(execute=@n14.else)
   n15 = Event OnWeaponChanged()
   n16 = Get Dynamic Material Instance(execute=[@n15.then, @n22.then_1], SkeletalMeshComponent=GetPartSkeletalMeshComponent())
   n17 = PickaxeItemDefinitionHasAnyMetaTag(Tags=Make Array("[0]"=(TagName="Cosmetics.MetaTags.MoonSatellite")), bRequiresEquipped=true)
   > n18 = Branch(Condition=@n17)
   > n19 = Branch(Condition=Get bIsSkyDiving())
   SetScalarParameterValue(execute=[@n20.then, @n21.else, @n19.then, @n18.else], self=@n16, ParameterName=Hide Element 02, Value=0)
   SetScalarParameterValue(execute=[@n19.else, @n21.then], self=@n16, ParameterName=Hide Element 02, Value=1)
   n20 = Event OnBeginSkydiving()
   Event OnEndSkydiving()
   > n21 = Branch(Condition=@n17)
   Event OnCharacterCustomizationLoaded()
   > n22 = Sequence()
   > Set Is Backbling A Cape(Is Backbling A Cape:bool=CheckForBackbling().bIsBackblingACape) -> (Output_Get:bool)
  ]}
]
variables [
 frontendAnimDynAlpha {type=number}
 frontendIdleAlpha {type=number}
 frontendIdleAlphaBool {type=boolean}
 highVelocityTurnOffAnimDyn {type=number}
 Is Backbling A Cape {type=boolean}
 isAccelerating {type=number}
 simSpaceChange {type=AnimPhysSimSpaceType}
]
