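# Physics asset of cape colliders for the MoonSatellite medium male player body, by its name giving collision volumes for cape motion.
# Capsules cover pelvis, spine and clavicles to neck, with a head sphere; arm and leg bodies have no shapes.
# Capsules also line pelvis-rooted dynamic bone chains at left, right and back, each side chain forking into A and B branches.
# Constraints link bodies along these chains with swing and twist free.
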PhysicsAsset /BRCosmetics/Characters/Player/Male/Medium/Bodies/M_MED_MoonSatellite/Meshes/M_MED_MoonSatellite_CapeColliders
bodies {
 calf_l []
 calf_r []
 clavicle_l [{shapeType=Capsule, length=9.0192, radius=9.8871, center=<9.4266 -1.1806 -2.4977>, rotation=<rot 69.3297 1.4476 176.9462>}]
 clavicle_r [{shapeType=Capsule, length=9.0192, radius=9.8871, center=<-9.4266 1.1806 2.4977>, rotation=<rot 69.3297 1.4476 -3.0538>}]
 dyn_L_sideChain_A_1 [{shapeType=Capsule, length=8, radius=8.7288, shapeName=dyn_L_sideChain_A_1_capsule, center=<2.9962 -1.6125 0>, rotation=<rot 90 0 0>}]
 dyn_L_sideChain_A_2 [{shapeType=Capsule, length=8, radius=10.4832, shapeName=dyn_L_sideChain_A_2_capsule, center=<2.9455 -1.7162 0>, rotation=<rot 90 0 0>}]
 dyn_L_sideChain_A_3 [{shapeType=Capsule, length=8, radius=11.4832, shapeName=dyn_L_sideChain_A_3_capsule, center=<2.5807 -2.2272 0>, rotation=<rot 90 0 0>}]
 dyn_L_sideChain_A_4 [{shapeType=Capsule, length=8, radius=11.4832, shapeName=dyn_L_sideChain_A_4_capsule, center=<0.6201 -3.3173 0>, rotation=<rot 90 0 0>}]
 dyn_L_sideChain_B_1 [{shapeType=Capsule, length=8, radius=8.7288, shapeName=dyn_L_sideChain_B_1_capsule, center=<3.0304 -1.2734 0>, rotation=<rot 90 0 0>}]
 dyn_L_sideChain_B_2 [{shapeType=Capsule, length=8, radius=10.4832, shapeName=dyn_L_sideChain_B_2_capsule, center=<3.0964 -0.8595 0>, rotation=<rot 90 0 0>}]
 dyn_L_sideChain_B_3 [{shapeType=Capsule, length=8, radius=11.4832, shapeName=dyn_L_sideChain_B_3_capsule, center=<2.8947 -1.6662 0>, rotation=<rot 90 0 0>}]
 dyn_L_sideChain_B_4 [{shapeType=Capsule, length=8, radius=11.4832, shapeName=dyn_L_sideChain_B_4_capsule, center=<0.6855 -2.748 0>, rotation=<rot 90 0 0>}]
 dyn_L_side_chainMain_1 [{shapeType=Capsule, length=8, radius=8.7288, shapeName=dyn_L_side_chainMain_1_capsule, center=<2.9989 -1.6049 0>, rotation=<rot 90 0 0>}]
 dyn_L_side_chainMain_2 [{shapeType=Capsule, length=8, radius=8.7288, shapeName=dyn_L_side_chainMain_2_capsule, center=<3.1036 -1.0858 0>, rotation=<rot 90 0 0>}]
 dyn_R_sideChain_A_1 [{shapeType=Capsule, length=8, radius=8.7288, shapeName=dyn_R_sideChain_A_1_capsule, center=<-2.9962 1.6125 0>, rotation=<rot 90 0 0>}]
 dyn_R_sideChain_A_2 [{shapeType=Capsule, length=8, radius=10.4832, shapeName=dyn_R_sideChain_A_2_capsule, center=<-2.9455 1.7162 0>, rotation=<rot 90 0 0>}]
 dyn_R_sideChain_A_3 [{shapeType=Capsule, length=8, radius=11.4832, shapeName=dyn_R_sideChain_A_3_capsule, center=<-2.5807 2.2272 0>, rotation=<rot 90 0 0>}]
 dyn_R_sideChain_A_4 [{shapeType=Capsule, length=8, radius=11.4832, shapeName=dyn_R_sideChain_A_4_capsule, center=<-0.6201 3.3173 0>, rotation=<rot 90 0 0>}]
 dyn_R_sideChain_B_1 [{shapeType=Capsule, length=8, radius=8.7288, shapeName=dyn_R_sideChain_B_1_capsule, center=<-3.0304 1.2734 0>, rotation=<rot 90 0 0>}]
 dyn_R_sideChain_B_2 [{shapeType=Capsule, length=8, radius=10.4832, shapeName=dyn_R_sideChain_B_2_capsule, center=<-3.0964 0.8595 0>, rotation=<rot 90 0 0>}]
 dyn_R_sideChain_B_3 [{shapeType=Capsule, length=8, radius=11.4832, shapeName=dyn_R_sideChain_B_3_capsule, center=<-2.8947 1.6662 0>, rotation=<rot 90 0 0>}]
 dyn_R_sideChain_B_4 [{shapeType=Capsule, length=8, radius=11.4832, shapeName=dyn_R_sideChain_B_4_capsule, center=<-0.6855 2.748 0>, rotation=<rot 90 0 0>}]
 dyn_R_side_chainMain_1 [{shapeType=Capsule, length=8, radius=8.7288, shapeName=dyn_R_side_chainMain_1_capsule, center=<-2.9989 1.6049 0>, rotation=<rot 90 0 0>}]
 dyn_R_side_chainMain_2 [{shapeType=Capsule, length=8, radius=8.7288, shapeName=dyn_R_side_chainMain_2_capsule, center=<-3.1036 1.0858 0>, rotation=<rot 90 0 0>}]
 dyn_back_chainMain_1 [{shapeType=Capsule, length=8, radius=8.7288, shapeName=dyn_back_chainMain_1_capsule, center=<3.1166 0.134 0>, rotation=<rot 90 0 0>}]
 dyn_back_chainMain_2 [{shapeType=Capsule, length=8, radius=8.7288, shapeName=dyn_back_chainMain_2_capsule, center=<3.1099 0.0222 0>, rotation=<rot 90 0 0>}]
 dyn_back_chain_1 [{shapeType=Capsule, length=8, radius=9.4832, shapeName=dyn_back_chain_1_capsule, center=<3.1418 -0.2645 0>, rotation=<rot 90 0 0>}]
 dyn_back_chain_2 [{shapeType=Capsule, length=8, radius=10.4832, shapeName=dyn_back_chain_2_capsule, center=<3.172 -0.1808 0>, rotation=<rot 90 0 0>}]
 dyn_back_chain_3 [{shapeType=Capsule, length=8, radius=11.4832, shapeName=dyn_back_chain_3_capsule, center=<2.8863 1.3688 0>, rotation=<rot 90 0 0>}]
 foot_l []
 foot_r []
 head [{shapeType=Sphere, radius=17.3125, center=<11.028 -0.174 0>}]
 lowerarm_l []
 lowerarm_r []
 neck_02 [{shapeType=Capsule, length=13.4375, radius=7.3175, center=<-0.2642 -0.3408 0>, rotation=<rot 90 -171.1193 -175.3978>}]
 pelvis [{shapeType=Capsule, length=8.6456, radius=14.5388, center=<-2.4887 0.1881 0>}]
 spine_01 [{shapeType=Capsule, length=7.9438, radius=13.5586, center=<-1.7965 0.1884 0>}]
 spine_02 [
  {shapeType=Capsule, length=7.3264, radius=13.6322, center=<-1.711 -0.1851 0.0001>}
  {shapeType=Capsule, length=29.3956, radius=11.5807, center=<2.5024 0.4992 3.9825>, rotation=<rot 90 0.0049 0.0049>}
  {shapeType=Capsule, length=29.875, radius=11.1013, center=<2.4704 0.4991 -3.9825>, rotation=<rot 90 -0.0396 -0.0396>}
 ]
 spine_03 [{shapeType=Capsule, length=8.2107, radius=13.7031, center=<0.1445 0.2396 0.0001>}]
 spine_04 [{shapeType=Capsule, length=9.4404, radius=14.1325, center=<2.766 0.7765 0.0002>}]
 spine_05 [
  {shapeType=Capsule, length=8.0226, radius=15.695, center=<-2.2795 1.8109 0>}
  {shapeType=Capsule, length=13.9518, radius=13.9425, center=<-0.2916 0.4452 0>}
 ]
 thigh_l []
 thigh_r []
 upperarm_l []
 upperarm_r []
}
constraints [
 {
  bone1Name=spine_01
  bone2Name=pelvis
  swing1LimitDegrees=45
  swing1Motion=Free
  swing2LimitDegrees=45
  swing2Motion=Free
  twistLimitDegrees=45
  twistMotion=Free
 }
 {
  bone1Name=spine_02
  bone2Name=spine_01
  swing1LimitDegrees=45
  swing1Motion=Free
  swing2LimitDegrees=45
  swing2Motion=Free
  twistLimitDegrees=45
  twistMotion=Free
 }
 {
  bone1Name=spine_03
  bone2Name=spine_02
  swing1LimitDegrees=45
  swing1Motion=Free
  swing2LimitDegrees=45
  swing2Motion=Free
  twistLimitDegrees=45
  twistMotion=Free
 }
 {
  bone1Name=spine_04
  bone2Name=spine_03
  swing1LimitDegrees=45
  swing1Motion=Free
  swing2LimitDegrees=45
  swing2Motion=Free
  twistLimitDegrees=45
  twistMotion=Free
 }
 {
  bone1Name=spine_05
  bone2Name=spine_04
  swing1LimitDegrees=45
  swing1Motion=Free
  swing2LimitDegrees=45
  swing2Motion=Free
  twistLimitDegrees=45
  twistMotion=Free
 }
 {
  bone1Name=neck_02
  bone2Name=spine_05
  swing1LimitDegrees=45
  swing1Motion=Free
  swing2LimitDegrees=45
  swing2Motion=Free
  twistLimitDegrees=45
  twistMotion=Free
 }
 {
  bone1Name=head
  bone2Name=neck_02
  swing1LimitDegrees=45
  swing1Motion=Free
  swing2LimitDegrees=45
  swing2Motion=Free
  twistLimitDegrees=45
  twistMotion=Free
 }
 {
  bone1Name=thigh_l
  bone2Name=pelvis
  swing1LimitDegrees=45
  swing1Motion=Free
  swing2LimitDegrees=45
  swing2Motion=Free
  twistLimitDegrees=45
  twistMotion=Free
 }
 {
  bone1Name=foot_l
  bone2Name=calf_l
  swing1LimitDegrees=45
  swing1Motion=Free
  swing2LimitDegrees=45
  swing2Motion=Free
  twistLimitDegrees=45
  twistMotion=Free
 }
 {
  bone1Name=clavicle_r
  bone2Name=spine_05
  swing1LimitDegrees=45
  swing1Motion=Free
  swing2LimitDegrees=45
  swing2Motion=Free
  twistLimitDegrees=45
  twistMotion=Free
 }
 {
  bone1Name=upperarm_r
  bone2Name=clavicle_r
  swing1LimitDegrees=45
  swing1Motion=Free
  swing2LimitDegrees=45
  swing2Motion=Free
  twistLimitDegrees=45
  twistMotion=Free
 }
 {
  bone1Name=lowerarm_r
  bone2Name=upperarm_r
  swing1LimitDegrees=45
  swing1Motion=Free
  swing2LimitDegrees=45
  swing2Motion=Free
  twistLimitDegrees=45
  twistMotion=Free
 }
 {
  bone1Name=foot_r
  bone2Name=calf_r
  swing1LimitDegrees=45
  swing1Motion=Free
  swing2LimitDegrees=45
  swing2Motion=Free
  twistLimitDegrees=45
  twistMotion=Free
 }
 {
  bone1Name=lowerarm_l
  bone2Name=upperarm_l
  swing1LimitDegrees=45
  swing1Motion=Free
  swing2LimitDegrees=45
  swing2Motion=Free
  twistLimitDegrees=45
  twistMotion=Free
 }
 {
  bone1Name=upperarm_l
  bone2Name=clavicle_l
  swing1LimitDegrees=45
  swing1Motion=Free
  swing2LimitDegrees=45
  swing2Motion=Free
  twistLimitDegrees=45
  twistMotion=Free
 }
 {
  bone1Name=clavicle_l
  bone2Name=spine_05
  swing1LimitDegrees=45
  swing1Motion=Free
  swing2LimitDegrees=45
  swing2Motion=Free
  twistLimitDegrees=45
  twistMotion=Free
 }
 {
  bone1Name=calf_r
  bone2Name=thigh_r
  swing1LimitDegrees=45
  swing1Motion=Free
  swing2LimitDegrees=45
  swing2Motion=Free
  twistLimitDegrees=45
  twistMotion=Free
 }
 {
  bone1Name=thigh_r
  bone2Name=pelvis
  swing1LimitDegrees=45
  swing1Motion=Free
  swing2LimitDegrees=45
  swing2Motion=Free
  twistLimitDegrees=45
  twistMotion=Free
 }
 {
  bone1Name=calf_l
  bone2Name=thigh_l
  swing1LimitDegrees=45
  swing1Motion=Free
  swing2LimitDegrees=45
  swing2Motion=Free
  twistLimitDegrees=45
  twistMotion=Free
 }
 {
  bone1Name=dyn_L_side_chainMain_1
  bone2Name=pelvis
  swing1LimitDegrees=45
  swing1Motion=Free
  swing2LimitDegrees=45
  swing2Motion=Free
  twistLimitDegrees=45
  twistMotion=Free
 }
 {
  bone1Name=dyn_L_side_chainMain_2
  bone2Name=dyn_L_side_chainMain_1
  swing1LimitDegrees=45
  swing1Motion=Free
  swing2LimitDegrees=45
  swing2Motion=Free
  twistLimitDegrees=45
  twistMotion=Free
 }
 {
  bone1Name=dyn_L_sideChain_A_1
  bone2Name=dyn_L_side_chainMain_2
  swing1LimitDegrees=45
  swing1Motion=Free
  swing2LimitDegrees=45
  swing2Motion=Free
  twistLimitDegrees=45
  twistMotion=Free
 }
 {
  bone1Name=dyn_L_sideChain_A_2
  bone2Name=dyn_L_sideChain_A_1
  swing1LimitDegrees=45
  swing1Motion=Free
  swing2LimitDegrees=45
  swing2Motion=Free
  twistLimitDegrees=45
  twistMotion=Free
 }
 {
  bone1Name=dyn_L_sideChain_A_3
  bone2Name=dyn_L_sideChain_A_2
  swing1LimitDegrees=45
  swing1Motion=Free
  swing2LimitDegrees=45
  swing2Motion=Free
  twistLimitDegrees=45
  twistMotion=Free
 }
 {
  bone1Name=dyn_L_sideChain_A_4
  bone2Name=dyn_L_sideChain_A_3
  swing1LimitDegrees=45
  swing1Motion=Free
  swing2LimitDegrees=45
  swing2Motion=Free
  twistLimitDegrees=45
  twistMotion=Free
 }
 {
  bone1Name=dyn_L_sideChain_B_1
  bone2Name=dyn_L_side_chainMain_2
  swing1LimitDegrees=45
  swing1Motion=Free
  swing2LimitDegrees=45
  swing2Motion=Free
  twistLimitDegrees=45
  twistMotion=Free
 }
 {
  bone1Name=dyn_L_sideChain_B_2
  bone2Name=dyn_L_sideChain_B_1
  swing1LimitDegrees=45
  swing1Motion=Free
  swing2LimitDegrees=45
  swing2Motion=Free
  twistLimitDegrees=45
  twistMotion=Free
 }
 {
  bone1Name=dyn_L_sideChain_B_3
  bone2Name=dyn_L_sideChain_B_2
  swing1LimitDegrees=45
  swing1Motion=Free
  swing2LimitDegrees=45
  swing2Motion=Free
  twistLimitDegrees=45
  twistMotion=Free
 }
 {
  bone1Name=dyn_L_sideChain_B_4
  bone2Name=dyn_L_sideChain_B_3
  swing1LimitDegrees=45
  swing1Motion=Free
  swing2LimitDegrees=45
  swing2Motion=Free
  twistLimitDegrees=45
  twistMotion=Free
 }
 {
  bone1Name=dyn_R_side_chainMain_1
  bone2Name=pelvis
  swing1LimitDegrees=45
  swing1Motion=Free
  swing2LimitDegrees=45
  swing2Motion=Free
  twistLimitDegrees=45
  twistMotion=Free
 }
 {
  bone1Name=dyn_R_side_chainMain_2
  bone2Name=dyn_R_side_chainMain_1
  swing1LimitDegrees=45
  swing1Motion=Free
  swing2LimitDegrees=45
  swing2Motion=Free
  twistLimitDegrees=45
  twistMotion=Free
 }
 {
  bone1Name=dyn_R_sideChain_A_1
  bone2Name=dyn_R_side_chainMain_2
  swing1LimitDegrees=45
  swing1Motion=Free
  swing2LimitDegrees=45
  swing2Motion=Free
  twistLimitDegrees=45
  twistMotion=Free
 }
 {
  bone1Name=dyn_R_sideChain_A_2
  bone2Name=dyn_R_sideChain_A_1
  swing1LimitDegrees=45
  swing1Motion=Free
  swing2LimitDegrees=45
  swing2Motion=Free
  twistLimitDegrees=45
  twistMotion=Free
 }
 {
  bone1Name=dyn_R_sideChain_A_3
  bone2Name=dyn_R_sideChain_A_2
  swing1LimitDegrees=45
  swing1Motion=Free
  swing2LimitDegrees=45
  swing2Motion=Free
  twistLimitDegrees=45
  twistMotion=Free
 }
 {
  bone1Name=dyn_R_sideChain_A_4
  bone2Name=dyn_R_sideChain_A_3
  swing1LimitDegrees=45
  swing1Motion=Free
  swing2LimitDegrees=45
  swing2Motion=Free
  twistLimitDegrees=45
  twistMotion=Free
 }
 {
  bone1Name=dyn_R_sideChain_B_1
  bone2Name=dyn_R_side_chainMain_2
  swing1LimitDegrees=45
  swing1Motion=Free
  swing2LimitDegrees=45
  swing2Motion=Free
  twistLimitDegrees=45
  twistMotion=Free
 }
 {
  bone1Name=dyn_R_sideChain_B_2
  bone2Name=dyn_R_sideChain_B_1
  swing1LimitDegrees=45
  swing1Motion=Free
  swing2LimitDegrees=45
  swing2Motion=Free
  twistLimitDegrees=45
  twistMotion=Free
 }
 {
  bone1Name=dyn_R_sideChain_B_3
  bone2Name=dyn_R_sideChain_B_2
  swing1LimitDegrees=45
  swing1Motion=Free
  swing2LimitDegrees=45
  swing2Motion=Free
  twistLimitDegrees=45
  twistMotion=Free
 }
 {
  bone1Name=dyn_R_sideChain_B_4
  bone2Name=dyn_R_sideChain_B_3
  swing1LimitDegrees=45
  swing1Motion=Free
  swing2LimitDegrees=45
  swing2Motion=Free
  twistLimitDegrees=45
  twistMotion=Free
 }
 {
  bone1Name=dyn_back_chainMain_1
  bone2Name=pelvis
  swing1LimitDegrees=45
  swing1Motion=Free
  swing2LimitDegrees=45
  swing2Motion=Free
  twistLimitDegrees=45
  twistMotion=Free
 }
 {
  bone1Name=dyn_back_chainMain_2
  bone2Name=dyn_back_chainMain_1
  swing1LimitDegrees=45
  swing1Motion=Free
  swing2LimitDegrees=45
  swing2Motion=Free
  twistLimitDegrees=45
  twistMotion=Free
 }
 {
  bone1Name=dyn_back_chain_1
  bone2Name=dyn_back_chainMain_2
  swing1LimitDegrees=45
  swing1Motion=Free
  swing2LimitDegrees=45
  swing2Motion=Free
  twistLimitDegrees=45
  twistMotion=Free
 }
 {
  bone1Name=dyn_back_chain_2
  bone2Name=dyn_back_chain_1
  swing1LimitDegrees=45
  swing1Motion=Free
  swing2LimitDegrees=45
  swing2Motion=Free
  twistLimitDegrees=45
  twistMotion=Free
 }
 {
  bone1Name=dyn_back_chain_3
  bone2Name=dyn_back_chain_2
  swing1LimitDegrees=45
  swing1Motion=Free
  swing2LimitDegrees=45
  swing2Motion=Free
  twistLimitDegrees=45
  twistMotion=Free
 }
]
