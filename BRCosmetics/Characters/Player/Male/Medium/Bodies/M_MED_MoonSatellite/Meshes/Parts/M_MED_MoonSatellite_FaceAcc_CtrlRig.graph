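# Control Rig for the MoonSatellite face accessory that adjusts the character's hair fin bones as the head rotates, likely to avoid clipping.
# It remaps the head's Z Euler angle relative to spine_05 into a weight for rotation offsets on side fin bones, including the left and right B chains.
# A second offset rotates the fin base by the same weight, gated to zero unless isWearingBackBling is true.
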ControlRigBlueprint /BRCosmetics/Characters/Player/Male/Medium/Bodies/M_MED_MoonSatellite/Meshes/Parts/M_MED_MoonSatellite_FaceAcc_CtrlRig
graphs [
 RigVMFunctionLibrary {links [], nodes []}
 RigVMModel {nodes [
   RigUnit_BeginExecution = &RigUnit_BeginExecution()
   n1 = &RigVMFunction_MathDoubleRemap(
    SourceMaximum=-15
    SourceMinimum=10
    TargetMaximum=1
    TargetMinimum=0
    Value=&RigVMFunction_MathQuaternionToEuler(
     RotationOrder=ZXY
     Value=&RigUnit_GetRelativeTransformForItem(Child=(Type=Bone,Name="head"), Parent=(Type=Bone,Name="spine_05"), bChildInitial=False, bParentInitial=False).RelativeTransform.Rotation
    ).Result.Z
   )
   &RigUnit_ModifyTransforms(
    ItemToModify="((Item=(Type=Bone,Name=\"dyn_L_hairFin_B_1\"),Transform=(Rotation=(X=-0.0151,Y=-0.0858,Z=-0.173,W=0.9811),Translation=(X=0,Y=0,Z=0),Scale3D=(X=1,Y=1,Z=1))),(Item=(Type=Bone,Name=\"dyn_R_hairFin_B_1\"),Transform=(Rotation=(X=-0.0151,Y=-0.0858,Z=-0.173,W=0.9811) ...(truncated)"
    WeightMinimum=0
    Weight=@n1.Result
    ExecutePin=@RigUnit_BeginExecution.ExecutePin
   )
   RigVMFunction_MathBoolToFloat = &RigVMFunction_MathBoolToFloat(Value=RigVMVariableNode() {variable=isWearingBackBling}.Value)
   Multiply = &RigVMFunction_MathDoubleMul(B=@n1.Result, A=@RigVMFunction_MathBoolToFloat.Result)
   > &RigUnit_ModifyTransforms(
    ItemToModify=((Item=(Type=Bone,Name="dyn_hairFin_base"),Transform=(Rotation=(X=0,Y=-0.1736,Z=0,W=0.9848),Translation=(X=0,Y=0,Z=0),Scale3D=(X=1,Y=1,Z=1))))
    WeightMinimum=0
    Weight=@Multiply.Result
   )
  ]}
]
hierarchy [
 root {loc=<0 0 0>, rot=<0 0 0 1>}
 pelvis {parent=root, loc=<0 -0.7071 93.1191>, rot=<0.0127 0.707 -0.0127 -0.707>}
 spine_01 {parent=pelvis, loc=<6.7494 -0.6388 0>, rot=<0 0 0.034 -0.9994>}
 spine_02 {parent=spine_01, loc=<7.252 -0.035 0>, rot=<0 0 0.0238 -0.9997>}
 spine_03 {parent=spine_02, loc=<6.0106 0 0>, rot=<0 0 0.0149 -0.9999>}
 spine_04 {parent=spine_03, loc=<8.1023 0 0>, rot=<0 0 -0.0132 -0.9999>}
 spine_05 {parent=spine_04, loc=<10.0142 0 0>, rot=<0 0 -0.0015 -1>}
 neck_01 {parent=spine_05, loc=<11.3138 0.1529 0>, rot=<0 0 -0.2003 -0.9797>}
 neck_02 {parent=neck_01, loc=<4.5627 0 0>, rot=<0 0 0.0504 -0.9987>}
 head {parent=neck_02, loc=<5.3914 0 0>, rot=<0 0 0.115 -0.9934>}
 dyn_hairFin_base {parent=head, loc=<10.9572 -9.6399 0.0068>, rot=<0.5021 -0.4979 0.4979 -0.5021>}
 dyn_top_hairFin_1 {parent=dyn_hairFin_base, loc=<0.9509 -0.3556 2.96>, rot=<-0.609 0.3778 -0.3355 -0.6114>}
 dyn_top_hairFin_2 {parent=dyn_top_hairFin_1, loc=<5.847 0 0>, rot=<0.0029 -0.0054 0.1113 -0.9938>}
 dyn_L_hairFin_A_1 {parent=dyn_hairFin_base, loc=<1.7073 1.7203 1.0674>, rot=<-0.2027 0.0902 -0.3964 -0.8909>}
 dyn_L_hairFin_A_2 {parent=dyn_L_hairFin_A_1, loc=<5.0912 1.31 -0.0176>, rot=<0.0064 -0.0066 0.0395 -0.9992>}
 dyn_L_hairFin_A_3 {parent=dyn_L_hairFin_A_2, loc=<6.6276 -0.2823 -0.0283>, rot=<0.0432 -0.0795 0.0861 -0.9922>}
 dyn_L_hairFin_A_4 {parent=dyn_L_hairFin_A_2, loc=<5.0929 -1.8539 -5.4239>, rot=<0.4309 -0.2214 0.1633 -0.8595>}
 dyn_L_hairFin_B_1 {parent=dyn_hairFin_base, loc=<1.6375 1.3448 -0.6081>, rot=<0.335 -0.1645 -0.4088 -0.8328>}
 dyn_L_hairFin_B_2 {parent=dyn_L_hairFin_B_1, loc=<6.8664 1.231 -0.4612>, rot=<-0.0911 -0.0452 0.0786 -0.9917>}
 dyn_L_hairFin_B_3 {parent=dyn_L_hairFin_B_2, loc=<6.0159 -0.9229 4.0361>, rot=<0.0562 0.2168 -0.0475 -0.9734>}
 dyn_L_hairFin_B_4 {parent=dyn_L_hairFin_B_2, loc=<5.602 -0.7799 -0.2097>, rot=<0.0917 0.0452 -0.0637 -0.9927>}
 dyn_L_hairFin_B_5 {parent=dyn_L_hairFin_B_2, loc=<4.9164 0.5159 -6.4151>, rot=<-0.2995 -0.2547 -0.2233 -0.8919>}
 dyn_R_hairFin_A_1 {parent=dyn_hairFin_base, loc=<1.7073 -1.7067 1.0676>, rot=<-0.3964 -0.8909 0.2027 -0.0902>}
 dyn_R_hairFin_A_2 {parent=dyn_R_hairFin_A_1, loc=<-5.1343 -1.8575 0.0254>, rot=<0.0064 -0.0066 0.0395 -0.9992>}
 dyn_R_hairFin_A_3 {parent=dyn_R_hairFin_A_2, loc=<-6.6843 0.4939 0.0617>, rot=<0.0495 -0.0047 0.0826 -0.9953>}
 dyn_R_hairFin_A_4 {parent=dyn_R_hairFin_A_2, loc=<-4.8069 2.6373 5.9758>, rot=<0.2287 -0.1795 0.2085 -0.9338>}
 dyn_R_hairFin_B_1 {parent=dyn_hairFin_base, loc=<1.6376 -1.3312 -0.6084>, rot=<-0.4088 -0.8328 -0.335 0.1645>}
 dyn_R_hairFin_B_2 {parent=dyn_R_hairFin_B_1, loc=<-6.8664 -1.231 0.4613>, rot=<-0.0911 -0.0452 0.0786 -0.9917>}
 dyn_R_hairFin_B_3 {parent=dyn_R_hairFin_B_2, loc=<-5.5244 -0.7775 -4.6229>, rot=<-0.0956 0.2193 -0.0191 -0.9708>}
 dyn_R_hairFin_B_4 {parent=dyn_R_hairFin_B_2, loc=<-7.6983 0.9434 0.6784>, rot=<0.0453 -0.0618 0.0311 -0.9966>}
 dyn_R_hairFin_B_5 {parent=dyn_R_hairFin_B_2, loc=<-3.7271 -0.4221 6.6597>, rot=<-0.1132 -0.2918 -0.1047 -0.944>}
]
variables [isWearingBackBling {type=bool, default=False}]
