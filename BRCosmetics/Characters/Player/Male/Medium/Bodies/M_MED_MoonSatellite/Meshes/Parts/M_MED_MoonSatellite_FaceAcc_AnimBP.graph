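# Anim Blueprint for the MoonSatellite male face accessory, adding physics jiggle to its hair fins.
# It copies the parent mesh pose into a Control Rig that receives a backbling flag, set when customization loads.
# Chained single-bone AnimDynamics nodes then simulate top, left and right fins root-relative with angular springs and reduced gravity.
# The event graph sets frontend alpha and simulation-space variables, but the AnimGraph does not read them.
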AnimBlueprint /BRCosmetics/Characters/Player/Male/Medium/Bodies/M_MED_MoonSatellite/Meshes/Parts/M_MED_MoonSatellite_FaceAcc_AnimBP
target_skeleton=&/M_MED_MoonSatellite_FaceAcc_Skeleton
graphs [
 AnimGraph {nodes [
   n1 = Control Rig!(Source=Copy Pose From Mesh() {bUseAttachedParent=true}, Alpha, isWearingBackBling:bool=Get Is Wearing Backbling()) -> (Pose) {
    alphaBoolBlend {blendOption=Linear}
    controlRigAssetReference {blueprintRigClass=&/M_MED_MoonSatellite_FaceAcc_CtrlRig.M_MED_MoonSatellite_FaceAcc_CtrlRig_C}
   }
   n2 = Anim Dynamics - Bone: dyn_top_hairFin_1!(SimulationSpace:AnimPhysSimSpaceType Enum, ComponentPose=Local To Component(LocalPose=@n1), Alpha=1) -> (Pose) {
    angularDampingOverride=1
    angularSpringConstant=23000
    bAngularSpring=true
    bOverrideAngularDamping=true
    bUseGravityOverride=true
    gravityScale=0.35
    simulationSpace=RootRelative
    alphaBoolBlend {blendOption=Linear}
    boundBone {boneName=dyn_top_hairFin_1}
    componentAppliedLinearAccClamp=<2000 2000 2000>
    componentLinearAccScale=<1 1 1>
    componentLinearVelScale=<1 1 1>
    gravityOverride {z=-980}
    physicsBodyDefinitions [{
      boundBone {boneName=dyn_top_hairFin_1}
      boxExtents=<30 30 30>
      constraintSetup {angularLimitsMax=<15 15 15>, angularLimitsMin=<-15 -15 -15>, angularTarget=<1 0 0>}
      localJointOffset=<3 0 0>
     }]
   }
   n3 = Anim Dynamics - Bone: dyn_top_hairFin_2!(SimulationSpace:AnimPhysSimSpaceType Enum, ComponentPose=@n2, Alpha=1) -> (Pose) {
    angularDampingOverride=1
    angularSpringConstant=23000
    bAngularSpring=true
    bOverrideAngularDamping=true
    bUseGravityOverride=true
    gravityScale=0.35
    simulationSpace=RootRelative
    alphaBoolBlend {blendOption=Linear}
    boundBone {boneName=dyn_top_hairFin_2}
    componentAppliedLinearAccClamp=<2000 2000 2000>
    componentLinearAccScale=<1 1 1>
    componentLinearVelScale=<1 1 1>
    gravityOverride {z=-980}
    physicsBodyDefinitions [{
      boundBone {boneName=dyn_top_hairFin_2}
      boxExtents=<30 30 30>
      constraintSetup {angularLimitsMax=<15 15 15>, angularLimitsMin=<-15 -15 -15>, angularTarget=<1 0 0>}
      localJointOffset=<3 0 0>
     }]
   }
   n4 = Anim Dynamics - Bone: dyn_L_hairFin_A_1!(SimulationSpace:AnimPhysSimSpaceType Enum, ComponentPose=@n3, Alpha=1) -> (Pose) {
    angularDampingOverride=1
    angularSpringConstant=23000
    bAngularSpring=true
    bOverrideAngularDamping=true
    bUseGravityOverride=true
    gravityScale=0.35
    simulationSpace=RootRelative
    alphaBoolBlend {blendOption=Linear}
    boundBone {boneName=dyn_L_hairFin_A_1}
    componentAppliedLinearAccClamp=<2000 2000 2000>
    componentLinearAccScale=<1 1 1>
    componentLinearVelScale=<1 1 1>
    gravityOverride {z=-980}
    physicsBodyDefinitions [{
      boundBone {boneName=dyn_L_hairFin_A_1}
      boxExtents=<30 30 30>
      constraintSetup {angularLimitsMax=<15 15 15>, angularLimitsMin=<-15 -15 -15>, angularTarget=<1 0 0>}
      localJointOffset=<3 0 0>
     }]
   }
   n5 = Anim Dynamics - Bone: dyn_L_hairFin_A_2!(SimulationSpace:AnimPhysSimSpaceType Enum, ComponentPose=@n4, Alpha=1) -> (Pose) {
    angularDampingOverride=1
    angularSpringConstant=23000
    bAngularSpring=true
    bOverrideAngularDamping=true
    bUseGravityOverride=true
    gravityScale=0.35
    simulationSpace=RootRelative
    alphaBoolBlend {blendOption=Linear}
    boundBone {boneName=dyn_L_hairFin_A_2}
    componentAppliedLinearAccClamp=<2000 2000 2000>
    componentLinearAccScale=<1 1 1>
    componentLinearVelScale=<1 1 1>
    gravityOverride {z=-980}
    physicsBodyDefinitions [{
      boundBone {boneName=dyn_L_hairFin_A_2}
      boxExtents=<30 30 30>
      constraintSetup {angularLimitsMax=<15 15 15>, angularLimitsMin=<-15 -15 -15>, angularTarget=<1 0 0>}
      localJointOffset=<3 0 0>
     }]
   }
   n6 = Anim Dynamics - Bone: dyn_L_hairFin_A_3!(SimulationSpace:AnimPhysSimSpaceType Enum, ComponentPose=@n5, Alpha=1) -> (Pose) {
    angularDampingOverride=1
    angularSpringConstant=23000
    bAngularSpring=true
    bOverrideAngularDamping=true
    bUseGravityOverride=true
    gravityScale=0.35
    simulationSpace=RootRelative
    alphaBoolBlend {blendOption=Linear}
    boundBone {boneName=dyn_L_hairFin_A_3}
    componentAppliedLinearAccClamp=<2000 2000 2000>
    componentLinearAccScale=<1 1 1>
    componentLinearVelScale=<1 1 1>
    gravityOverride {z=-980}
    physicsBodyDefinitions [{
      boundBone {boneName=dyn_L_hairFin_A_3}
      boxExtents=<30 30 30>
      constraintSetup {angularLimitsMax=<15 15 15>, angularLimitsMin=<-15 -15 -15>, angularTarget=<1 0 0>}
      localJointOffset=<3 0 0>
     }]
   }
   n7 = Anim Dynamics - Bone: dyn_L_hairFin_A_4!(SimulationSpace:AnimPhysSimSpaceType Enum, ComponentPose=@n6, Alpha=1) -> (Pose) {
    angularDampingOverride=1
    angularSpringConstant=23000
    bAngularSpring=true
    bOverrideAngularDamping=true
    bUseGravityOverride=true
    gravityScale=0.35
    simulationSpace=RootRelative
    alphaBoolBlend {blendOption=Linear}
    boundBone {boneName=dyn_L_hairFin_A_4}
    componentAppliedLinearAccClamp=<2000 2000 2000>
    componentLinearAccScale=<1 1 1>
    componentLinearVelScale=<1 1 1>
    gravityOverride {z=-980}
    physicsBodyDefinitions [{
      boundBone {boneName=dyn_L_hairFin_A_4}
      boxExtents=<30 30 30>
      constraintSetup {angularLimitsMax=<15 15 15>, angularLimitsMin=<-15 -15 -15>, angularTarget=<1 0 0>}
      localJointOffset=<3 0 0>
     }]
   }
   n8 = Anim Dynamics - Bone: dyn_L_hairFin_B_1!(SimulationSpace:AnimPhysSimSpaceType Enum, ComponentPose=@n7, Alpha=1) -> (Pose) {
    angularDampingOverride=1
    angularSpringConstant=23000
    bAngularSpring=true
    bOverrideAngularDamping=true
    bUseGravityOverride=true
    gravityScale=0.35
    simulationSpace=RootRelative
    alphaBoolBlend {blendOption=Linear}
    boundBone {boneName=dyn_L_hairFin_B_1}
    componentAppliedLinearAccClamp=<2000 2000 2000>
    componentLinearAccScale=<1 1 1>
    componentLinearVelScale=<1 1 1>
    gravityOverride {z=-980}
    physicsBodyDefinitions [{
      boundBone {boneName=dyn_L_hairFin_B_1}
      boxExtents=<30 30 30>
      constraintSetup {angularLimitsMax=<15 15 15>, angularLimitsMin=<-15 -15 -15>, angularTarget=<1 0 0>}
      localJointOffset=<3 0 0>
     }]
   }
   n9 = Anim Dynamics - Bone: dyn_L_hairFin_B_2!(SimulationSpace:AnimPhysSimSpaceType Enum, ComponentPose=@n8, Alpha=1) -> (Pose) {
    angularDampingOverride=1
    angularSpringConstant=23000
    bAngularSpring=true
    bOverrideAngularDamping=true
    bUseGravityOverride=true
    gravityScale=0.35
    simulationSpace=RootRelative
    alphaBoolBlend {blendOption=Linear}
    boundBone {boneName=dyn_L_hairFin_B_2}
    componentAppliedLinearAccClamp=<2000 2000 2000>
    componentLinearAccScale=<1 1 1>
    componentLinearVelScale=<1 1 1>
    gravityOverride {z=-980}
    physicsBodyDefinitions [{
      boundBone {boneName=dyn_L_hairFin_B_2}
      boxExtents=<30 30 30>
      constraintSetup {angularLimitsMax=<15 15 15>, angularLimitsMin=<-15 -15 -15>, angularTarget=<1 0 0>}
      localJointOffset=<3 0 0>
     }]
   }
   n10 = Anim Dynamics - Bone: dyn_L_hairFin_B_3!(SimulationSpace:AnimPhysSimSpaceType Enum, ComponentPose=@n9, Alpha=1) -> (Pose) {
    angularDampingOverride=1
    angularSpringConstant=23000
    bAngularSpring=true
    bOverrideAngularDamping=true
    bUseGravityOverride=true
    gravityScale=0.35
    simulationSpace=RootRelative
    alphaBoolBlend {blendOption=Linear}
    boundBone {boneName=dyn_L_hairFin_B_3}
    componentAppliedLinearAccClamp=<2000 2000 2000>
    componentLinearAccScale=<1 1 1>
    componentLinearVelScale=<1 1 1>
    gravityOverride {z=-980}
    physicsBodyDefinitions [{
      boundBone {boneName=dyn_L_hairFin_B_3}
      boxExtents=<30 30 30>
      constraintSetup {angularLimitsMax=<15 15 15>, angularLimitsMin=<-15 -15 -15>, angularTarget=<1 0 0>}
      localJointOffset=<3 0 0>
     }]
   }
   n11 = Anim Dynamics - Bone: dyn_L_hairFin_B_4!(SimulationSpace:AnimPhysSimSpaceType Enum, ComponentPose=@n10, Alpha=1) -> (Pose) {
    angularDampingOverride=1
    angularSpringConstant=23000
    bAngularSpring=true
    bOverrideAngularDamping=true
    bUseGravityOverride=true
    gravityScale=0.35
    simulationSpace=RootRelative
    alphaBoolBlend {blendOption=Linear}
    boundBone {boneName=dyn_L_hairFin_B_4}
    componentAppliedLinearAccClamp=<2000 2000 2000>
    componentLinearAccScale=<1 1 1>
    componentLinearVelScale=<1 1 1>
    gravityOverride {z=-980}
    physicsBodyDefinitions [{
      boundBone {boneName=dyn_L_hairFin_B_4}
      boxExtents=<30 30 30>
      constraintSetup {angularLimitsMax=<15 15 15>, angularLimitsMin=<-15 -15 -15>, angularTarget=<1 0 0>}
      localJointOffset=<3 0 0>
     }]
   }
   n12 = Anim Dynamics - Bone: dyn_L_hairFin_B_5!(SimulationSpace:AnimPhysSimSpaceType Enum, ComponentPose=@n11, Alpha=1) -> (Pose) {
    angularDampingOverride=1
    angularSpringConstant=23000
    bAngularSpring=true
    bOverrideAngularDamping=true
    bUseGravityOverride=true
    gravityScale=0.35
    simulationSpace=RootRelative
    alphaBoolBlend {blendOption=Linear}
    boundBone {boneName=dyn_L_hairFin_B_5}
    componentAppliedLinearAccClamp=<2000 2000 2000>
    componentLinearAccScale=<1 1 1>
    componentLinearVelScale=<1 1 1>
    gravityOverride {z=-980}
    physicsBodyDefinitions [{
      boundBone {boneName=dyn_L_hairFin_B_5}
      boxExtents=<30 30 30>
      constraintSetup {angularLimitsMax=<15 15 15>, angularLimitsMin=<-15 -15 -15>, angularTarget=<1 0 0>}
      localJointOffset=<3 0 0>
     }]
   }
   n13 = Anim Dynamics - Bone: dyn_R_hairFin_A_1!(SimulationSpace:AnimPhysSimSpaceType Enum, ComponentPose=@n12, Alpha=1) -> (Pose) {
    angularDampingOverride=1
    angularSpringConstant=23000
    bAngularSpring=true
    bOverrideAngularDamping=true
    bUseGravityOverride=true
    gravityScale=0.35
    simulationSpace=RootRelative
    alphaBoolBlend {blendOption=Linear}
    boundBone {boneName=dyn_R_hairFin_A_1}
    componentAppliedLinearAccClamp=<2000 2000 2000>
    componentLinearAccScale=<1 1 1>
    componentLinearVelScale=<1 1 1>
    gravityOverride {z=-980}
    physicsBodyDefinitions [{
      boundBone {boneName=dyn_R_hairFin_A_1}
      boxExtents=<33 33 33>
      constraintSetup {angularLimitsMax=<15 15 15>, angularLimitsMin=<-15 -15 -15>, angularTarget=<1 0 0>}
      localJointOffset=<-3 0 0>
     }]
   }
   n14 = Anim Dynamics - Bone: dyn_R_hairFin_A_2!(SimulationSpace:AnimPhysSimSpaceType Enum, ComponentPose=@n13, Alpha=1) -> (Pose) {
    angularDampingOverride=1
    angularSpringConstant=23000
    bAngularSpring=true
    bOverrideAngularDamping=true
    bUseGravityOverride=true
    gravityScale=0.35
    simulationSpace=RootRelative
    alphaBoolBlend {blendOption=Linear}
    boundBone {boneName=dyn_R_hairFin_A_2}
    componentAppliedLinearAccClamp=<2000 2000 2000>
    componentLinearAccScale=<1 1 1>
    componentLinearVelScale=<1 1 1>
    gravityOverride {z=-980}
    physicsBodyDefinitions [{
      boundBone {boneName=dyn_R_hairFin_A_2}
      boxExtents=<33 33 33>
      constraintSetup {angularLimitsMax=<15 15 15>, angularLimitsMin=<-15 -15 -15>, angularTarget=<1 0 0>}
      localJointOffset=<-3 0 0>
     }]
   }
   n15 = Anim Dynamics - Bone: dyn_R_hairFin_A_3!(SimulationSpace:AnimPhysSimSpaceType Enum, ComponentPose=@n14, Alpha=1) -> (Pose) {
    angularDampingOverride=1
    angularSpringConstant=23000
    bAngularSpring=true
    bOverrideAngularDamping=true
    bUseGravityOverride=true
    gravityScale=0.35
    simulationSpace=RootRelative
    alphaBoolBlend {blendOption=Linear}
    boundBone {boneName=dyn_R_hairFin_A_3}
    componentAppliedLinearAccClamp=<2000 2000 2000>
    componentLinearAccScale=<1 1 1>
    componentLinearVelScale=<1 1 1>
    gravityOverride {z=-980}
    physicsBodyDefinitions [{
      boundBone {boneName=dyn_R_hairFin_A_3}
      boxExtents=<33 33 33>
      constraintSetup {angularLimitsMax=<15 15 15>, angularLimitsMin=<-15 -15 -15>, angularTarget=<1 0 0>}
      localJointOffset=<-3 0 0>
     }]
   }
   n16 = Anim Dynamics - Bone: dyn_R_hairFin_A_4!(SimulationSpace:AnimPhysSimSpaceType Enum, ComponentPose=@n15, Alpha=1) -> (Pose) {
    angularDampingOverride=1
    angularSpringConstant=23000
    bAngularSpring=true
    bOverrideAngularDamping=true
    bUseGravityOverride=true
    gravityScale=0.35
    simulationSpace=RootRelative
    alphaBoolBlend {blendOption=Linear}
    boundBone {boneName=dyn_R_hairFin_A_4}
    componentAppliedLinearAccClamp=<2000 2000 2000>
    componentLinearAccScale=<1 1 1>
    componentLinearVelScale=<1 1 1>
    gravityOverride {z=-980}
    physicsBodyDefinitions [{
      boundBone {boneName=dyn_R_hairFin_A_4}
      boxExtents=<33 33 33>
      constraintSetup {angularLimitsMax=<15 15 15>, angularLimitsMin=<-15 -15 -15>, angularTarget=<1 0 0>}
      localJointOffset=<-3 0 0>
     }]
   }
   n17 = Anim Dynamics - Bone: dyn_R_hairFin_B_1!(SimulationSpace:AnimPhysSimSpaceType Enum, ComponentPose=@n16, Alpha=1) -> (Pose) {
    angularDampingOverride=1
    angularSpringConstant=23000
    bAngularSpring=true
    bOverrideAngularDamping=true
    bUseGravityOverride=true
    gravityScale=0.35
    simulationSpace=RootRelative
    alphaBoolBlend {blendOption=Linear}
    boundBone {boneName=dyn_R_hairFin_B_1}
    componentAppliedLinearAccClamp=<2000 2000 2000>
    componentLinearAccScale=<1 1 1>
    componentLinearVelScale=<1 1 1>
    gravityOverride {z=-980}
    physicsBodyDefinitions [{
      boundBone {boneName=dyn_R_hairFin_B_1}
      boxExtents=<33 33 33>
      constraintSetup {angularLimitsMax=<15 15 15>, angularLimitsMin=<-15 -15 -15>, angularTarget=<1 0 0>}
      localJointOffset=<-3 0 0>
     }]
   }
   n18 = Anim Dynamics - Bone: dyn_R_hairFin_B_2!(SimulationSpace:AnimPhysSimSpaceType Enum, ComponentPose=@n17, Alpha=1) -> (Pose) {
    angularDampingOverride=1
    angularSpringConstant=23000
    bAngularSpring=true
    bOverrideAngularDamping=true
    bUseGravityOverride=true
    gravityScale=0.35
    simulationSpace=RootRelative
    alphaBoolBlend {blendOption=Linear}
    boundBone {boneName=dyn_R_hairFin_B_2}
    componentAppliedLinearAccClamp=<2000 2000 2000>
    componentLinearAccScale=<1 1 1>
    componentLinearVelScale=<1 1 1>
    gravityOverride {z=-980}
    physicsBodyDefinitions [{
      boundBone {boneName=dyn_R_hairFin_B_2}
      boxExtents=<33 33 33>
      constraintSetup {angularLimitsMax=<15 15 15>, angularLimitsMin=<-15 -15 -15>, angularTarget=<1 0 0>}
      localJointOffset=<-3 0 0>
     }]
   }
   n19 = Anim Dynamics - Bone: dyn_R_hairFin_B_3!(SimulationSpace:AnimPhysSimSpaceType Enum, ComponentPose=@n18, Alpha=1) -> (Pose) {
    angularDampingOverride=1
    angularSpringConstant=23000
    bAngularSpring=true
    bOverrideAngularDamping=true
    bUseGravityOverride=true
    gravityScale=0.35
    simulationSpace=RootRelative
    alphaBoolBlend {blendOption=Linear}
    boundBone {boneName=dyn_R_hairFin_B_3}
    componentAppliedLinearAccClamp=<2000 2000 2000>
    componentLinearAccScale=<1 1 1>
    componentLinearVelScale=<1 1 1>
    gravityOverride {z=-980}
    physicsBodyDefinitions [{
      boundBone {boneName=dyn_R_hairFin_B_3}
      boxExtents=<33 33 33>
      constraintSetup {angularLimitsMax=<15 15 15>, angularLimitsMin=<-15 -15 -15>, angularTarget=<1 0 0>}
      localJointOffset=<-3 0 0>
     }]
   }
   n20 = Anim Dynamics - Bone: dyn_R_hairFin_B_4!(SimulationSpace:AnimPhysSimSpaceType Enum, ComponentPose=@n19, Alpha=1) -> (Pose) {
    angularDampingOverride=1
    angularSpringConstant=23000
    bAngularSpring=true
    bOverrideAngularDamping=true
    bUseGravityOverride=true
    gravityScale=0.35
    simulationSpace=RootRelative
    alphaBoolBlend {blendOption=Linear}
    boundBone {boneName=dyn_R_hairFin_B_4}
    componentAppliedLinearAccClamp=<2000 2000 2000>
    componentLinearAccScale=<1 1 1>
    componentLinearVelScale=<1 1 1>
    gravityOverride {z=-980}
    physicsBodyDefinitions [{
      boundBone {boneName=dyn_R_hairFin_B_4}
      boxExtents=<33 33 33>
      constraintSetup {angularLimitsMax=<15 15 15>, angularLimitsMin=<-15 -15 -15>, angularTarget=<1 0 0>}
      localJointOffset=<-3 0 0>
     }]
   }
   n21 = Anim Dynamics - Bone: dyn_R_hairFin_B_5!(SimulationSpace:AnimPhysSimSpaceType Enum, ComponentPose=@n20, Alpha=1) -> (Pose) {
    angularDampingOverride=1
    angularSpringConstant=23000
    bAngularSpring=true
    bOverrideAngularDamping=true
    bUseGravityOverride=true
    gravityScale=0.35
    simulationSpace=RootRelative
    alphaBoolBlend {blendOption=Linear}
    boundBone {boneName=dyn_R_hairFin_B_5}
    componentAppliedLinearAccClamp=<2000 2000 2000>
    componentLinearAccScale=<1 1 1>
    componentLinearVelScale=<1 1 1>
    gravityOverride {z=-980}
    physicsBodyDefinitions [{
      boundBone {boneName=dyn_R_hairFin_B_5}
      boxExtents=<33 33 33>
      constraintSetup {angularLimitsMax=<15 15 15>, angularLimitsMin=<-15 -15 -15>, angularTarget=<1 0 0>}
      localJointOffset=<-3 0 0>
     }]
   }
   Output Pose(Result=Component To Local(ComponentPose=@n21)) {layerGroup=None}
  ]}
 EventGraph {nodes [
   Event BlueprintUpdateAnimation()
   > Set frontendAlpha(frontendAlpha:double=SelectFloat(A=0.5, B=1, bPickA=Get bIsInFrontEnd())) -> (Output_Get:double)
   > n1 = Branch(Condition=OR Boolean(A=Get bIsInFrontEnd(), B=Get bIsDBNOCarried()))
   > Set simSpaceChange(simSpaceChange:AnimPhysSimSpaceType Enum=World) -> (Output_Get:AnimPhysSimSpaceType Enum)
   Set simSpaceChange(execute=@n1.else)
   Event OnCharacterCustomizationLoaded()
   > Set Is Wearing Backbling(Is Wearing Backbling=CheckForBackbling())
  ]}
]
variables [
 frontendAlpha {type=number}
 Is Wearing Backbling {type=boolean}
 simSpaceChange {type=AnimPhysSimSpaceType}
]
